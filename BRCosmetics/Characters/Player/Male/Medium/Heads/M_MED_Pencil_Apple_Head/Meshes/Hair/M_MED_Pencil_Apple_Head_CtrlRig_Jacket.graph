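# Control Rig for the Pencil Apple head's hair, Jacket variant, that reshapes side and back hair bones as the head rotates.
# It reads the head bone's rotation relative to spine_05 and remaps one Euler axis (Z) to a clamped weight.
# That weight drives a chain of six Modify Transforms nodes applying fixed rotation, translation and scale correctives to dyn_apl_hair bones.
# The facial and pose curves in the hierarchy are not read by the graph.
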ControlRigBlueprint /BRCosmetics/Characters/Player/Male/Medium/Heads/M_MED_Pencil_Apple_Head/Meshes/Hair/M_MED_Pencil_Apple_Head_CtrlRig_Jacket
graphs [
 RigVMFunctionLibrary {links [], nodes []}
 RigVMModel {nodes [
   RigUnit_BeginExecution = &RigUnit_BeginExecution()
   n1 = &RigVMFunction_MathFloatRemap(
    SourceMaximum=-45
    SourceMinimum=10
    TargetMinimum=0
    bClamp=true
    Value=&RigVMFunction_MathQuaternionToEuler(
     RotationOrder=ZYX
     Value=&RigUnit_GetRelativeTransformForItem(Child=(Type=Bone,Name="head"), Parent=(Type=Bone,Name="spine_05")).RelativeTransform.Rotation
    ).Result.Z
   )
   &RigUnit_ModifyTransforms(
    ItemToModify=((Item=(Type=Bone,Name="dyn_apl_hair_Side_R_MSTR"),Transform=(Rotation=(X=0,Y=0,Z=0,W=1),Translation=(X=0,Y=0,Z=0),Scale3D=(X=1,Y=1,Z=1))))
    Weight=@n1.Result
    ItemToModify.0.Transform.Rotation=&RigVMFunction_MathQuaternionFromEuler(Euler=(X=0,Y=0,Z=-20)).Result
    ExecutePin=@RigUnit_BeginExecution.ExecutePin
   )
   > &RigUnit_ModifyTransforms(
    ItemToModify=((Item=(Type=Bone,Name="dyn_apl_hair_back_R_02"),Transform=(Rotation=(X=0,Y=0,Z=0,W=1),Translation=(X=10,Y=5,Z=5),Scale3D=(X=1,Y=1,Z=0.7))))
    Weight=@n1.Result
    ItemToModify.0.Transform.Rotation=&RigVMFunction_MathQuaternionFromEuler(Euler=(X=30,Y=18,Z=0)).Result
   )
   > &RigUnit_ModifyTransforms(
    ItemToModify=((Item=(Type=Bone,Name="dyn_apl_hair_back_R_MSTR"),Transform=(Rotation=(X=0,Y=0,Z=0,W=1),Translation=(X=3,Y=0,Z=-0.05),Scale3D=(X=0.5,Y=1,Z=1))))
    ItemToModify.0.Transform.Rotation=&RigVMFunction_MathQuaternionFromEuler(Euler=(X=0,Y=25,Z=0)).Result
    Weight=@n1.Result
   )
   > &RigUnit_ModifyTransforms(
    ItemToModify=((Item=(Type=Bone,Name="dyn_apl_hair_back_02"),Transform=(Rotation=(X=0,Y=0,Z=0,W=1),Translation=(X=-3,Y=0,Z=3),Scale3D=(X=1,Y=1,Z=0.5))))
    ItemToModify.0.Transform.Rotation=&RigVMFunction_MathQuaternionFromEuler(Euler=(X=0,Y=15,Z=0)).Result
    Weight=@n1.Result
   )
   > &RigUnit_ModifyTransforms(
    ItemToModify=((Item=(Type=Bone,Name="dyn_apl_hair_back_L_MSTR"),Transform=(Rotation=(X=0,Y=0,Z=0,W=1),Translation=(X=-1,Y=1,Z=0),Scale3D=(X=1,Y=0.8,Z=0.2))))
    ItemToModify.0.Transform.Rotation=&RigVMFunction_MathQuaternionFromEuler(Euler=(X=-30,Y=-10,Z=0)).Result
    Weight=@n1.Result
   )
   > &RigUnit_ModifyTransforms(
    ItemToModify=((Item=(Type=Bone,Name="dyn_apl_hair_back_L_05"),Transform=(Rotation=(X=0,Y=0,Z=0,W=1),Translation=(X=-1,Y=0,Z=0),Scale3D=(X=1,Y=1,Z=1))))
    ItemToModify.0.Transform.Rotation=&RigVMFunction_MathQuaternionFromEuler(Euler=(X=0,Y=-20,Z=-50)).Result
    Weight=@n1.Result
   )
  ]}
]
hierarchy [
 look_left_pose {type=Curve, loc=<0 0 0>, rot=<0 0 0 1>}
 look_right_pose {type=Curve, loc=<0 0 0>, rot=<0 0 0 1>}
 look_up_pose {type=Curve, loc=<0 0 0>, rot=<0 0 0 1>}
 look_down_pose {type=Curve, loc=<0 0 0>, rot=<0 0 0 1>}
 L_blink_pose {type=Curve, loc=<0 0 0>, rot=<0 0 0 1>}
 R_blink_pose {type=Curve, loc=<0 0 0>, rot=<0 0 0 1>}
 L_wide_pose {type=Curve, loc=<0 0 0>, rot=<0 0 0 1>}
 R_wide_pose {type=Curve, loc=<0 0 0>, rot=<0 0 0 1>}
 R_squint_inner_pose {type=Curve, loc=<0 0 0>, rot=<0 0 0 1>}
 R_squeeze_pose {type=Curve, loc=<0 0 0>, rot=<0 0 0 1>}
 L_squint_inner_pose {type=Curve, loc=<0 0 0>, rot=<0 0 0 1>}
 L_squeeze_pose {type=Curve, loc=<0 0 0>, rot=<0 0 0 1>}
 L_brow_up_pose {type=Curve, loc=<0 0 0>, rot=<0 0 0 1>}
 L_brow_down_pose {type=Curve, loc=<0 0 0>, rot=<0 0 0 1>}
 R_brow_up_pose {type=Curve, loc=<0 0 0>, rot=<0 0 0 1>}
 R_brow_down_pose {type=Curve, loc=<0 0 0>, rot=<0 0 0 1>}
 C_glabella_up_pose {type=Curve, loc=<0 0 0>, rot=<0 0 0 1>}
 C_glabella_down_pose {type=Curve, loc=<0 0 0>, rot=<0 0 0 1>}
 L_brow_squeeze_in_pose {type=Curve, loc=<0 0 0>, rot=<0 0 0 1>}
 R_brow_squeeze_in_pose {type=Curve, loc=<0 0 0>, rot=<0 0 0 1>}
 L_upper_lip_raiser_pose {type=Curve, loc=<0 0 0>, rot=<0 0 0 1>}
 R_upper_lip_raiser_pose {type=Curve, loc=<0 0 0>, rot=<0 0 0 1>}
 L_upper_lip_lower_pose {type=Curve, loc=<0 0 0>, rot=<0 0 0 1>}
 R_upper_lip_lower_pose {type=Curve, loc=<0 0 0>, rot=<0 0 0 1>}
 L_lower_lip_up_pose {type=Curve, loc=<0 0 0>, rot=<0 0 0 1>}
 R_lower_lip_up_pose {type=Curve, loc=<0 0 0>, rot=<0 0 0 1>}
 L_lower_lip_down_pose {type=Curve, loc=<0 0 0>, rot=<0 0 0 1>}
 R_lower_lip_down_pose {type=Curve, loc=<0 0 0>, rot=<0 0 0 1>}
 L_lip_corner_narrow_pose {type=Curve, loc=<0 0 0>, rot=<0 0 0 1>}
 L_lip_corner_wide_pose {type=Curve, loc=<0 0 0>, rot=<0 0 0 1>}
 R_lip_corner_narrow_pose {type=Curve, loc=<0 0 0>, rot=<0 0 0 1>}
 R_lip_corner_wide_pose {type=Curve, loc=<0 0 0>, rot=<0 0 0 1>}
 L_smile_pose {type=Curve, loc=<0 0 0>, rot=<0 0 0 1>}
 R_smile_pose {type=Curve, loc=<0 0 0>, rot=<0 0 0 1>}
 L_frown_pose {type=Curve, loc=<0 0 0>, rot=<0 0 0 1>}
 R_frown_pose {type=Curve, loc=<0 0 0>, rot=<0 0 0 1>}
 phoneme_oo_pose {type=Curve, loc=<0 0 0>, rot=<0 0 0 1>}
 phoneme_ch_pose {type=Curve, loc=<0 0 0>, rot=<0 0 0 1>}
 phoneme_mbp_pose {type=Curve, loc=<0 0 0>, rot=<0 0 0 1>}
 phoneme_fv_pose {type=Curve, loc=<0 0 0>, rot=<0 0 0 1>}
 mouth_up_pose {type=Curve, loc=<0 0 0>, rot=<0 0 0 1>}
 mouth_down_pose {type=Curve, loc=<0 0 0>, rot=<0 0 0 1>}
 C_nose_wrinkler_pose {type=Curve, loc=<0 0 0>, rot=<0 0 0 1>}
 L_cheek_up_pose {type=Curve, loc=<0 0 0>, rot=<0 0 0 1>}
 R_cheek_up_pose {type=Curve, loc=<0 0 0>, rot=<0 0 0 1>}
 jaw_open_pose {type=Curve, loc=<0 0 0>, rot=<0 0 0 1>}
 jaw_right_pose {type=Curve, loc=<0 0 0>, rot=<0 0 0 1>}
 jaw_left_pose {type=Curve, loc=<0 0 0>, rot=<0 0 0 1>}
 eye_convergence_pose {type=Curve, loc=<0 0 0>, rot=<0 0 0 1>}
 L_upperLidDown_pose {type=Curve, loc=<0 0 0>, rot=<0 0 0 1>}
 R_upperLidDown_pose {type=Curve, loc=<0 0 0>, rot=<0 0 0 1>}
 jaw_up_pose {type=Curve, loc=<0 0 0>, rot=<0 0 0 1>}
 jaw_forward_pose {type=Curve, loc=<0 0 0>, rot=<0 0 0 1>}
 jaw_back_pose {type=Curve, loc=<0 0 0>, rot=<0 0 0 1>}
 tongue_forward_pose {type=Curve, loc=<0 0 0>, rot=<0 0 0 1>}
 tongue_back_pose {type=Curve, loc=<0 0 0>, rot=<0 0 0 1>}
 tongue_up_pose {type=Curve, loc=<0 0 0>, rot=<0 0 0 1>}
 tongue_down_pose {type=Curve, loc=<0 0 0>, rot=<0 0 0 1>}
 base_pose {type=Curve, loc=<0 0 0>, rot=<0 0 0 1>}
 Hair_Scarf {type=Curve, loc=<0 0 0>, rot=<0 0 0 1>}
 Hair_Hat {type=Curve, loc=<0 0 0>, rot=<0 0 0 1>}
 Hair_Helmet {type=Curve, loc=<0 0 0>, rot=<0 0 0 1>}
 Hair_Glasses {type=Curve, loc=<0 0 0>, rot=<0 0 0 1>}
 Hair_Cap {type=Curve, loc=<0 0 0>, rot=<0 0 0 1>}
 Hair_Headsets {type=Curve, loc=<0 0 0>, rot=<0 0 0 1>}
 Hair_Hood {type=Curve, loc=<0 0 0>, rot=<0 0 0 1>}
 Face_Mask {type=Curve, loc=<0 0 0>, rot=<0 0 0 1>}
 HairStrands_Hat {type=Curve, loc=<0 0 0>, rot=<0 0 0 1>}
 Hat_Hair {type=Curve, loc=<0 0 0>, rot=<0 0 0 1>}
 Helmet_Hair {type=Curve, loc=<0 0 0>, rot=<0 0 0 1>}
 Hair_Muff {type=Curve, loc=<0 0 0>, rot=<0 0 0 1>}
 Hair_SpaceSuit {type=Curve, loc=<0 0 0>, rot=<0 0 0 1>}
 Hair_Hide {type=Curve, loc=<0 0 0>, rot=<0 0 0 1>}
 Head_SpaceSuit {type=Curve, loc=<0 0 0>, rot=<0 0 0 1>}
 Hair_Null {type=Curve, loc=<0 0 0>, rot=<0 0 0 1>}
 DisableFaceOverride {type=Curve, loc=<0 0 0>, rot=<0 0 0 1>}
 LiveLink {type=Curve, loc=<0 0 0>, rot=<0 0 0 1>}
 Hat_MESH {type=Curve, loc=<0 0 0>, rot=<0 0 0 1>}
 Null_MESH {type=Curve, loc=<0 0 0>, rot=<0 0 0 1>}
 Pose_0 {type=Curve, loc=<0 0 0>, rot=<0 0 0 1>}
 Hair_Mask {type=Curve, loc=<0 0 0>, rot=<0 0 0 1>}
 eye_convergence {type=Curve, loc=<0 0 0>, rot=<0 0 0 1>}
 headTransform {type=Curve, loc=<0 0 0>, rot=<0 0 0 1>}
 Pose_1 {type=Curve, loc=<0 0 0>, rot=<0 0 0 1>}
 Pose_2 {type=Curve, loc=<0 0 0>, rot=<0 0 0 1>}
 Pose_3 {type=Curve, loc=<0 0 0>, rot=<0 0 0 1>}
 Pose_4 {type=Curve, loc=<0 0 0>, rot=<0 0 0 1>}
 Pose_5 {type=Curve, loc=<0 0 0>, rot=<0 0 0 1>}
 Pose_6 {type=Curve, loc=<0 0 0>, rot=<0 0 0 1>}
 Pose_7 {type=Curve, loc=<0 0 0>, rot=<0 0 0 1>}
 Pose_8 {type=Curve, loc=<0 0 0>, rot=<0 0 0 1>}
 Pose_9 {type=Curve, loc=<0 0 0>, rot=<0 0 0 1>}
 Pose_10 {type=Curve, loc=<0 0 0>, rot=<0 0 0 1>}
 Pose_11 {type=Curve, loc=<0 0 0>, rot=<0 0 0 1>}
 Pose_12 {type=Curve, loc=<0 0 0>, rot=<0 0 0 1>}
 Pose_13 {type=Curve, loc=<0 0 0>, rot=<0 0 0 1>}
 Pose_14 {type=Curve, loc=<0 0 0>, rot=<0 0 0 1>}
 Pose_15 {type=Curve, loc=<0 0 0>, rot=<0 0 0 1>}
 Pose_16 {type=Curve, loc=<0 0 0>, rot=<0 0 0 1>}
 Pose_17 {type=Curve, loc=<0 0 0>, rot=<0 0 0 1>}
 Pose_18 {type=Curve, loc=<0 0 0>, rot=<0 0 0 1>}
 Pose_19 {type=Curve, loc=<0 0 0>, rot=<0 0 0 1>}
 Pose_20 {type=Curve, loc=<0 0 0>, rot=<0 0 0 1>}
 Pose_21 {type=Curve, loc=<0 0 0>, rot=<0 0 0 1>}
 Pose_22 {type=Curve, loc=<0 0 0>, rot=<0 0 0 1>}
 Pose_23 {type=Curve, loc=<0 0 0>, rot=<0 0 0 1>}
 Pose_24 {type=Curve, loc=<0 0 0>, rot=<0 0 0 1>}
 Pose_25 {type=Curve, loc=<0 0 0>, rot=<0 0 0 1>}
 Pose_26 {type=Curve, loc=<0 0 0>, rot=<0 0 0 1>}
 Pose_27 {type=Curve, loc=<0 0 0>, rot=<0 0 0 1>}
 Pose_28 {type=Curve, loc=<0 0 0>, rot=<0 0 0 1>}
 Pose_29 {type=Curve, loc=<0 0 0>, rot=<0 0 0 1>}
 Pose_30 {type=Curve, loc=<0 0 0>, rot=<0 0 0 1>}
 Pose_31 {type=Curve, loc=<0 0 0>, rot=<0 0 0 1>}
 Pose_32 {type=Curve, loc=<0 0 0>, rot=<0 0 0 1>}
 Pose_33 {type=Curve, loc=<0 0 0>, rot=<0 0 0 1>}
 Pose_34 {type=Curve, loc=<0 0 0>, rot=<0 0 0 1>}
 Pose_35 {type=Curve, loc=<0 0 0>, rot=<0 0 0 1>}
 Pose_36 {type=Curve, loc=<0 0 0>, rot=<0 0 0 1>}
 Pose_37 {type=Curve, loc=<0 0 0>, rot=<0 0 0 1>}
 Pose_38 {type=Curve, loc=<0 0 0>, rot=<0 0 0 1>}
 Pose_39 {type=Curve, loc=<0 0 0>, rot=<0 0 0 1>}
 Pose_40 {type=Curve, loc=<0 0 0>, rot=<0 0 0 1>}
 Pose_41 {type=Curve, loc=<0 0 0>, rot=<0 0 0 1>}
 Pose_42 {type=Curve, loc=<0 0 0>, rot=<0 0 0 1>}
 Pose_43 {type=Curve, loc=<0 0 0>, rot=<0 0 0 1>}
 Pose_44 {type=Curve, loc=<0 0 0>, rot=<0 0 0 1>}
 Pose_45 {type=Curve, loc=<0 0 0>, rot=<0 0 0 1>}
 Pose_46 {type=Curve, loc=<0 0 0>, rot=<0 0 0 1>}
 Pose_47 {type=Curve, loc=<0 0 0>, rot=<0 0 0 1>}
 Pose_48 {type=Curve, loc=<0 0 0>, rot=<0 0 0 1>}
 Pose_49 {type=Curve, loc=<0 0 0>, rot=<0 0 0 1>}
 Pose_50 {type=Curve, loc=<0 0 0>, rot=<0 0 0 1>}
 Pose_51 {type=Curve, loc=<0 0 0>, rot=<0 0 0 1>}
 Pose_52 {type=Curve, loc=<0 0 0>, rot=<0 0 0 1>}
 Pose_53 {type=Curve, loc=<0 0 0>, rot=<0 0 0 1>}
 Pose_54 {type=Curve, loc=<0 0 0>, rot=<0 0 0 1>}
 Pose_55 {type=Curve, loc=<0 0 0>, rot=<0 0 0 1>}
 Pose_56 {type=Curve, loc=<0 0 0>, rot=<0 0 0 1>}
 Pose_57 {type=Curve, loc=<0 0 0>, rot=<0 0 0 1>}
 Pose_58 {type=Curve, loc=<0 0 0>, rot=<0 0 0 1>}
 Pose_59 {type=Curve, loc=<0 0 0>, rot=<0 0 0 1>}
 Pose_60 {type=Curve, loc=<0 0 0>, rot=<0 0 0 1>}
 Pose_61 {type=Curve, loc=<0 0 0>, rot=<0 0 0 1>}
 Pose_62 {type=Curve, loc=<0 0 0>, rot=<0 0 0 1>}
 Pose_63 {type=Curve, loc=<0 0 0>, rot=<0 0 0 1>}
 Pose_64 {type=Curve, loc=<0 0 0>, rot=<0 0 0 1>}
 Pose_65 {type=Curve, loc=<0 0 0>, rot=<0 0 0 1>}
 Pose_66 {type=Curve, loc=<0 0 0>, rot=<0 0 0 1>}
 Pose_67 {type=Curve, loc=<0 0 0>, rot=<0 0 0 1>}
 Pose_68 {type=Curve, loc=<0 0 0>, rot=<0 0 0 1>}
 Pose_69 {type=Curve, loc=<0 0 0>, rot=<0 0 0 1>}
 Pose_70 {type=Curve, loc=<0 0 0>, rot=<0 0 0 1>}
 Pose_71 {type=Curve, loc=<0 0 0>, rot=<0 0 0 1>}
 Pose_72 {type=Curve, loc=<0 0 0>, rot=<0 0 0 1>}
 Pose_73 {type=Curve, loc=<0 0 0>, rot=<0 0 0 1>}
 Pose_74 {type=Curve, loc=<0 0 0>, rot=<0 0 0 1>}
 Pose_75 {type=Curve, loc=<0 0 0>, rot=<0 0 0 1>}
 Pose_76 {type=Curve, loc=<0 0 0>, rot=<0 0 0 1>}
 Pose_77 {type=Curve, loc=<0 0 0>, rot=<0 0 0 1>}
 Pose_78 {type=Curve, loc=<0 0 0>, rot=<0 0 0 1>}
 Pose_79 {type=Curve, loc=<0 0 0>, rot=<0 0 0 1>}
 Pose_80 {type=Curve, loc=<0 0 0>, rot=<0 0 0 1>}
 Pose_81 {type=Curve, loc=<0 0 0>, rot=<0 0 0 1>}
 Pose_82 {type=Curve, loc=<0 0 0>, rot=<0 0 0 1>}
 Pose_83 {type=Curve, loc=<0 0 0>, rot=<0 0 0 1>}
 Pose_84 {type=Curve, loc=<0 0 0>, rot=<0 0 0 1>}
 Pose_85 {type=Curve, loc=<0 0 0>, rot=<0 0 0 1>}
 Pose_86 {type=Curve, loc=<0 0 0>, rot=<0 0 0 1>}
 Pose_87 {type=Curve, loc=<0 0 0>, rot=<0 0 0 1>}
 Pose_88 {type=Curve, loc=<0 0 0>, rot=<0 0 0 1>}
 Pose_89 {type=Curve, loc=<0 0 0>, rot=<0 0 0 1>}
 Pose_90 {type=Curve, loc=<0 0 0>, rot=<0 0 0 1>}
 Pose_91 {type=Curve, loc=<0 0 0>, rot=<0 0 0 1>}
 Pose_92 {type=Curve, loc=<0 0 0>, rot=<0 0 0 1>}
 Pose_93 {type=Curve, loc=<0 0 0>, rot=<0 0 0 1>}
 Pose_94 {type=Curve, loc=<0 0 0>, rot=<0 0 0 1>}
 Pose_95 {type=Curve, loc=<0 0 0>, rot=<0 0 0 1>}
 Pose_96 {type=Curve, loc=<0 0 0>, rot=<0 0 0 1>}
 Pose_97 {type=Curve, loc=<0 0 0>, rot=<0 0 0 1>}
 Pose_98 {type=Curve, loc=<0 0 0>, rot=<0 0 0 1>}
 Pose_99 {type=Curve, loc=<0 0 0>, rot=<0 0 0 1>}
 Pose_100 {type=Curve, loc=<0 0 0>, rot=<0 0 0 1>}
 Pose_101 {type=Curve, loc=<0 0 0>, rot=<0 0 0 1>}
 Pose_102 {type=Curve, loc=<0 0 0>, rot=<0 0 0 1>}
 Pose_103 {type=Curve, loc=<0 0 0>, rot=<0 0 0 1>}
 Pose_104 {type=Curve, loc=<0 0 0>, rot=<0 0 0 1>}
 Pose_105 {type=Curve, loc=<0 0 0>, rot=<0 0 0 1>}
 Pose_106 {type=Curve, loc=<0 0 0>, rot=<0 0 0 1>}
 Pose_107 {type=Curve, loc=<0 0 0>, rot=<0 0 0 1>}
 Pose_108 {type=Curve, loc=<0 0 0>, rot=<0 0 0 1>}
 Pose_109 {type=Curve, loc=<0 0 0>, rot=<0 0 0 1>}
 Pose_110 {type=Curve, loc=<0 0 0>, rot=<0 0 0 1>}
 Pose_111 {type=Curve, loc=<0 0 0>, rot=<0 0 0 1>}
 Pose_112 {type=Curve, loc=<0 0 0>, rot=<0 0 0 1>}
 Pose_113 {type=Curve, loc=<0 0 0>, rot=<0 0 0 1>}
 Pose_114 {type=Curve, loc=<0 0 0>, rot=<0 0 0 1>}
 Pose_115 {type=Curve, loc=<0 0 0>, rot=<0 0 0 1>}
 Pose_116 {type=Curve, loc=<0 0 0>, rot=<0 0 0 1>}
 Pose_117 {type=Curve, loc=<0 0 0>, rot=<0 0 0 1>}
 Pose_118 {type=Curve, loc=<0 0 0>, rot=<0 0 0 1>}
 Pose_119 {type=Curve, loc=<0 0 0>, rot=<0 0 0 1>}
 Pose_120 {type=Curve, loc=<0 0 0>, rot=<0 0 0 1>}
 Pose_121 {type=Curve, loc=<0 0 0>, rot=<0 0 0 1>}
 Pose_122 {type=Curve, loc=<0 0 0>, rot=<0 0 0 1>}
 Pose_123 {type=Curve, loc=<0 0 0>, rot=<0 0 0 1>}
 Pose_124 {type=Curve, loc=<0 0 0>, rot=<0 0 0 1>}
 Pose_125 {type=Curve, loc=<0 0 0>, rot=<0 0 0 1>}
 Pose_126 {type=Curve, loc=<0 0 0>, rot=<0 0 0 1>}
 Pose_127 {type=Curve, loc=<0 0 0>, rot=<0 0 0 1>}
 Pose_128 {type=Curve, loc=<0 0 0>, rot=<0 0 0 1>}
 Pose_129 {type=Curve, loc=<0 0 0>, rot=<0 0 0 1>}
 Pose_130 {type=Curve, loc=<0 0 0>, rot=<0 0 0 1>}
 Pose_131 {type=Curve, loc=<0 0 0>, rot=<0 0 0 1>}
 Pose_132 {type=Curve, loc=<0 0 0>, rot=<0 0 0 1>}
 Pose_133 {type=Curve, loc=<0 0 0>, rot=<0 0 0 1>}
 Pose_134 {type=Curve, loc=<0 0 0>, rot=<0 0 0 1>}
 Pose_135 {type=Curve, loc=<0 0 0>, rot=<0 0 0 1>}
 Pose_136 {type=Curve, loc=<0 0 0>, rot=<0 0 0 1>}
 Pose_137 {type=Curve, loc=<0 0 0>, rot=<0 0 0 1>}
 Pose_138 {type=Curve, loc=<0 0 0>, rot=<0 0 0 1>}
 Pose_139 {type=Curve, loc=<0 0 0>, rot=<0 0 0 1>}
 Pose_140 {type=Curve, loc=<0 0 0>, rot=<0 0 0 1>}
 Pose_141 {type=Curve, loc=<0 0 0>, rot=<0 0 0 1>}
 Pose_142 {type=Curve, loc=<0 0 0>, rot=<0 0 0 1>}
 Pose_143 {type=Curve, loc=<0 0 0>, rot=<0 0 0 1>}
 Pose_144 {type=Curve, loc=<0 0 0>, rot=<0 0 0 1>}
 Pose_145 {type=Curve, loc=<0 0 0>, rot=<0 0 0 1>}
 Pose_146 {type=Curve, loc=<0 0 0>, rot=<0 0 0 1>}
 Pose_147 {type=Curve, loc=<0 0 0>, rot=<0 0 0 1>}
 Pose_148 {type=Curve, loc=<0 0 0>, rot=<0 0 0 1>}
 Pose_149 {type=Curve, loc=<0 0 0>, rot=<0 0 0 1>}
 Pose_150 {type=Curve, loc=<0 0 0>, rot=<0 0 0 1>}
 Pose_151 {type=Curve, loc=<0 0 0>, rot=<0 0 0 1>}
 Pose_152 {type=Curve, loc=<0 0 0>, rot=<0 0 0 1>}
 Pose_153 {type=Curve, loc=<0 0 0>, rot=<0 0 0 1>}
 Pose_154 {type=Curve, loc=<0 0 0>, rot=<0 0 0 1>}
 Pose_155 {type=Curve, loc=<0 0 0>, rot=<0 0 0 1>}
 Pose_156 {type=Curve, loc=<0 0 0>, rot=<0 0 0 1>}
 Pose_157 {type=Curve, loc=<0 0 0>, rot=<0 0 0 1>}
 Pose_158 {type=Curve, loc=<0 0 0>, rot=<0 0 0 1>}
 Pose_159 {type=Curve, loc=<0 0 0>, rot=<0 0 0 1>}
 Pose_160 {type=Curve, loc=<0 0 0>, rot=<0 0 0 1>}
 Pose_161 {type=Curve, loc=<0 0 0>, rot=<0 0 0 1>}
 Pose_162 {type=Curve, loc=<0 0 0>, rot=<0 0 0 1>}
 Pose_163 {type=Curve, loc=<0 0 0>, rot=<0 0 0 1>}
 Pose_164 {type=Curve, loc=<0 0 0>, rot=<0 0 0 1>}
 Pose_165 {type=Curve, loc=<0 0 0>, rot=<0 0 0 1>}
 Pose_166 {type=Curve, loc=<0 0 0>, rot=<0 0 0 1>}
 Pose_167 {type=Curve, loc=<0 0 0>, rot=<0 0 0 1>}
 Pose_168 {type=Curve, loc=<0 0 0>, rot=<0 0 0 1>}
 Pose_169 {type=Curve, loc=<0 0 0>, rot=<0 0 0 1>}
 Pose_170 {type=Curve, loc=<0 0 0>, rot=<0 0 0 1>}
 Pose_171 {type=Curve, loc=<0 0 0>, rot=<0 0 0 1>}
 Pose_172 {type=Curve, loc=<0 0 0>, rot=<0 0 0 1>}
 Pose_173 {type=Curve, loc=<0 0 0>, rot=<0 0 0 1>}
 Pose_174 {type=Curve, loc=<0 0 0>, rot=<0 0 0 1>}
 Pose_175 {type=Curve, loc=<0 0 0>, rot=<0 0 0 1>}
 Pose_176 {type=Curve, loc=<0 0 0>, rot=<0 0 0 1>}
 Pose_177 {type=Curve, loc=<0 0 0>, rot=<0 0 0 1>}
 Pose_178 {type=Curve, loc=<0 0 0>, rot=<0 0 0 1>}
 Pose_179 {type=Curve, loc=<0 0 0>, rot=<0 0 0 1>}
 eyeLookDownRight {type=Curve, loc=<0 0 0>, rot=<0 0 0 1>}
 eyeLookOutRight {type=Curve, loc=<0 0 0>, rot=<0 0 0 1>}
 eyeLookUpRight {type=Curve, loc=<0 0 0>, rot=<0 0 0 1>}
 eyeLookOutLeft {type=Curve, loc=<0 0 0>, rot=<0 0 0 1>}
 eyeBlinkLeft {type=Curve, loc=<0 0 0>, rot=<0 0 0 1>}
 eyeBlinkRight {type=Curve, loc=<0 0 0>, rot=<0 0 0 1>}
 eyeSquintLeft {type=Curve, loc=<0 0 0>, rot=<0 0 0 1>}
 eyeSquintRight {type=Curve, loc=<0 0 0>, rot=<0 0 0 1>}
 eyeWideLeft {type=Curve, loc=<0 0 0>, rot=<0 0 0 1>}
 eyeWideRight {type=Curve, loc=<0 0 0>, rot=<0 0 0 1>}
 jawForward {type=Curve, loc=<0 0 0>, rot=<0 0 0 1>}
 jawLeft {type=Curve, loc=<0 0 0>, rot=<0 0 0 1>}
 jawRight {type=Curve, loc=<0 0 0>, rot=<0 0 0 1>}
 jawOpen {type=Curve, loc=<0 0 0>, rot=<0 0 0 1>}
 mouthClose {type=Curve, loc=<0 0 0>, rot=<0 0 0 1>}
 mouthFunnel {type=Curve, loc=<0 0 0>, rot=<0 0 0 1>}
 mouthPucker {type=Curve, loc=<0 0 0>, rot=<0 0 0 1>}
 mouthLeft {type=Curve, loc=<0 0 0>, rot=<0 0 0 1>}
 mouthRight {type=Curve, loc=<0 0 0>, rot=<0 0 0 1>}
 mouthSmileLeft {type=Curve, loc=<0 0 0>, rot=<0 0 0 1>}
 mouthSmileRight {type=Curve, loc=<0 0 0>, rot=<0 0 0 1>}
 mouthFrownLeft {type=Curve, loc=<0 0 0>, rot=<0 0 0 1>}
 mouthFrownRight {type=Curve, loc=<0 0 0>, rot=<0 0 0 1>}
 mouthDimpleLeft {type=Curve, loc=<0 0 0>, rot=<0 0 0 1>}
 mouthDimpleRight {type=Curve, loc=<0 0 0>, rot=<0 0 0 1>}
 mouthStretchLeft {type=Curve, loc=<0 0 0>, rot=<0 0 0 1>}
 mouthStretchRight {type=Curve, loc=<0 0 0>, rot=<0 0 0 1>}
 mouthRollLower {type=Curve, loc=<0 0 0>, rot=<0 0 0 1>}
 mouthRollUpper {type=Curve, loc=<0 0 0>, rot=<0 0 0 1>}
 mouthShrugLower {type=Curve, loc=<0 0 0>, rot=<0 0 0 1>}
 mouthShrugUpper {type=Curve, loc=<0 0 0>, rot=<0 0 0 1>}
 mouthPressLeft {type=Curve, loc=<0 0 0>, rot=<0 0 0 1>}
 mouthPressRight {type=Curve, loc=<0 0 0>, rot=<0 0 0 1>}
 mouthLowerDownLeft {type=Curve, loc=<0 0 0>, rot=<0 0 0 1>}
 mouthLowerDownRight {type=Curve, loc=<0 0 0>, rot=<0 0 0 1>}
 mouthUpperUpLeft {type=Curve, loc=<0 0 0>, rot=<0 0 0 1>}
 mouthUpperUpRight {type=Curve, loc=<0 0 0>, rot=<0 0 0 1>}
 browDownLeft {type=Curve, loc=<0 0 0>, rot=<0 0 0 1>}
 browDownRight {type=Curve, loc=<0 0 0>, rot=<0 0 0 1>}
 browInnerUp {type=Curve, loc=<0 0 0>, rot=<0 0 0 1>}
 browOuterUpLeft {type=Curve, loc=<0 0 0>, rot=<0 0 0 1>}
 browOuterUpRight {type=Curve, loc=<0 0 0>, rot=<0 0 0 1>}
 cheekPuff {type=Curve, loc=<0 0 0>, rot=<0 0 0 1>}
 cheekSquintLeft {type=Curve, loc=<0 0 0>, rot=<0 0 0 1>}
 cheekSquintRight {type=Curve, loc=<0 0 0>, rot=<0 0 0 1>}
 noseSneerLeft {type=Curve, loc=<0 0 0>, rot=<0 0 0 1>}
 noseSneerRight {type=Curve, loc=<0 0 0>, rot=<0 0 0 1>}
 C_glabella_down {type=Curve, loc=<0 0 0>, rot=<0 0 0 1>}
 L_brow_squeeze_in {type=Curve, loc=<0 0 0>, rot=<0 0 0 1>}
 R_brow_squeeze_in {type=Curve, loc=<0 0 0>, rot=<0 0 0 1>}
 L_upper_lip_lower {type=Curve, loc=<0 0 0>, rot=<0 0 0 1>}
 R_upper_lip_lower {type=Curve, loc=<0 0 0>, rot=<0 0 0 1>}
 L_lower_lip_up {type=Curve, loc=<0 0 0>, rot=<0 0 0 1>}
 R_lower_lip_up {type=Curve, loc=<0 0 0>, rot=<0 0 0 1>}
 L_lip_corner_narrow {type=Curve, loc=<0 0 0>, rot=<0 0 0 1>}
 R_lip_corner_narrow {type=Curve, loc=<0 0 0>, rot=<0 0 0 1>}
 phoneme_ch {type=Curve, loc=<0 0 0>, rot=<0 0 0 1>}
 phoneme_mbp {type=Curve, loc=<0 0 0>, rot=<0 0 0 1>}
 phoneme_fv {type=Curve, loc=<0 0 0>, rot=<0 0 0 1>}
 mouth_up {type=Curve, loc=<0 0 0>, rot=<0 0 0 1>}
 mouth_down {type=Curve, loc=<0 0 0>, rot=<0 0 0 1>}
 L_cheek_up {type=Curve, loc=<0 0 0>, rot=<0 0 0 1>}
 R_cheek_up {type=Curve, loc=<0 0 0>, rot=<0 0 0 1>}
 L_upperLidDown {type=Curve, loc=<0 0 0>, rot=<0 0 0 1>}
 R_upperLidDown {type=Curve, loc=<0 0 0>, rot=<0 0 0 1>}
 jaw_up {type=Curve, loc=<0 0 0>, rot=<0 0 0 1>}
 jaw_back {type=Curve, loc=<0 0 0>, rot=<0 0 0 1>}
 tongue_forward {type=Curve, loc=<0 0 0>, rot=<0 0 0 1>}
 tongue_back {type=Curve, loc=<0 0 0>, rot=<0 0 0 1>}
 tongue_up {type=Curve, loc=<0 0 0>, rot=<0 0 0 1>}
 tongue_down {type=Curve, loc=<0 0 0>, rot=<0 0 0 1>}
 basePose {type=Curve, loc=<0 0 0>, rot=<0 0 0 1>}
 Pose_180 {type=Curve, loc=<0 0 0>, rot=<0 0 0 1>}
 Pose_181 {type=Curve, loc=<0 0 0>, rot=<0 0 0 1>}
 Pose_182 {type=Curve, loc=<0 0 0>, rot=<0 0 0 1>}
 Pose_183 {type=Curve, loc=<0 0 0>, rot=<0 0 0 1>}
 Pose_184 {type=Curve, loc=<0 0 0>, rot=<0 0 0 1>}
 Pose_185 {type=Curve, loc=<0 0 0>, rot=<0 0 0 1>}
 Pose_186 {type=Curve, loc=<0 0 0>, rot=<0 0 0 1>}
 Pose_187 {type=Curve, loc=<0 0 0>, rot=<0 0 0 1>}
 Pose_188 {type=Curve, loc=<0 0 0>, rot=<0 0 0 1>}
 Pose_189 {type=Curve, loc=<0 0 0>, rot=<0 0 0 1>}
 Pose_190 {type=Curve, loc=<0 0 0>, rot=<0 0 0 1>}
 Pose_191 {type=Curve, loc=<0 0 0>, rot=<0 0 0 1>}
 Pose_192 {type=Curve, loc=<0 0 0>, rot=<0 0 0 1>}
 Pose_193 {type=Curve, loc=<0 0 0>, rot=<0 0 0 1>}
 Pose_194 {type=Curve, loc=<0 0 0>, rot=<0 0 0 1>}
 Pose_195 {type=Curve, loc=<0 0 0>, rot=<0 0 0 1>}
 Pose_196 {type=Curve, loc=<0 0 0>, rot=<0 0 0 1>}
 Pose_197 {type=Curve, loc=<0 0 0>, rot=<0 0 0 1>}
 Pose_198 {type=Curve, loc=<0 0 0>, rot=<0 0 0 1>}
 Pose_199 {type=Curve, loc=<0 0 0>, rot=<0 0 0 1>}
 Pose_200 {type=Curve, loc=<0 0 0>, rot=<0 0 0 1>}
 Pose_201 {type=Curve, loc=<0 0 0>, rot=<0 0 0 1>}
 Pose_202 {type=Curve, loc=<0 0 0>, rot=<0 0 0 1>}
 Pose_203 {type=Curve, loc=<0 0 0>, rot=<0 0 0 1>}
 Pose_204 {type=Curve, loc=<0 0 0>, rot=<0 0 0 1>}
 Pose_205 {type=Curve, loc=<0 0 0>, rot=<0 0 0 1>}
 Pose_206 {type=Curve, loc=<0 0 0>, rot=<0 0 0 1>}
 Pose_207 {type=Curve, loc=<0 0 0>, rot=<0 0 0 1>}
 Pose_208 {type=Curve, loc=<0 0 0>, rot=<0 0 0 1>}
 Pose_209 {type=Curve, loc=<0 0 0>, rot=<0 0 0 1>}
 Pose_210 {type=Curve, loc=<0 0 0>, rot=<0 0 0 1>}
 Pose_211 {type=Curve, loc=<0 0 0>, rot=<0 0 0 1>}
 Pose_212 {type=Curve, loc=<0 0 0>, rot=<0 0 0 1>}
 Pose_213 {type=Curve, loc=<0 0 0>, rot=<0 0 0 1>}
 Pose_214 {type=Curve, loc=<0 0 0>, rot=<0 0 0 1>}
 Pose_215 {type=Curve, loc=<0 0 0>, rot=<0 0 0 1>}
 Pose_216 {type=Curve, loc=<0 0 0>, rot=<0 0 0 1>}
 Pose_217 {type=Curve, loc=<0 0 0>, rot=<0 0 0 1>}
 Pose_218 {type=Curve, loc=<0 0 0>, rot=<0 0 0 1>}
 Pose_219 {type=Curve, loc=<0 0 0>, rot=<0 0 0 1>}
 Pose_220 {type=Curve, loc=<0 0 0>, rot=<0 0 0 1>}
 Pose_221 {type=Curve, loc=<0 0 0>, rot=<0 0 0 1>}
 Pose_222 {type=Curve, loc=<0 0 0>, rot=<0 0 0 1>}
 Pose_223 {type=Curve, loc=<0 0 0>, rot=<0 0 0 1>}
 Pose_224 {type=Curve, loc=<0 0 0>, rot=<0 0 0 1>}
 Pose_225 {type=Curve, loc=<0 0 0>, rot=<0 0 0 1>}
 Pose_226 {type=Curve, loc=<0 0 0>, rot=<0 0 0 1>}
 Pose_227 {type=Curve, loc=<0 0 0>, rot=<0 0 0 1>}
 Pose_228 {type=Curve, loc=<0 0 0>, rot=<0 0 0 1>}
 Pose_229 {type=Curve, loc=<0 0 0>, rot=<0 0 0 1>}
 Pose_230 {type=Curve, loc=<0 0 0>, rot=<0 0 0 1>}
 Pose_231 {type=Curve, loc=<0 0 0>, rot=<0 0 0 1>}
 Pose_232 {type=Curve, loc=<0 0 0>, rot=<0 0 0 1>}
 Pose_233 {type=Curve, loc=<0 0 0>, rot=<0 0 0 1>}
 Pose_234 {type=Curve, loc=<0 0 0>, rot=<0 0 0 1>}
 Pose_235 {type=Curve, loc=<0 0 0>, rot=<0 0 0 1>}
 Pose_236 {type=Curve, loc=<0 0 0>, rot=<0 0 0 1>}
 Pose_237 {type=Curve, loc=<0 0 0>, rot=<0 0 0 1>}
 Pose_238 {type=Curve, loc=<0 0 0>, rot=<0 0 0 1>}
 Pose_239 {type=Curve, loc=<0 0 0>, rot=<0 0 0 1>}
 Pose_240 {type=Curve, loc=<0 0 0>, rot=<0 0 0 1>}
 Pose_241 {type=Curve, loc=<0 0 0>, rot=<0 0 0 1>}
 Pose_242 {type=Curve, loc=<0 0 0>, rot=<0 0 0 1>}
 Pose_243 {type=Curve, loc=<0 0 0>, rot=<0 0 0 1>}
 Pose_244 {type=Curve, loc=<0 0 0>, rot=<0 0 0 1>}
 Pose_245 {type=Curve, loc=<0 0 0>, rot=<0 0 0 1>}
 Pose_246 {type=Curve, loc=<0 0 0>, rot=<0 0 0 1>}
 Pose_247 {type=Curve, loc=<0 0 0>, rot=<0 0 0 1>}
 Pose_248 {type=Curve, loc=<0 0 0>, rot=<0 0 0 1>}
 Pose_249 {type=Curve, loc=<0 0 0>, rot=<0 0 0 1>}
 Pose_250 {type=Curve, loc=<0 0 0>, rot=<0 0 0 1>}
 Pose_251 {type=Curve, loc=<0 0 0>, rot=<0 0 0 1>}
 Pose_252 {type=Curve, loc=<0 0 0>, rot=<0 0 0 1>}
 Pose_253 {type=Curve, loc=<0 0 0>, rot=<0 0 0 1>}
 Pose_254 {type=Curve, loc=<0 0 0>, rot=<0 0 0 1>}
 Pose_255 {type=Curve, loc=<0 0 0>, rot=<0 0 0 1>}
 Pose_256 {type=Curve, loc=<0 0 0>, rot=<0 0 0 1>}
 Pose_257 {type=Curve, loc=<0 0 0>, rot=<0 0 0 1>}
 Pose_258 {type=Curve, loc=<0 0 0>, rot=<0 0 0 1>}
 neutral {type=Curve, loc=<0 0 0>, rot=<0 0 0 1>}
 Pose_259 {type=Curve, loc=<0 0 0>, rot=<0 0 0 1>}
 cheek_puff_low {type=Curve, loc=<0 0 0>, rot=<0 0 0 1>}
 Pose_260 {type=Curve, loc=<0 0 0>, rot=<0 0 0 1>}
 Pose_261 {type=Curve, loc=<0 0 0>, rot=<0 0 0 1>}
 Pose_262 {type=Curve, loc=<0 0 0>, rot=<0 0 0 1>}
 Pose_263 {type=Curve, loc=<0 0 0>, rot=<0 0 0 1>}
 Pose_264 {type=Curve, loc=<0 0 0>, rot=<0 0 0 1>}
 Pose_265 {type=Curve, loc=<0 0 0>, rot=<0 0 0 1>}
 Pose_266 {type=Curve, loc=<0 0 0>, rot=<0 0 0 1>}
 Pose_267 {type=Curve, loc=<0 0 0>, rot=<0 0 0 1>}
 Pose_268 {type=Curve, loc=<0 0 0>, rot=<0 0 0 1>}
 Pose_269 {type=Curve, loc=<0 0 0>, rot=<0 0 0 1>}
 Pose_270 {type=Curve, loc=<0 0 0>, rot=<0 0 0 1>}
 Pose_271 {type=Curve, loc=<0 0 0>, rot=<0 0 0 1>}
 Pose_272 {type=Curve, loc=<0 0 0>, rot=<0 0 0 1>}
 Pose_273 {type=Curve, loc=<0 0 0>, rot=<0 0 0 1>}
 Pose_274 {type=Curve, loc=<0 0 0>, rot=<0 0 0 1>}
 Pose_275 {type=Curve, loc=<0 0 0>, rot=<0 0 0 1>}
 Pose_276 {type=Curve, loc=<0 0 0>, rot=<0 0 0 1>}
 Pose_277 {type=Curve, loc=<0 0 0>, rot=<0 0 0 1>}
 Pose_278 {type=Curve, loc=<0 0 0>, rot=<0 0 0 1>}
 Pose_279 {type=Curve, loc=<0 0 0>, rot=<0 0 0 1>}
 Pose_280 {type=Curve, loc=<0 0 0>, rot=<0 0 0 1>}
 Pose_281 {type=Curve, loc=<0 0 0>, rot=<0 0 0 1>}
 Pose_282 {type=Curve, loc=<0 0 0>, rot=<0 0 0 1>}
 Pose_283 {type=Curve, loc=<0 0 0>, rot=<0 0 0 1>}
 Pose_284 {type=Curve, loc=<0 0 0>, rot=<0 0 0 1>}
 Pose_285 {type=Curve, loc=<0 0 0>, rot=<0 0 0 1>}
 Pose_286 {type=Curve, loc=<0 0 0>, rot=<0 0 0 1>}
 Pose_287 {type=Curve, loc=<0 0 0>, rot=<0 0 0 1>}
 Pose_288 {type=Curve, loc=<0 0 0>, rot=<0 0 0 1>}
 Pose_289 {type=Curve, loc=<0 0 0>, rot=<0 0 0 1>}
 Pose_290 {type=Curve, loc=<0 0 0>, rot=<0 0 0 1>}
 Pose_291 {type=Curve, loc=<0 0 0>, rot=<0 0 0 1>}
 Pose_292 {type=Curve, loc=<0 0 0>, rot=<0 0 0 1>}
 Pose_293 {type=Curve, loc=<0 0 0>, rot=<0 0 0 1>}
 Pose_294 {type=Curve, loc=<0 0 0>, rot=<0 0 0 1>}
 Pose_295 {type=Curve, loc=<0 0 0>, rot=<0 0 0 1>}
 Pose_296 {type=Curve, loc=<0 0 0>, rot=<0 0 0 1>}
 Pose_297 {type=Curve, loc=<0 0 0>, rot=<0 0 0 1>}
 Pose_298 {type=Curve, loc=<0 0 0>, rot=<0 0 0 1>}
 Pose_299 {type=Curve, loc=<0 0 0>, rot=<0 0 0 1>}
 Pose_300 {type=Curve, loc=<0 0 0>, rot=<0 0 0 1>}
 Pose_301 {type=Curve, loc=<0 0 0>, rot=<0 0 0 1>}
 Pose_302 {type=Curve, loc=<0 0 0>, rot=<0 0 0 1>}
 Pose_303 {type=Curve, loc=<0 0 0>, rot=<0 0 0 1>}
 Pose_304 {type=Curve, loc=<0 0 0>, rot=<0 0 0 1>}
 Pose_305 {type=Curve, loc=<0 0 0>, rot=<0 0 0 1>}
 Pose_306 {type=Curve, loc=<0 0 0>, rot=<0 0 0 1>}
 Pose_307 {type=Curve, loc=<0 0 0>, rot=<0 0 0 1>}
 Pose_308 {type=Curve, loc=<0 0 0>, rot=<0 0 0 1>}
 Pose_309 {type=Curve, loc=<0 0 0>, rot=<0 0 0 1>}
 Pose_310 {type=Curve, loc=<0 0 0>, rot=<0 0 0 1>}
 Pose_311 {type=Curve, loc=<0 0 0>, rot=<0 0 0 1>}
 Pose_312 {type=Curve, loc=<0 0 0>, rot=<0 0 0 1>}
 Pose_313 {type=Curve, loc=<0 0 0>, rot=<0 0 0 1>}
 Pose_314 {type=Curve, loc=<0 0 0>, rot=<0 0 0 1>}
 Pose_315 {type=Curve, loc=<0 0 0>, rot=<0 0 0 1>}
 Pose_316 {type=Curve, loc=<0 0 0>, rot=<0 0 0 1>}
 Pose_317 {type=Curve, loc=<0 0 0>, rot=<0 0 0 1>}
 Pose_318 {type=Curve, loc=<0 0 0>, rot=<0 0 0 1>}
 Pose_319 {type=Curve, loc=<0 0 0>, rot=<0 0 0 1>}
 Pose_320 {type=Curve, loc=<0 0 0>, rot=<0 0 0 1>}
 Pose_321 {type=Curve, loc=<0 0 0>, rot=<0 0 0 1>}
 Desert_Island_Jonesy_Mouth {type=Curve, loc=<0 0 0>, rot=<0 0 0 1>}
 Pose_322 {type=Curve, loc=<0 0 0>, rot=<0 0 0 1>}
 HT_L1_TransitionCurve {type=Curve, loc=<0 0 0>, rot=<0 0 0 1>}
 Hair_CapShape {type=Curve, loc=<0 0 0>, rot=<0 0 0 1>}
 Hair_CapShape1 {type=Curve, loc=<0 0 0>, rot=<0 0 0 1>}
 jekyll {type=Curve, loc=<0 0 0>, rot=<0 0 0 1>}
 JH_TransformCurve {type=Curve, loc=<0 0 0>, rot=<0 0 0 1>}
 eyes_LOD0 {type=Curve, loc=<0 0 0>, rot=<0 0 0 1>}
 MaterialCurveA {type=Curve, loc=<0 0 0>, rot=<0 0 0 1>}
 MaterialCurveB {type=Curve, loc=<0 0 0>, rot=<0 0 0 1>}
 MaterialCurveC {type=Curve, loc=<0 0 0>, rot=<0 0 0 1>}
 MaterialCurveD {type=Curve, loc=<0 0 0>, rot=<0 0 0 1>}
 MaterialCurveE {type=Curve, loc=<0 0 0>, rot=<0 0 0 1>}
 Custom_Frontend_Face_Idle {type=Curve, loc=<0 0 0>, rot=<0 0 0 1>}
 blendParent1 {type=Curve, loc=<0 0 0>, rot=<0 0 0 1>}
 BaseHeadDefaultFaceIdle {type=Curve, loc=<0 0 0>, rot=<0 0 0 1>}
 VeiledTransitionGlowAmount {type=Curve, loc=<0 0 0>, rot=<0 0 0 1>}
 is_3L {type=Curve, loc=<0 0 0>, rot=<0 0 0 1>}
 Hair_Hat_LOD0 {type=Curve, loc=<0 0 0>, rot=<0 0 0 1>}
 Hair_Helmet_LOD0 {type=Curve, loc=<0 0 0>, rot=<0 0 0 1>}
 Hair_Null_LOD0 {type=Curve, loc=<0 0 0>, rot=<0 0 0 1>}
 Pose_323 {type=Curve, loc=<0 0 0>, rot=<0 0 0 1>}
 Pose_324 {type=Curve, loc=<0 0 0>, rot=<0 0 0 1>}
 Pose_325 {type=Curve, loc=<0 0 0>, rot=<0 0 0 1>}
 Pose_326 {type=Curve, loc=<0 0 0>, rot=<0 0 0 1>}
 Pose_327 {type=Curve, loc=<0 0 0>, rot=<0 0 0 1>}
 Pose_328 {type=Curve, loc=<0 0 0>, rot=<0 0 0 1>}
 Pose_329 {type=Curve, loc=<0 0 0>, rot=<0 0 0 1>}
 Pose_330 {type=Curve, loc=<0 0 0>, rot=<0 0 0 1>}
 Pose_331 {type=Curve, loc=<0 0 0>, rot=<0 0 0 1>}
 Pose_332 {type=Curve, loc=<0 0 0>, rot=<0 0 0 1>}
 Pose_333 {type=Curve, loc=<0 0 0>, rot=<0 0 0 1>}
 Pose_334 {type=Curve, loc=<0 0 0>, rot=<0 0 0 1>}
 Pose_335 {type=Curve, loc=<0 0 0>, rot=<0 0 0 1>}
 Pose_336 {type=Curve, loc=<0 0 0>, rot=<0 0 0 1>}
 Pose_337 {type=Curve, loc=<0 0 0>, rot=<0 0 0 1>}
 Pose_338 {type=Curve, loc=<0 0 0>, rot=<0 0 0 1>}
 Pose_339 {type=Curve, loc=<0 0 0>, rot=<0 0 0 1>}
 Pose_340 {type=Curve, loc=<0 0 0>, rot=<0 0 0 1>}
 Pose_341 {type=Curve, loc=<0 0 0>, rot=<0 0 0 1>}
 Pose_342 {type=Curve, loc=<0 0 0>, rot=<0 0 0 1>}
 Pose_343 {type=Curve, loc=<0 0 0>, rot=<0 0 0 1>}
 Pose_344 {type=Curve, loc=<0 0 0>, rot=<0 0 0 1>}
 Pose_345 {type=Curve, loc=<0 0 0>, rot=<0 0 0 1>}
 Pose_346 {type=Curve, loc=<0 0 0>, rot=<0 0 0 1>}
 Pose_347 {type=Curve, loc=<0 0 0>, rot=<0 0 0 1>}
 Pose_348 {type=Curve, loc=<0 0 0>, rot=<0 0 0 1>}
 Pose_349 {type=Curve, loc=<0 0 0>, rot=<0 0 0 1>}
 Pose_350 {type=Curve, loc=<0 0 0>, rot=<0 0 0 1>}
 Pose_351 {type=Curve, loc=<0 0 0>, rot=<0 0 0 1>}
 Pose_352 {type=Curve, loc=<0 0 0>, rot=<0 0 0 1>}
 Pose_353 {type=Curve, loc=<0 0 0>, rot=<0 0 0 1>}
 Pose_354 {type=Curve, loc=<0 0 0>, rot=<0 0 0 1>}
 Pose_355 {type=Curve, loc=<0 0 0>, rot=<0 0 0 1>}
 Pose_356 {type=Curve, loc=<0 0 0>, rot=<0 0 0 1>}
 Pose_357 {type=Curve, loc=<0 0 0>, rot=<0 0 0 1>}
 Pose_358 {type=Curve, loc=<0 0 0>, rot=<0 0 0 1>}
 Pose_359 {type=Curve, loc=<0 0 0>, rot=<0 0 0 1>}
 Pose_360 {type=Curve, loc=<0 0 0>, rot=<0 0 0 1>}
 Pose_361 {type=Curve, loc=<0 0 0>, rot=<0 0 0 1>}
 Pose_362 {type=Curve, loc=<0 0 0>, rot=<0 0 0 1>}
 Pose_363 {type=Curve, loc=<0 0 0>, rot=<0 0 0 1>}
 Pose_364 {type=Curve, loc=<0 0 0>, rot=<0 0 0 1>}
 Pose_365 {type=Curve, loc=<0 0 0>, rot=<0 0 0 1>}
 Pose_366 {type=Curve, loc=<0 0 0>, rot=<0 0 0 1>}
 Pose_367 {type=Curve, loc=<0 0 0>, rot=<0 0 0 1>}
 Pose_368 {type=Curve, loc=<0 0 0>, rot=<0 0 0 1>}
 Pose_369 {type=Curve, loc=<0 0 0>, rot=<0 0 0 1>}
 Pose_370 {type=Curve, loc=<0 0 0>, rot=<0 0 0 1>}
 Pose_371 {type=Curve, loc=<0 0 0>, rot=<0 0 0 1>}
 Pose_372 {type=Curve, loc=<0 0 0>, rot=<0 0 0 1>}
 Pose_373 {type=Curve, loc=<0 0 0>, rot=<0 0 0 1>}
 Pose_374 {type=Curve, loc=<0 0 0>, rot=<0 0 0 1>}
 Pose_375 {type=Curve, loc=<0 0 0>, rot=<0 0 0 1>}
 Pose_376 {type=Curve, loc=<0 0 0>, rot=<0 0 0 1>}
 Pose_377 {type=Curve, loc=<0 0 0>, rot=<0 0 0 1>}
 Pose_378 {type=Curve, loc=<0 0 0>, rot=<0 0 0 1>}
 Pose_379 {type=Curve, loc=<0 0 0>, rot=<0 0 0 1>}
 Pose_380 {type=Curve, loc=<0 0 0>, rot=<0 0 0 1>}
 Pose_381 {type=Curve, loc=<0 0 0>, rot=<0 0 0 1>}
 Pose_382 {type=Curve, loc=<0 0 0>, rot=<0 0 0 1>}
 Pose_383 {type=Curve, loc=<0 0 0>, rot=<0 0 0 1>}
 Pose_384 {type=Curve, loc=<0 0 0>, rot=<0 0 0 1>}
 Pose_385 {type=Curve, loc=<0 0 0>, rot=<0 0 0 1>}
 Pose_386 {type=Curve, loc=<0 0 0>, rot=<0 0 0 1>}
 Pose_387 {type=Curve, loc=<0 0 0>, rot=<0 0 0 1>}
 Pose_388 {type=Curve, loc=<0 0 0>, rot=<0 0 0 1>}
 Pose_389 {type=Curve, loc=<0 0 0>, rot=<0 0 0 1>}
 Pose_390 {type=Curve, loc=<0 0 0>, rot=<0 0 0 1>}
 Pose_391 {type=Curve, loc=<0 0 0>, rot=<0 0 0 1>}
 Pose_392 {type=Curve, loc=<0 0 0>, rot=<0 0 0 1>}
 Pose_393 {type=Curve, loc=<0 0 0>, rot=<0 0 0 1>}
 Pose_394 {type=Curve, loc=<0 0 0>, rot=<0 0 0 1>}
 Pose_395 {type=Curve, loc=<0 0 0>, rot=<0 0 0 1>}
 Pose_396 {type=Curve, loc=<0 0 0>, rot=<0 0 0 1>}
 Pose_397 {type=Curve, loc=<0 0 0>, rot=<0 0 0 1>}
 Pose_398 {type=Curve, loc=<0 0 0>, rot=<0 0 0 1>}
 Pose_399 {type=Curve, loc=<0 0 0>, rot=<0 0 0 1>}
 Pose_400 {type=Curve, loc=<0 0 0>, rot=<0 0 0 1>}
 Pose_401 {type=Curve, loc=<0 0 0>, rot=<0 0 0 1>}
 Pose_402 {type=Curve, loc=<0 0 0>, rot=<0 0 0 1>}
 Pose_403 {type=Curve, loc=<0 0 0>, rot=<0 0 0 1>}
 Pose_404 {type=Curve, loc=<0 0 0>, rot=<0 0 0 1>}
 Pose_405 {type=Curve, loc=<0 0 0>, rot=<0 0 0 1>}
 Pose_406 {type=Curve, loc=<0 0 0>, rot=<0 0 0 1>}
 Pose_407 {type=Curve, loc=<0 0 0>, rot=<0 0 0 1>}
 Pose_408 {type=Curve, loc=<0 0 0>, rot=<0 0 0 1>}
 Pose_409 {type=Curve, loc=<0 0 0>, rot=<0 0 0 1>}
 Pose_410 {type=Curve, loc=<0 0 0>, rot=<0 0 0 1>}
 Pose_411 {type=Curve, loc=<0 0 0>, rot=<0 0 0 1>}
 Pose_412 {type=Curve, loc=<0 0 0>, rot=<0 0 0 1>}
 Pose_413 {type=Curve, loc=<0 0 0>, rot=<0 0 0 1>}
 Pose_414 {type=Curve, loc=<0 0 0>, rot=<0 0 0 1>}
 mouth_close_pose {type=Curve, loc=<0 0 0>, rot=<0 0 0 1>}
 mouth_pucker_pose {type=Curve, loc=<0 0 0>, rot=<0 0 0 1>}
 mouth_left_pose {type=Curve, loc=<0 0 0>, rot=<0 0 0 1>}
 mouth_right_pose {type=Curve, loc=<0 0 0>, rot=<0 0 0 1>}
 L_mouth_stretch_pose {type=Curve, loc=<0 0 0>, rot=<0 0 0 1>}
 R_mouth_stretch_pose {type=Curve, loc=<0 0 0>, rot=<0 0 0 1>}
 mouth_roll_lower_pose {type=Curve, loc=<0 0 0>, rot=<0 0 0 1>}
 mouth_roll_upper_pose {type=Curve, loc=<0 0 0>, rot=<0 0 0 1>}
 mouth_shrug_lower_pose {type=Curve, loc=<0 0 0>, rot=<0 0 0 1>}
 mouth_shrug_upper_pose {type=Curve, loc=<0 0 0>, rot=<0 0 0 1>}
 L_mouth_press_pose {type=Curve, loc=<0 0 0>, rot=<0 0 0 1>}
 R_mouth_press_pose {type=Curve, loc=<0 0 0>, rot=<0 0 0 1>}
 cheek_puff_pose {type=Curve, loc=<0 0 0>, rot=<0 0 0 1>}
 L_nose_wrinkler_pose {type=Curve, loc=<0 0 0>, rot=<0 0 0 1>}
 R_nose_wrinkler_pose {type=Curve, loc=<0 0 0>, rot=<0 0 0 1>}
 Pose_415 {type=Curve, loc=<0 0 0>, rot=<0 0 0 1>}
 Pose_416 {type=Curve, loc=<0 0 0>, rot=<0 0 0 1>}
 Pose_417 {type=Curve, loc=<0 0 0>, rot=<0 0 0 1>}
 Pose_418 {type=Curve, loc=<0 0 0>, rot=<0 0 0 1>}
 Pose_419 {type=Curve, loc=<0 0 0>, rot=<0 0 0 1>}
 Pose_420 {type=Curve, loc=<0 0 0>, rot=<0 0 0 1>}
 Pose_421 {type=Curve, loc=<0 0 0>, rot=<0 0 0 1>}
 Pose_422 {type=Curve, loc=<0 0 0>, rot=<0 0 0 1>}
 Pose_423 {type=Curve, loc=<0 0 0>, rot=<0 0 0 1>}
 Pose_424 {type=Curve, loc=<0 0 0>, rot=<0 0 0 1>}
 Pose_425 {type=Curve, loc=<0 0 0>, rot=<0 0 0 1>}
 Pose_426 {type=Curve, loc=<0 0 0>, rot=<0 0 0 1>}
 Pose_427 {type=Curve, loc=<0 0 0>, rot=<0 0 0 1>}
 Pose_428 {type=Curve, loc=<0 0 0>, rot=<0 0 0 1>}
 Pose_429 {type=Curve, loc=<0 0 0>, rot=<0 0 0 1>}
 Pose_430 {type=Curve, loc=<0 0 0>, rot=<0 0 0 1>}
 Pose_431 {type=Curve, loc=<0 0 0>, rot=<0 0 0 1>}
 Pose_432 {type=Curve, loc=<0 0 0>, rot=<0 0 0 1>}
 Pose_433 {type=Curve, loc=<0 0 0>, rot=<0 0 0 1>}
 Pose_434 {type=Curve, loc=<0 0 0>, rot=<0 0 0 1>}
 Pose_435 {type=Curve, loc=<0 0 0>, rot=<0 0 0 1>}
 Pose_436 {type=Curve, loc=<0 0 0>, rot=<0 0 0 1>}
 Pose_437 {type=Curve, loc=<0 0 0>, rot=<0 0 0 1>}
 Pose_438 {type=Curve, loc=<0 0 0>, rot=<0 0 0 1>}
 Pose_439 {type=Curve, loc=<0 0 0>, rot=<0 0 0 1>}
 Pose_440 {type=Curve, loc=<0 0 0>, rot=<0 0 0 1>}
 Pose_441 {type=Curve, loc=<0 0 0>, rot=<0 0 0 1>}
 Pose_442 {type=Curve, loc=<0 0 0>, rot=<0 0 0 1>}
 Pose_443 {type=Curve, loc=<0 0 0>, rot=<0 0 0 1>}
 Pose_444 {type=Curve, loc=<0 0 0>, rot=<0 0 0 1>}
 Pose_445 {type=Curve, loc=<0 0 0>, rot=<0 0 0 1>}
 Pose_446 {type=Curve, loc=<0 0 0>, rot=<0 0 0 1>}
 Pose_447 {type=Curve, loc=<0 0 0>, rot=<0 0 0 1>}
 Pose_448 {type=Curve, loc=<0 0 0>, rot=<0 0 0 1>}
 Pose_449 {type=Curve, loc=<0 0 0>, rot=<0 0 0 1>}
 Pose_450 {type=Curve, loc=<0 0 0>, rot=<0 0 0 1>}
 Pose_451 {type=Curve, loc=<0 0 0>, rot=<0 0 0 1>}
 Pose_452 {type=Curve, loc=<0 0 0>, rot=<0 0 0 1>}
 Pose_453 {type=Curve, loc=<0 0 0>, rot=<0 0 0 1>}
 Pose_454 {type=Curve, loc=<0 0 0>, rot=<0 0 0 1>}
 Pose_455 {type=Curve, loc=<0 0 0>, rot=<0 0 0 1>}
 Pose_456 {type=Curve, loc=<0 0 0>, rot=<0 0 0 1>}
 Pose_457 {type=Curve, loc=<0 0 0>, rot=<0 0 0 1>}
 Pose_458 {type=Curve, loc=<0 0 0>, rot=<0 0 0 1>}
 Pose_459 {type=Curve, loc=<0 0 0>, rot=<0 0 0 1>}
 Pose_460 {type=Curve, loc=<0 0 0>, rot=<0 0 0 1>}
 Pose_461 {type=Curve, loc=<0 0 0>, rot=<0 0 0 1>}
 Pose_462 {type=Curve, loc=<0 0 0>, rot=<0 0 0 1>}
 Pose_463 {type=Curve, loc=<0 0 0>, rot=<0 0 0 1>}
 Pose_464 {type=Curve, loc=<0 0 0>, rot=<0 0 0 1>}
 Pose_465 {type=Curve, loc=<0 0 0>, rot=<0 0 0 1>}
 Pose_466 {type=Curve, loc=<0 0 0>, rot=<0 0 0 1>}
 Pose_467 {type=Curve, loc=<0 0 0>, rot=<0 0 0 1>}
 Pose_468 {type=Curve, loc=<0 0 0>, rot=<0 0 0 1>}
 Pose_469 {type=Curve, loc=<0 0 0>, rot=<0 0 0 1>}
 Pose_470 {type=Curve, loc=<0 0 0>, rot=<0 0 0 1>}
 Pose_471 {type=Curve, loc=<0 0 0>, rot=<0 0 0 1>}
 Pose_472 {type=Curve, loc=<0 0 0>, rot=<0 0 0 1>}
 Pose_473 {type=Curve, loc=<0 0 0>, rot=<0 0 0 1>}
 Pose_474 {type=Curve, loc=<0 0 0>, rot=<0 0 0 1>}
 Pose_475 {type=Curve, loc=<0 0 0>, rot=<0 0 0 1>}
 Pose_476 {type=Curve, loc=<0 0 0>, rot=<0 0 0 1>}
 Pose_477 {type=Curve, loc=<0 0 0>, rot=<0 0 0 1>}
 Pose_478 {type=Curve, loc=<0 0 0>, rot=<0 0 0 1>}
 Pose_479 {type=Curve, loc=<0 0 0>, rot=<0 0 0 1>}
 Pose_480 {type=Curve, loc=<0 0 0>, rot=<0 0 0 1>}
 Pose_481 {type=Curve, loc=<0 0 0>, rot=<0 0 0 1>}
 Pose_482 {type=Curve, loc=<0 0 0>, rot=<0 0 0 1>}
 Pose_483 {type=Curve, loc=<0 0 0>, rot=<0 0 0 1>}
 Pose_484 {type=Curve, loc=<0 0 0>, rot=<0 0 0 1>}
 Pose_485 {type=Curve, loc=<0 0 0>, rot=<0 0 0 1>}
 Pose_486 {type=Curve, loc=<0 0 0>, rot=<0 0 0 1>}
 Pose_487 {type=Curve, loc=<0 0 0>, rot=<0 0 0 1>}
 Pose_488 {type=Curve, loc=<0 0 0>, rot=<0 0 0 1>}
 Pose_489 {type=Curve, loc=<0 0 0>, rot=<0 0 0 1>}
 Pose_490 {type=Curve, loc=<0 0 0>, rot=<0 0 0 1>}
 Pose_491 {type=Curve, loc=<0 0 0>, rot=<0 0 0 1>}
 Pose_492 {type=Curve, loc=<0 0 0>, rot=<0 0 0 1>}
 Pose_493 {type=Curve, loc=<0 0 0>, rot=<0 0 0 1>}
 Pose_494 {type=Curve, loc=<0 0 0>, rot=<0 0 0 1>}
 Pose_495 {type=Curve, loc=<0 0 0>, rot=<0 0 0 1>}
 Pose_496 {type=Curve, loc=<0 0 0>, rot=<0 0 0 1>}
 Pose_497 {type=Curve, loc=<0 0 0>, rot=<0 0 0 1>}
 Pose_498 {type=Curve, loc=<0 0 0>, rot=<0 0 0 1>}
 Pose_499 {type=Curve, loc=<0 0 0>, rot=<0 0 0 1>}
 Pose_500 {type=Curve, loc=<0 0 0>, rot=<0 0 0 1>}
 Pose_501 {type=Curve, loc=<0 0 0>, rot=<0 0 0 1>}
 Pose_502 {type=Curve, loc=<0 0 0>, rot=<0 0 0 1>}
 Pose_503 {type=Curve, loc=<0 0 0>, rot=<0 0 0 1>}
 Pose_504 {type=Curve, loc=<0 0 0>, rot=<0 0 0 1>}
 Pose_505 {type=Curve, loc=<0 0 0>, rot=<0 0 0 1>}
 Pose_506 {type=Curve, loc=<0 0 0>, rot=<0 0 0 1>}
 Pose_507 {type=Curve, loc=<0 0 0>, rot=<0 0 0 1>}
 Pose_508 {type=Curve, loc=<0 0 0>, rot=<0 0 0 1>}
 Pose_509 {type=Curve, loc=<0 0 0>, rot=<0 0 0 1>}
 Pose_510 {type=Curve, loc=<0 0 0>, rot=<0 0 0 1>}
 Pose_511 {type=Curve, loc=<0 0 0>, rot=<0 0 0 1>}
 Pose_512 {type=Curve, loc=<0 0 0>, rot=<0 0 0 1>}
 Pose_513 {type=Curve, loc=<0 0 0>, rot=<0 0 0 1>}
 Pose_514 {type=Curve, loc=<0 0 0>, rot=<0 0 0 1>}
 Pose_515 {type=Curve, loc=<0 0 0>, rot=<0 0 0 1>}
 Pose_516 {type=Curve, loc=<0 0 0>, rot=<0 0 0 1>}
 Pose_517 {type=Curve, loc=<0 0 0>, rot=<0 0 0 1>}
 Pose_518 {type=Curve, loc=<0 0 0>, rot=<0 0 0 1>}
 Pose_519 {type=Curve, loc=<0 0 0>, rot=<0 0 0 1>}
 Pose_520 {type=Curve, loc=<0 0 0>, rot=<0 0 0 1>}
 Pose_521 {type=Curve, loc=<0 0 0>, rot=<0 0 0 1>}
 Pose_522 {type=Curve, loc=<0 0 0>, rot=<0 0 0 1>}
 Pose_523 {type=Curve, loc=<0 0 0>, rot=<0 0 0 1>}
 Pose_524 {type=Curve, loc=<0 0 0>, rot=<0 0 0 1>}
 Pose_525 {type=Curve, loc=<0 0 0>, rot=<0 0 0 1>}
 Pose_526 {type=Curve, loc=<0 0 0>, rot=<0 0 0 1>}
 Pose_527 {type=Curve, loc=<0 0 0>, rot=<0 0 0 1>}
 Pose_528 {type=Curve, loc=<0 0 0>, rot=<0 0 0 1>}
 Pose_529 {type=Curve, loc=<0 0 0>, rot=<0 0 0 1>}
 Pose_530 {type=Curve, loc=<0 0 0>, rot=<0 0 0 1>}
 Pose_531 {type=Curve, loc=<0 0 0>, rot=<0 0 0 1>}
 Pose_532 {type=Curve, loc=<0 0 0>, rot=<0 0 0 1>}
 Pose_533 {type=Curve, loc=<0 0 0>, rot=<0 0 0 1>}
 Pose_534 {type=Curve, loc=<0 0 0>, rot=<0 0 0 1>}
 Pose_535 {type=Curve, loc=<0 0 0>, rot=<0 0 0 1>}
 Pose_536 {type=Curve, loc=<0 0 0>, rot=<0 0 0 1>}
 Pose_537 {type=Curve, loc=<0 0 0>, rot=<0 0 0 1>}
 Pose_538 {type=Curve, loc=<0 0 0>, rot=<0 0 0 1>}
 Pose_539 {type=Curve, loc=<0 0 0>, rot=<0 0 0 1>}
 Pose_540 {type=Curve, loc=<0 0 0>, rot=<0 0 0 1>}
 Pose_541 {type=Curve, loc=<0 0 0>, rot=<0 0 0 1>}
 Pose_542 {type=Curve, loc=<0 0 0>, rot=<0 0 0 1>}
 Pose_543 {type=Curve, loc=<0 0 0>, rot=<0 0 0 1>}
 Pose_544 {type=Curve, loc=<0 0 0>, rot=<0 0 0 1>}
 Pose_545 {type=Curve, loc=<0 0 0>, rot=<0 0 0 1>}
 Pose_546 {type=Curve, loc=<0 0 0>, rot=<0 0 0 1>}
 Pose_547 {type=Curve, loc=<0 0 0>, rot=<0 0 0 1>}
 Pose_548 {type=Curve, loc=<0 0 0>, rot=<0 0 0 1>}
 Pose_549 {type=Curve, loc=<0 0 0>, rot=<0 0 0 1>}
 Pose_550 {type=Curve, loc=<0 0 0>, rot=<0 0 0 1>}
 Pose_551 {type=Curve, loc=<0 0 0>, rot=<0 0 0 1>}
 Pose_552 {type=Curve, loc=<0 0 0>, rot=<0 0 0 1>}
 Pose_553 {type=Curve, loc=<0 0 0>, rot=<0 0 0 1>}
 Pose_554 {type=Curve, loc=<0 0 0>, rot=<0 0 0 1>}
 Pose_555 {type=Curve, loc=<0 0 0>, rot=<0 0 0 1>}
 Pose_556 {type=Curve, loc=<0 0 0>, rot=<0 0 0 1>}
 Pose_557 {type=Curve, loc=<0 0 0>, rot=<0 0 0 1>}
 Pose_558 {type=Curve, loc=<0 0 0>, rot=<0 0 0 1>}
 Pose_559 {type=Curve, loc=<0 0 0>, rot=<0 0 0 1>}
 Pose_560 {type=Curve, loc=<0 0 0>, rot=<0 0 0 1>}
 Pose_561 {type=Curve, loc=<0 0 0>, rot=<0 0 0 1>}
 Pose_562 {type=Curve, loc=<0 0 0>, rot=<0 0 0 1>}
 Pose_563 {type=Curve, loc=<0 0 0>, rot=<0 0 0 1>}
 Pose_564 {type=Curve, loc=<0 0 0>, rot=<0 0 0 1>}
 Pose_565 {type=Curve, loc=<0 0 0>, rot=<0 0 0 1>}
 Pose_566 {type=Curve, loc=<0 0 0>, rot=<0 0 0 1>}
 Pose_567 {type=Curve, loc=<0 0 0>, rot=<0 0 0 1>}
 Pose_568 {type=Curve, loc=<0 0 0>, rot=<0 0 0 1>}
 Pose_569 {type=Curve, loc=<0 0 0>, rot=<0 0 0 1>}
 Pose_570 {type=Curve, loc=<0 0 0>, rot=<0 0 0 1>}
 Pose_571 {type=Curve, loc=<0 0 0>, rot=<0 0 0 1>}
 Pose_572 {type=Curve, loc=<0 0 0>, rot=<0 0 0 1>}
 Pose_573 {type=Curve, loc=<0 0 0>, rot=<0 0 0 1>}
 Pose_574 {type=Curve, loc=<0 0 0>, rot=<0 0 0 1>}
 Pose_575 {type=Curve, loc=<0 0 0>, rot=<0 0 0 1>}
 Pose_576 {type=Curve, loc=<0 0 0>, rot=<0 0 0 1>}
 Pose_577 {type=Curve, loc=<0 0 0>, rot=<0 0 0 1>}
 Pose_578 {type=Curve, loc=<0 0 0>, rot=<0 0 0 1>}
 Pose_579 {type=Curve, loc=<0 0 0>, rot=<0 0 0 1>}
 Pose_580 {type=Curve, loc=<0 0 0>, rot=<0 0 0 1>}
 Pose_581 {type=Curve, loc=<0 0 0>, rot=<0 0 0 1>}
 Pose_582 {type=Curve, loc=<0 0 0>, rot=<0 0 0 1>}
 Pose_583 {type=Curve, loc=<0 0 0>, rot=<0 0 0 1>}
 Pose_584 {type=Curve, loc=<0 0 0>, rot=<0 0 0 1>}
 Pose_585 {type=Curve, loc=<0 0 0>, rot=<0 0 0 1>}
 Pose_586 {type=Curve, loc=<0 0 0>, rot=<0 0 0 1>}
 Pose_587 {type=Curve, loc=<0 0 0>, rot=<0 0 0 1>}
 Pose_588 {type=Curve, loc=<0 0 0>, rot=<0 0 0 1>}
 Pose_589 {type=Curve, loc=<0 0 0>, rot=<0 0 0 1>}
 Pose_590 {type=Curve, loc=<0 0 0>, rot=<0 0 0 1>}
 Pose_591 {type=Curve, loc=<0 0 0>, rot=<0 0 0 1>}
 Pose_592 {type=Curve, loc=<0 0 0>, rot=<0 0 0 1>}
 Pose_593 {type=Curve, loc=<0 0 0>, rot=<0 0 0 1>}
 Pose_594 {type=Curve, loc=<0 0 0>, rot=<0 0 0 1>}
 Pose_595 {type=Curve, loc=<0 0 0>, rot=<0 0 0 1>}
 Pose_596 {type=Curve, loc=<0 0 0>, rot=<0 0 0 1>}
 Pose_597 {type=Curve, loc=<0 0 0>, rot=<0 0 0 1>}
 Pose_598 {type=Curve, loc=<0 0 0>, rot=<0 0 0 1>}
 Pose_599 {type=Curve, loc=<0 0 0>, rot=<0 0 0 1>}
 Pose_600 {type=Curve, loc=<0 0 0>, rot=<0 0 0 1>}
 Pose_601 {type=Curve, loc=<0 0 0>, rot=<0 0 0 1>}
 Pose_602 {type=Curve, loc=<0 0 0>, rot=<0 0 0 1>}
 Pose_603 {type=Curve, loc=<0 0 0>, rot=<0 0 0 1>}
 Pose_604 {type=Curve, loc=<0 0 0>, rot=<0 0 0 1>}
 Pose_605 {type=Curve, loc=<0 0 0>, rot=<0 0 0 1>}
 Pose_606 {type=Curve, loc=<0 0 0>, rot=<0 0 0 1>}
 Pose_607 {type=Curve, loc=<0 0 0>, rot=<0 0 0 1>}
 Pose_608 {type=Curve, loc=<0 0 0>, rot=<0 0 0 1>}
 Pose_609 {type=Curve, loc=<0 0 0>, rot=<0 0 0 1>}
 Pose_610 {type=Curve, loc=<0 0 0>, rot=<0 0 0 1>}
 Pose_611 {type=Curve, loc=<0 0 0>, rot=<0 0 0 1>}
 Pose_612 {type=Curve, loc=<0 0 0>, rot=<0 0 0 1>}
 Pose_613 {type=Curve, loc=<0 0 0>, rot=<0 0 0 1>}
 Pose_614 {type=Curve, loc=<0 0 0>, rot=<0 0 0 1>}
 Pose_615 {type=Curve, loc=<0 0 0>, rot=<0 0 0 1>}
 Pose_616 {type=Curve, loc=<0 0 0>, rot=<0 0 0 1>}
 Pose_617 {type=Curve, loc=<0 0 0>, rot=<0 0 0 1>}
 Pose_618 {type=Curve, loc=<0 0 0>, rot=<0 0 0 1>}
 Pose_619 {type=Curve, loc=<0 0 0>, rot=<0 0 0 1>}
 Pose_620 {type=Curve, loc=<0 0 0>, rot=<0 0 0 1>}
 Pose_621 {type=Curve, loc=<0 0 0>, rot=<0 0 0 1>}
 Pose_622 {type=Curve, loc=<0 0 0>, rot=<0 0 0 1>}
 Pose_623 {type=Curve, loc=<0 0 0>, rot=<0 0 0 1>}
 Pose_624 {type=Curve, loc=<0 0 0>, rot=<0 0 0 1>}
 Pose_625 {type=Curve, loc=<0 0 0>, rot=<0 0 0 1>}
 Pose_626 {type=Curve, loc=<0 0 0>, rot=<0 0 0 1>}
 Pose_627 {type=Curve, loc=<0 0 0>, rot=<0 0 0 1>}
 Pose_628 {type=Curve, loc=<0 0 0>, rot=<0 0 0 1>}
 Pose_629 {type=Curve, loc=<0 0 0>, rot=<0 0 0 1>}
 Pose_630 {type=Curve, loc=<0 0 0>, rot=<0 0 0 1>}
 Pose_631 {type=Curve, loc=<0 0 0>, rot=<0 0 0 1>}
 Pose_632 {type=Curve, loc=<0 0 0>, rot=<0 0 0 1>}
 Pose_633 {type=Curve, loc=<0 0 0>, rot=<0 0 0 1>}
 Pose_634 {type=Curve, loc=<0 0 0>, rot=<0 0 0 1>}
 Pose_635 {type=Curve, loc=<0 0 0>, rot=<0 0 0 1>}
 Pose_636 {type=Curve, loc=<0 0 0>, rot=<0 0 0 1>}
 Pose_637 {type=Curve, loc=<0 0 0>, rot=<0 0 0 1>}
 Pose_638 {type=Curve, loc=<0 0 0>, rot=<0 0 0 1>}
 Pose_639 {type=Curve, loc=<0 0 0>, rot=<0 0 0 1>}
 Pose_640 {type=Curve, loc=<0 0 0>, rot=<0 0 0 1>}
 Pose_641 {type=Curve, loc=<0 0 0>, rot=<0 0 0 1>}
 Pose_642 {type=Curve, loc=<0 0 0>, rot=<0 0 0 1>}
 Pose_643 {type=Curve, loc=<0 0 0>, rot=<0 0 0 1>}
 Pose_644 {type=Curve, loc=<0 0 0>, rot=<0 0 0 1>}
 Pose_645 {type=Curve, loc=<0 0 0>, rot=<0 0 0 1>}
 Pose_646 {type=Curve, loc=<0 0 0>, rot=<0 0 0 1>}
 Pose_647 {type=Curve, loc=<0 0 0>, rot=<0 0 0 1>}
 Pose_648 {type=Curve, loc=<0 0 0>, rot=<0 0 0 1>}
 Pose_649 {type=Curve, loc=<0 0 0>, rot=<0 0 0 1>}
 Pose_650 {type=Curve, loc=<0 0 0>, rot=<0 0 0 1>}
 Pose_651 {type=Curve, loc=<0 0 0>, rot=<0 0 0 1>}
 base {type=Curve, loc=<0 0 0>, rot=<0 0 0 1>}
 eyeLookDownLeft {type=Curve, loc=<0 0 0>, rot=<0 0 0 1>}
 eyeLookInLeft {type=Curve, loc=<0 0 0>, rot=<0 0 0 1>}
 eyeLookUpLeft {type=Curve, loc=<0 0 0>, rot=<0 0 0 1>}
 eyeLookInRight {type=Curve, loc=<0 0 0>, rot=<0 0 0 1>}
 Pose_652 {type=Curve, loc=<0 0 0>, rot=<0 0 0 1>}
 Pose_653 {type=Curve, loc=<0 0 0>, rot=<0 0 0 1>}
 Pose_654 {type=Curve, loc=<0 0 0>, rot=<0 0 0 1>}
 Pose_655 {type=Curve, loc=<0 0 0>, rot=<0 0 0 1>}
 Pose_656 {type=Curve, loc=<0 0 0>, rot=<0 0 0 1>}
 Pose_657 {type=Curve, loc=<0 0 0>, rot=<0 0 0 1>}
 Pose_658 {type=Curve, loc=<0 0 0>, rot=<0 0 0 1>}
 Pose_659 {type=Curve, loc=<0 0 0>, rot=<0 0 0 1>}
 Pose_660 {type=Curve, loc=<0 0 0>, rot=<0 0 0 1>}
 Pose_661 {type=Curve, loc=<0 0 0>, rot=<0 0 0 1>}
 Pose_662 {type=Curve, loc=<0 0 0>, rot=<0 0 0 1>}
 Pose_663 {type=Curve, loc=<0 0 0>, rot=<0 0 0 1>}
 Pose_664 {type=Curve, loc=<0 0 0>, rot=<0 0 0 1>}
 Pose_665 {type=Curve, loc=<0 0 0>, rot=<0 0 0 1>}
 Pose_666 {type=Curve, loc=<0 0 0>, rot=<0 0 0 1>}
 Pose_667 {type=Curve, loc=<0 0 0>, rot=<0 0 0 1>}
 Pose_668 {type=Curve, loc=<0 0 0>, rot=<0 0 0 1>}
 Pose_669 {type=Curve, loc=<0 0 0>, rot=<0 0 0 1>}
 Pose_670 {type=Curve, loc=<0 0 0>, rot=<0 0 0 1>}
 Pose_671 {type=Curve, loc=<0 0 0>, rot=<0 0 0 1>}
 Pose_672 {type=Curve, loc=<0 0 0>, rot=<0 0 0 1>}
 Pose_673 {type=Curve, loc=<0 0 0>, rot=<0 0 0 1>}
 Pose_674 {type=Curve, loc=<0 0 0>, rot=<0 0 0 1>}
 Pose_675 {type=Curve, loc=<0 0 0>, rot=<0 0 0 1>}
 Pose_676 {type=Curve, loc=<0 0 0>, rot=<0 0 0 1>}
 Pose_677 {type=Curve, loc=<0 0 0>, rot=<0 0 0 1>}
 Pose_678 {type=Curve, loc=<0 0 0>, rot=<0 0 0 1>}
 Pose_679 {type=Curve, loc=<0 0 0>, rot=<0 0 0 1>}
 Pose_680 {type=Curve, loc=<0 0 0>, rot=<0 0 0 1>}
 Pose_681 {type=Curve, loc=<0 0 0>, rot=<0 0 0 1>}
 Pose_682 {type=Curve, loc=<0 0 0>, rot=<0 0 0 1>}
 Pose_683 {type=Curve, loc=<0 0 0>, rot=<0 0 0 1>}
 Pose_684 {type=Curve, loc=<0 0 0>, rot=<0 0 0 1>}
 Pose_685 {type=Curve, loc=<0 0 0>, rot=<0 0 0 1>}
 Pose_686 {type=Curve, loc=<0 0 0>, rot=<0 0 0 1>}
 Pose_687 {type=Curve, loc=<0 0 0>, rot=<0 0 0 1>}
 Pose_688 {type=Curve, loc=<0 0 0>, rot=<0 0 0 1>}
 Pose_689 {type=Curve, loc=<0 0 0>, rot=<0 0 0 1>}
 Pose_690 {type=Curve, loc=<0 0 0>, rot=<0 0 0 1>}
 Pose_691 {type=Curve, loc=<0 0 0>, rot=<0 0 0 1>}
 Pose_692 {type=Curve, loc=<0 0 0>, rot=<0 0 0 1>}
 Pose_693 {type=Curve, loc=<0 0 0>, rot=<0 0 0 1>}
 Pose_694 {type=Curve, loc=<0 0 0>, rot=<0 0 0 1>}
 Pose_695 {type=Curve, loc=<0 0 0>, rot=<0 0 0 1>}
 Pose_696 {type=Curve, loc=<0 0 0>, rot=<0 0 0 1>}
 Pose_697 {type=Curve, loc=<0 0 0>, rot=<0 0 0 1>}
 Pose_698 {type=Curve, loc=<0 0 0>, rot=<0 0 0 1>}
 Pose_699 {type=Curve, loc=<0 0 0>, rot=<0 0 0 1>}
 Pose_700 {type=Curve, loc=<0 0 0>, rot=<0 0 0 1>}
 Pose_701 {type=Curve, loc=<0 0 0>, rot=<0 0 0 1>}
 Pose_702 {type=Curve, loc=<0 0 0>, rot=<0 0 0 1>}
 Pose_703 {type=Curve, loc=<0 0 0>, rot=<0 0 0 1>}
 Pose_704 {type=Curve, loc=<0 0 0>, rot=<0 0 0 1>}
 Pose_705 {type=Curve, loc=<0 0 0>, rot=<0 0 0 1>}
 Pose_706 {type=Curve, loc=<0 0 0>, rot=<0 0 0 1>}
 Pose_707 {type=Curve, loc=<0 0 0>, rot=<0 0 0 1>}
 Pose_708 {type=Curve, loc=<0 0 0>, rot=<0 0 0 1>}
 Pose_709 {type=Curve, loc=<0 0 0>, rot=<0 0 0 1>}
 Pose_710 {type=Curve, loc=<0 0 0>, rot=<0 0 0 1>}
 Pose_711 {type=Curve, loc=<0 0 0>, rot=<0 0 0 1>}
 Pose_712 {type=Curve, loc=<0 0 0>, rot=<0 0 0 1>}
 Pose_713 {type=Curve, loc=<0 0 0>, rot=<0 0 0 1>}
 Pose_714 {type=Curve, loc=<0 0 0>, rot=<0 0 0 1>}
 Pose_715 {type=Curve, loc=<0 0 0>, rot=<0 0 0 1>}
 Pose_716 {type=Curve, loc=<0 0 0>, rot=<0 0 0 1>}
 Pose_717 {type=Curve, loc=<0 0 0>, rot=<0 0 0 1>}
 Pose_718 {type=Curve, loc=<0 0 0>, rot=<0 0 0 1>}
 Pose_719 {type=Curve, loc=<0 0 0>, rot=<0 0 0 1>}
 Pose_720 {type=Curve, loc=<0 0 0>, rot=<0 0 0 1>}
 Pose_721 {type=Curve, loc=<0 0 0>, rot=<0 0 0 1>}
 Pose_722 {type=Curve, loc=<0 0 0>, rot=<0 0 0 1>}
 Pose_723 {type=Curve, loc=<0 0 0>, rot=<0 0 0 1>}
 Pose_724 {type=Curve, loc=<0 0 0>, rot=<0 0 0 1>}
 Pose_725 {type=Curve, loc=<0 0 0>, rot=<0 0 0 1>}
 Pose_726 {type=Curve, loc=<0 0 0>, rot=<0 0 0 1>}
 Pose_727 {type=Curve, loc=<0 0 0>, rot=<0 0 0 1>}
 Pose_728 {type=Curve, loc=<0 0 0>, rot=<0 0 0 1>}
 Pose_729 {type=Curve, loc=<0 0 0>, rot=<0 0 0 1>}
 Pose_730 {type=Curve, loc=<0 0 0>, rot=<0 0 0 1>}
 Pose_731 {type=Curve, loc=<0 0 0>, rot=<0 0 0 1>}
 Pose_732 {type=Curve, loc=<0 0 0>, rot=<0 0 0 1>}
 Pose_733 {type=Curve, loc=<0 0 0>, rot=<0 0 0 1>}
 Pose_734 {type=Curve, loc=<0 0 0>, rot=<0 0 0 1>}
 Pose_735 {type=Curve, loc=<0 0 0>, rot=<0 0 0 1>}
 Pose_736 {type=Curve, loc=<0 0 0>, rot=<0 0 0 1>}
 Pose_737 {type=Curve, loc=<0 0 0>, rot=<0 0 0 1>}
 Pose_738 {type=Curve, loc=<0 0 0>, rot=<0 0 0 1>}
 Pose_739 {type=Curve, loc=<0 0 0>, rot=<0 0 0 1>}
 Pose_740 {type=Curve, loc=<0 0 0>, rot=<0 0 0 1>}
 Pose_741 {type=Curve, loc=<0 0 0>, rot=<0 0 0 1>}
 Pose_742 {type=Curve, loc=<0 0 0>, rot=<0 0 0 1>}
 Pose_743 {type=Curve, loc=<0 0 0>, rot=<0 0 0 1>}
 Pose_744 {type=Curve, loc=<0 0 0>, rot=<0 0 0 1>}
 Pose_745 {type=Curve, loc=<0 0 0>, rot=<0 0 0 1>}
 Pose_746 {type=Curve, loc=<0 0 0>, rot=<0 0 0 1>}
 Pose_747 {type=Curve, loc=<0 0 0>, rot=<0 0 0 1>}
 Pose_748 {type=Curve, loc=<0 0 0>, rot=<0 0 0 1>}
 Pose_749 {type=Curve, loc=<0 0 0>, rot=<0 0 0 1>}
 Pose_750 {type=Curve, loc=<0 0 0>, rot=<0 0 0 1>}
 Pose_751 {type=Curve, loc=<0 0 0>, rot=<0 0 0 1>}
 Pose_752 {type=Curve, loc=<0 0 0>, rot=<0 0 0 1>}
 Pose_753 {type=Curve, loc=<0 0 0>, rot=<0 0 0 1>}
 Pose_754 {type=Curve, loc=<0 0 0>, rot=<0 0 0 1>}
 Pose_755 {type=Curve, loc=<0 0 0>, rot=<0 0 0 1>}
 Pose_756 {type=Curve, loc=<0 0 0>, rot=<0 0 0 1>}
 Pose_757 {type=Curve, loc=<0 0 0>, rot=<0 0 0 1>}
 Pose_758 {type=Curve, loc=<0 0 0>, rot=<0 0 0 1>}
 Pose_759 {type=Curve, loc=<0 0 0>, rot=<0 0 0 1>}
 Pose_760 {type=Curve, loc=<0 0 0>, rot=<0 0 0 1>}
 Pose_761 {type=Curve, loc=<0 0 0>, rot=<0 0 0 1>}
 Pose_762 {type=Curve, loc=<0 0 0>, rot=<0 0 0 1>}
 Pose_763 {type=Curve, loc=<0 0 0>, rot=<0 0 0 1>}
 Pose_764 {type=Curve, loc=<0 0 0>, rot=<0 0 0 1>}
 Pose_765 {type=Curve, loc=<0 0 0>, rot=<0 0 0 1>}
 Pose_766 {type=Curve, loc=<0 0 0>, rot=<0 0 0 1>}
 Pose_767 {type=Curve, loc=<0 0 0>, rot=<0 0 0 1>}
 Pose_768 {type=Curve, loc=<0 0 0>, rot=<0 0 0 1>}
 Pose_769 {type=Curve, loc=<0 0 0>, rot=<0 0 0 1>}
 Pose_770 {type=Curve, loc=<0 0 0>, rot=<0 0 0 1>}
 Pose_771 {type=Curve, loc=<0 0 0>, rot=<0 0 0 1>}
 Pose_772 {type=Curve, loc=<0 0 0>, rot=<0 0 0 1>}
 Pose_773 {type=Curve, loc=<0 0 0>, rot=<0 0 0 1>}
 Pose_774 {type=Curve, loc=<0 0 0>, rot=<0 0 0 1>}
 Pose_775 {type=Curve, loc=<0 0 0>, rot=<0 0 0 1>}
 Pose_776 {type=Curve, loc=<0 0 0>, rot=<0 0 0 1>}
 Pose_777 {type=Curve, loc=<0 0 0>, rot=<0 0 0 1>}
 Pose_778 {type=Curve, loc=<0 0 0>, rot=<0 0 0 1>}
 Pose_779 {type=Curve, loc=<0 0 0>, rot=<0 0 0 1>}
 Pose_780 {type=Curve, loc=<0 0 0>, rot=<0 0 0 1>}
 Pose_781 {type=Curve, loc=<0 0 0>, rot=<0 0 0 1>}
 Pose_782 {type=Curve, loc=<0 0 0>, rot=<0 0 0 1>}
 Pose_783 {type=Curve, loc=<0 0 0>, rot=<0 0 0 1>}
 Pose_784 {type=Curve, loc=<0 0 0>, rot=<0 0 0 1>}
 Pose_785 {type=Curve, loc=<0 0 0>, rot=<0 0 0 1>}
 Pose_786 {type=Curve, loc=<0 0 0>, rot=<0 0 0 1>}
 Pose_787 {type=Curve, loc=<0 0 0>, rot=<0 0 0 1>}
 Pose_788 {type=Curve, loc=<0 0 0>, rot=<0 0 0 1>}
 Pose_789 {type=Curve, loc=<0 0 0>, rot=<0 0 0 1>}
 Pose_790 {type=Curve, loc=<0 0 0>, rot=<0 0 0 1>}
 Pose_791 {type=Curve, loc=<0 0 0>, rot=<0 0 0 1>}
 Pose_792 {type=Curve, loc=<0 0 0>, rot=<0 0 0 1>}
 Pose_793 {type=Curve, loc=<0 0 0>, rot=<0 0 0 1>}
 Pose_794 {type=Curve, loc=<0 0 0>, rot=<0 0 0 1>}
 Pose_795 {type=Curve, loc=<0 0 0>, rot=<0 0 0 1>}
 Pose_796 {type=Curve, loc=<0 0 0>, rot=<0 0 0 1>}
 Pose_797 {type=Curve, loc=<0 0 0>, rot=<0 0 0 1>}
 Pose_798 {type=Curve, loc=<0 0 0>, rot=<0 0 0 1>}
 Pose_799 {type=Curve, loc=<0 0 0>, rot=<0 0 0 1>}
 Pose_800 {type=Curve, loc=<0 0 0>, rot=<0 0 0 1>}
 Pose_801 {type=Curve, loc=<0 0 0>, rot=<0 0 0 1>}
 Pose_802 {type=Curve, loc=<0 0 0>, rot=<0 0 0 1>}
 Pose_803 {type=Curve, loc=<0 0 0>, rot=<0 0 0 1>}
 Pose_804 {type=Curve, loc=<0 0 0>, rot=<0 0 0 1>}
 Pose_805 {type=Curve, loc=<0 0 0>, rot=<0 0 0 1>}
 Pose_806 {type=Curve, loc=<0 0 0>, rot=<0 0 0 1>}
 Pose_807 {type=Curve, loc=<0 0 0>, rot=<0 0 0 1>}
 Pose_808 {type=Curve, loc=<0 0 0>, rot=<0 0 0 1>}
 Pose_809 {type=Curve, loc=<0 0 0>, rot=<0 0 0 1>}
 Pose_810 {type=Curve, loc=<0 0 0>, rot=<0 0 0 1>}
 Pose_811 {type=Curve, loc=<0 0 0>, rot=<0 0 0 1>}
 Pose_812 {type=Curve, loc=<0 0 0>, rot=<0 0 0 1>}
 Pose_813 {type=Curve, loc=<0 0 0>, rot=<0 0 0 1>}
 Pose_814 {type=Curve, loc=<0 0 0>, rot=<0 0 0 1>}
 Pose_815 {type=Curve, loc=<0 0 0>, rot=<0 0 0 1>}
 Pose_816 {type=Curve, loc=<0 0 0>, rot=<0 0 0 1>}
 Pose_817 {type=Curve, loc=<0 0 0>, rot=<0 0 0 1>}
 Pose_818 {type=Curve, loc=<0 0 0>, rot=<0 0 0 1>}
 Pose_819 {type=Curve, loc=<0 0 0>, rot=<0 0 0 1>}
 Pose_820 {type=Curve, loc=<0 0 0>, rot=<0 0 0 1>}
 Pose_821 {type=Curve, loc=<0 0 0>, rot=<0 0 0 1>}
 Pose_822 {type=Curve, loc=<0 0 0>, rot=<0 0 0 1>}
 Pose_823 {type=Curve, loc=<0 0 0>, rot=<0 0 0 1>}
 Pose_824 {type=Curve, loc=<0 0 0>, rot=<0 0 0 1>}
 Pose_825 {type=Curve, loc=<0 0 0>, rot=<0 0 0 1>}
 Pose_826 {type=Curve, loc=<0 0 0>, rot=<0 0 0 1>}
 Pose_827 {type=Curve, loc=<0 0 0>, rot=<0 0 0 1>}
 Pose_828 {type=Curve, loc=<0 0 0>, rot=<0 0 0 1>}
 Pose_829 {type=Curve, loc=<0 0 0>, rot=<0 0 0 1>}
 Pose_830 {type=Curve, loc=<0 0 0>, rot=<0 0 0 1>}
 Pose_831 {type=Curve, loc=<0 0 0>, rot=<0 0 0 1>}
 Pose_832 {type=Curve, loc=<0 0 0>, rot=<0 0 0 1>}
 Pose_833 {type=Curve, loc=<0 0 0>, rot=<0 0 0 1>}
 Pose_834 {type=Curve, loc=<0 0 0>, rot=<0 0 0 1>}
 Pose_835 {type=Curve, loc=<0 0 0>, rot=<0 0 0 1>}
 Pose_836 {type=Curve, loc=<0 0 0>, rot=<0 0 0 1>}
 Pose_837 {type=Curve, loc=<0 0 0>, rot=<0 0 0 1>}
 Pose_838 {type=Curve, loc=<0 0 0>, rot=<0 0 0 1>}
 Pose_839 {type=Curve, loc=<0 0 0>, rot=<0 0 0 1>}
 Pose_840 {type=Curve, loc=<0 0 0>, rot=<0 0 0 1>}
 Pose_841 {type=Curve, loc=<0 0 0>, rot=<0 0 0 1>}
 Pose_842 {type=Curve, loc=<0 0 0>, rot=<0 0 0 1>}
 Pose_843 {type=Curve, loc=<0 0 0>, rot=<0 0 0 1>}
 Pose_844 {type=Curve, loc=<0 0 0>, rot=<0 0 0 1>}
 Pose_845 {type=Curve, loc=<0 0 0>, rot=<0 0 0 1>}
 Pose_846 {type=Curve, loc=<0 0 0>, rot=<0 0 0 1>}
 Pose_847 {type=Curve, loc=<0 0 0>, rot=<0 0 0 1>}
 Pose_848 {type=Curve, loc=<0 0 0>, rot=<0 0 0 1>}
 Pose_849 {type=Curve, loc=<0 0 0>, rot=<0 0 0 1>}
 Pose_850 {type=Curve, loc=<0 0 0>, rot=<0 0 0 1>}
 Pose_851 {type=Curve, loc=<0 0 0>, rot=<0 0 0 1>}
 Pose_852 {type=Curve, loc=<0 0 0>, rot=<0 0 0 1>}
 Pose_853 {type=Curve, loc=<0 0 0>, rot=<0 0 0 1>}
 Pose_854 {type=Curve, loc=<0 0 0>, rot=<0 0 0 1>}
 Pose_855 {type=Curve, loc=<0 0 0>, rot=<0 0 0 1>}
 Pose_856 {type=Curve, loc=<0 0 0>, rot=<0 0 0 1>}
 Pose_857 {type=Curve, loc=<0 0 0>, rot=<0 0 0 1>}
 Pose_858 {type=Curve, loc=<0 0 0>, rot=<0 0 0 1>}
 Pose_859 {type=Curve, loc=<0 0 0>, rot=<0 0 0 1>}
 Pose_860 {type=Curve, loc=<0 0 0>, rot=<0 0 0 1>}
 Pose_861 {type=Curve, loc=<0 0 0>, rot=<0 0 0 1>}
 Pose_862 {type=Curve, loc=<0 0 0>, rot=<0 0 0 1>}
 Pose_863 {type=Curve, loc=<0 0 0>, rot=<0 0 0 1>}
 Pose_864 {type=Curve, loc=<0 0 0>, rot=<0 0 0 1>}
 Pose_865 {type=Curve, loc=<0 0 0>, rot=<0 0 0 1>}
 Pose_866 {type=Curve, loc=<0 0 0>, rot=<0 0 0 1>}
 Pose_867 {type=Curve, loc=<0 0 0>, rot=<0 0 0 1>}
 Pose_868 {type=Curve, loc=<0 0 0>, rot=<0 0 0 1>}
 Pose_869 {type=Curve, loc=<0 0 0>, rot=<0 0 0 1>}
 Pose_870 {type=Curve, loc=<0 0 0>, rot=<0 0 0 1>}
 Pose_871 {type=Curve, loc=<0 0 0>, rot=<0 0 0 1>}
 Pose_872 {type=Curve, loc=<0 0 0>, rot=<0 0 0 1>}
 Pose_873 {type=Curve, loc=<0 0 0>, rot=<0 0 0 1>}
 Pose_874 {type=Curve, loc=<0 0 0>, rot=<0 0 0 1>}
 Pose_875 {type=Curve, loc=<0 0 0>, rot=<0 0 0 1>}
 Pose_876 {type=Curve, loc=<0 0 0>, rot=<0 0 0 1>}
 Pose_877 {type=Curve, loc=<0 0 0>, rot=<0 0 0 1>}
 Pose_878 {type=Curve, loc=<0 0 0>, rot=<0 0 0 1>}
 Pose_879 {type=Curve, loc=<0 0 0>, rot=<0 0 0 1>}
 Pose_880 {type=Curve, loc=<0 0 0>, rot=<0 0 0 1>}
 Pose_881 {type=Curve, loc=<0 0 0>, rot=<0 0 0 1>}
 Pose_882 {type=Curve, loc=<0 0 0>, rot=<0 0 0 1>}
 Pose_883 {type=Curve, loc=<0 0 0>, rot=<0 0 0 1>}
 Pose_884 {type=Curve, loc=<0 0 0>, rot=<0 0 0 1>}
 Pose_885 {type=Curve, loc=<0 0 0>, rot=<0 0 0 1>}
 Pose_886 {type=Curve, loc=<0 0 0>, rot=<0 0 0 1>}
 Pose_887 {type=Curve, loc=<0 0 0>, rot=<0 0 0 1>}
 Pose_888 {type=Curve, loc=<0 0 0>, rot=<0 0 0 1>}
 Pose_889 {type=Curve, loc=<0 0 0>, rot=<0 0 0 1>}
 Pose_890 {type=Curve, loc=<0 0 0>, rot=<0 0 0 1>}
 Pose_891 {type=Curve, loc=<0 0 0>, rot=<0 0 0 1>}
 Pose_892 {type=Curve, loc=<0 0 0>, rot=<0 0 0 1>}
 Pose_893 {type=Curve, loc=<0 0 0>, rot=<0 0 0 1>}
 Pose_894 {type=Curve, loc=<0 0 0>, rot=<0 0 0 1>}
 Pose_895 {type=Curve, loc=<0 0 0>, rot=<0 0 0 1>}
 Pose_896 {type=Curve, loc=<0 0 0>, rot=<0 0 0 1>}
 Pose_897 {type=Curve, loc=<0 0 0>, rot=<0 0 0 1>}
 Pose_898 {type=Curve, loc=<0 0 0>, rot=<0 0 0 1>}
 Pose_899 {type=Curve, loc=<0 0 0>, rot=<0 0 0 1>}
 Pose_900 {type=Curve, loc=<0 0 0>, rot=<0 0 0 1>}
 Pose_901 {type=Curve, loc=<0 0 0>, rot=<0 0 0 1>}
 Pose_902 {type=Curve, loc=<0 0 0>, rot=<0 0 0 1>}
 Pose_903 {type=Curve, loc=<0 0 0>, rot=<0 0 0 1>}
 Pose_904 {type=Curve, loc=<0 0 0>, rot=<0 0 0 1>}
 Pose_905 {type=Curve, loc=<0 0 0>, rot=<0 0 0 1>}
 Pose_906 {type=Curve, loc=<0 0 0>, rot=<0 0 0 1>}
 Pose_907 {type=Curve, loc=<0 0 0>, rot=<0 0 0 1>}
 Pose_908 {type=Curve, loc=<0 0 0>, rot=<0 0 0 1>}
 Pose_909 {type=Curve, loc=<0 0 0>, rot=<0 0 0 1>}
 Pose_910 {type=Curve, loc=<0 0 0>, rot=<0 0 0 1>}
 Pose_911 {type=Curve, loc=<0 0 0>, rot=<0 0 0 1>}
 Pose_912 {type=Curve, loc=<0 0 0>, rot=<0 0 0 1>}
 Pose_913 {type=Curve, loc=<0 0 0>, rot=<0 0 0 1>}
 Pose_914 {type=Curve, loc=<0 0 0>, rot=<0 0 0 1>}
 Pose_915 {type=Curve, loc=<0 0 0>, rot=<0 0 0 1>}
 Pose_916 {type=Curve, loc=<0 0 0>, rot=<0 0 0 1>}
 Pose_917 {type=Curve, loc=<0 0 0>, rot=<0 0 0 1>}
 Pose_918 {type=Curve, loc=<0 0 0>, rot=<0 0 0 1>}
 Pose_919 {type=Curve, loc=<0 0 0>, rot=<0 0 0 1>}
 Pose_920 {type=Curve, loc=<0 0 0>, rot=<0 0 0 1>}
 Pose_921 {type=Curve, loc=<0 0 0>, rot=<0 0 0 1>}
 Pose_922 {type=Curve, loc=<0 0 0>, rot=<0 0 0 1>}
 Pose_923 {type=Curve, loc=<0 0 0>, rot=<0 0 0 1>}
 Pose_924 {type=Curve, loc=<0 0 0>, rot=<0 0 0 1>}
 Pose_925 {type=Curve, loc=<0 0 0>, rot=<0 0 0 1>}
 Pose_926 {type=Curve, loc=<0 0 0>, rot=<0 0 0 1>}
 Pose_927 {type=Curve, loc=<0 0 0>, rot=<0 0 0 1>}
 Pose_928 {type=Curve, loc=<0 0 0>, rot=<0 0 0 1>}
 Pose_929 {type=Curve, loc=<0 0 0>, rot=<0 0 0 1>}
 Pose_930 {type=Curve, loc=<0 0 0>, rot=<0 0 0 1>}
 Pose_931 {type=Curve, loc=<0 0 0>, rot=<0 0 0 1>}
 Pose_932 {type=Curve, loc=<0 0 0>, rot=<0 0 0 1>}
 Pose_933 {type=Curve, loc=<0 0 0>, rot=<0 0 0 1>}
 Pose_934 {type=Curve, loc=<0 0 0>, rot=<0 0 0 1>}
 Pose_935 {type=Curve, loc=<0 0 0>, rot=<0 0 0 1>}
 Pose_936 {type=Curve, loc=<0 0 0>, rot=<0 0 0 1>}
 Pose_937 {type=Curve, loc=<0 0 0>, rot=<0 0 0 1>}
 Pose_938 {type=Curve, loc=<0 0 0>, rot=<0 0 0 1>}
 Pose_939 {type=Curve, loc=<0 0 0>, rot=<0 0 0 1>}
 Pose_940 {type=Curve, loc=<0 0 0>, rot=<0 0 0 1>}
 Pose_941 {type=Curve, loc=<0 0 0>, rot=<0 0 0 1>}
 Pose_942 {type=Curve, loc=<0 0 0>, rot=<0 0 0 1>}
 Pose_943 {type=Curve, loc=<0 0 0>, rot=<0 0 0 1>}
 Pose_944 {type=Curve, loc=<0 0 0>, rot=<0 0 0 1>}
 Pose_945 {type=Curve, loc=<0 0 0>, rot=<0 0 0 1>}
 Pose_946 {type=Curve, loc=<0 0 0>, rot=<0 0 0 1>}
 Pose_947 {type=Curve, loc=<0 0 0>, rot=<0 0 0 1>}
 Pose_948 {type=Curve, loc=<0 0 0>, rot=<0 0 0 1>}
 Pose_949 {type=Curve, loc=<0 0 0>, rot=<0 0 0 1>}
 Pose_950 {type=Curve, loc=<0 0 0>, rot=<0 0 0 1>}
 Pose_951 {type=Curve, loc=<0 0 0>, rot=<0 0 0 1>}
 Pose_952 {type=Curve, loc=<0 0 0>, rot=<0 0 0 1>}
 Pose_953 {type=Curve, loc=<0 0 0>, rot=<0 0 0 1>}
 Pose_954 {type=Curve, loc=<0 0 0>, rot=<0 0 0 1>}
 Pose_955 {type=Curve, loc=<0 0 0>, rot=<0 0 0 1>}
 Pose_956 {type=Curve, loc=<0 0 0>, rot=<0 0 0 1>}
 Pose_957 {type=Curve, loc=<0 0 0>, rot=<0 0 0 1>}
 Pose_958 {type=Curve, loc=<0 0 0>, rot=<0 0 0 1>}
 Pose_959 {type=Curve, loc=<0 0 0>, rot=<0 0 0 1>}
 Pose_960 {type=Curve, loc=<0 0 0>, rot=<0 0 0 1>}
 Pose_961 {type=Curve, loc=<0 0 0>, rot=<0 0 0 1>}
 is_face_bone_anim {type=Curve, loc=<0 0 0>, rot=<0 0 0 1>}
 FullHourglass {type=Curve, loc=<0 0 0>, rot=<0 0 0 1>}
 LeanRectangle {type=Curve, loc=<0 0 0>, rot=<0 0 0 1>}
 InvertedTriangle {type=Curve, loc=<0 0 0>, rot=<0 0 0 1>}
 root {loc=<0 0 0>, rot=<0 0 0 1>}
 pelvis {parent=root, loc=<0 -0.7071 93.1191>, rot=<0.0127 0.707 -0.0127 -0.707>}
 spine_01 {parent=pelvis, loc=<6.7494 -0.6388 0>, rot=<0 0 0.034 -0.9994>}
 spine_02 {parent=spine_01, loc=<7.2518 -0.035 0>, rot=<0 0 0.0238 -0.9997>}
 spine_03 {parent=spine_02, loc=<6.0104 0 0>, rot=<0 0 0.0149 -0.9999>}
 spine_04 {parent=spine_03, loc=<8.1024 0 0>, rot=<0 0 -0.0132 -0.9999>}
 spine_05 {parent=spine_04, loc=<10.0146 0 0>, rot=<0 0 -0.0015 -1>}
 neck_01 {parent=spine_05, loc=<11.3138 0.1529 0>, rot=<0 0 -0.2003 -0.9797>}
 neck_02 {parent=neck_01, loc=<4.5627 0 0>, rot=<0 0 0.0504 -0.9987>}
 head {parent=neck_02, loc=<5.3914 0 0>, rot=<0 0 0.115 -0.9934>}
 faceAttach {parent=head, loc=<1.1549 0.1797 0.6454>, rot=<0.0029 -0.7071 -0.0029 -0.7071>}
 C_brow_mid {parent=faceAttach, loc=<0.6332 9.2477 3.9129>, rot=<0 0 0 -1>}
 C_jaw {parent=faceAttach, loc=<0.1538 1.3281 -0.8248>, rot=<0 0 0 -1>}
 C_lip_lower_mid {parent=C_jaw, loc=<0.5404 7.8082 -1.3402>, rot=<0 0 0 -1>}
 L_lip_lower_outer {parent=C_jaw, loc=<2.2095 7.1191 -0.6782>, rot=<0.0514 0.2276 0.2946 -0.9267>}
 R_lip_lower_outer {parent=C_jaw, loc=<-1.0153 7.1978 -0.9778>, rot=<0.0367 -0.1572 -0.3002 -0.9401>}
 teeth_lower {parent=C_jaw, loc=<0.4503 6.8908 -2.0389>, rot=<0 0 0 1>}
 tongue {parent=C_jaw, loc=<0.5924 3.2357 -1.2324>, rot=<-0.1568 0 0 -0.9876>}
 C_lip_upper_mid {parent=faceAttach, loc=<0.7082 9.2524 -1.9189>, rot=<0.2219 0.01 0.0021 -0.975>}
 C_nose_bridge {parent=faceAttach, loc=<0.6383 10.3853 1.0265>, rot=<-0.2218 0 0 -0.9751>}
 L_brow_mid {parent=faceAttach, loc=<3.6666 8.5785 5.1184>, rot=<0.0001 -0.0035 0.2553 -0.9669>}
 L_brow_outer {parent=faceAttach, loc=<6.0236 5.6128 5.0398>, rot=<-0.0287 -0.04 0.4489 -0.8922>}
 L_cheek_inner {parent=faceAttach, loc=<4.4851 7.4548 -0.1636>, rot=<-0.0706 -0.2697 0.3184 -0.906>}
 L_eye {parent=faceAttach, loc=<3.4077 5.9685 3.6454>, rot=<0.0003 -0.0233 0.9996 -0.0137>}
 L_eye_lid_lower_mid {parent=L_eye, loc=<-0.0095 -0.0011 0.0014>, rot=<0.0075 -0.0069 0.0314 -0.9995>}
 L_eye_lid_upper_mid {parent=L_eye, loc=<-0.0037 -0.0052 0.0007>, rot=<0.0556 0.0057 0.0306 -0.998>}
 L_lip_corner {parent=faceAttach, loc=<2.9869 7.7473 -1.0531>, rot=<0.1907 -0.0457 -0.0534 -0.9791>}
 L_lip_upper_outer {parent=faceAttach, loc=<2.1624 8.5958 -1.8206>, rot=<0.1253 0.0436 0.3515 -0.9268>}
 R_brow_mid {parent=faceAttach, loc=<-2.3221 8.6142 5.1407>, rot=<-0.0001 0.0042 -0.2553 -0.9669>}
 R_brow_outer {parent=faceAttach, loc=<-4.6664 5.8108 5.2536>, rot=<-0.0287 0.04 -0.4491 -0.8921>}
 R_cheek_inner {parent=faceAttach, loc=<-3.0796 7.371 -0.1833>, rot=<-0.0706 0.2697 -0.3184 -0.906>}
 R_eye {parent=faceAttach, loc=<-2.1167 5.9794 3.6462>, rot=<-0.0003 -0.0233 0.9996 0.0136>}
 R_eye_lid_lower_mid {parent=R_eye, loc=<0.0067 -0.0103 0.0007>, rot=<0.007 0.0072 -0.027 -0.9996>}
 R_eye_lid_upper_mid {parent=R_eye, loc=<0.0023 -0.0047 0.0007>, rot=<0.0549 -0.0045 -0.0231 -0.9982>}
 R_lip_corner {parent=faceAttach, loc=<-1.5322 7.8883 -1.4753>, rot=<0.1184 0.0249 0.0385 -0.9919>}
 R_lip_upper_outer {parent=faceAttach, loc=<-0.7041 8.6939 -1.9399>, rot=<0.091 0.0016 -0.3182 -0.9436>}
 teeth_upper {parent=faceAttach, loc=<0.6479 7.7495 -1.2669>, rot=<0 0 0 -1>}
 dyn_mstr_side_l {parent=head, loc=<6.0897 1.95 -6.8544>, rot=<0.0029 -0.7071 -0.0029 -0.7071>}
 dyn_side_l_01 {parent=dyn_mstr_side_l, loc=<0 0 0>, rot=<-0.0972 -0.7055 -0.3184 -0.6256>}
 dyn_side_l_02 {parent=dyn_side_l_01, loc=<3.4132 0 0>, rot=<0 0 0 1>}
 dyn_mstr_hair_l {parent=head, loc=<1.5372 -4.3463 -4.1911>, rot=<0.0029 -0.7071 -0.0029 -0.7071>}
 dyn_hair_l_01 {parent=dyn_mstr_hair_l, loc=<0 0 0>, rot=<-0.1263 -0.6776 0.1625 -0.7061>}
 dyn_hair_l_02 {parent=dyn_hair_l_01, loc=<4.4007 0 0>, rot=<-0.001 0.0177 0.0058 -0.9998>}
 dyn_hair_l_03 {parent=dyn_hair_l_02, loc=<4.3608 0 0>, rot=<0 0 0 1>}
 dyn_hair_l_02_01 {parent=dyn_hair_l_02, loc=<2.232 1.9216 0.027>, rot=<0.0454 0.0475 -0.2499 -0.966>}
 dyn_hair_l_01_01 {parent=dyn_hair_l_01, loc=<3.3947 3.2628 0.9539>, rot=<0.3144 0.0502 -0.3325 -0.8877>}
 dyn_mstr_hair_b {parent=head, loc=<2.5189 -5.8751 0>, rot=<0.0029 -0.7071 -0.0029 -0.7071>}
 dyn_hair_b_01 {parent=dyn_mstr_hair_b, loc=<0 0 0>, rot=<-0.4835 -0.4835 0.516 -0.516>}
 dyn_hair_b_02 {parent=dyn_hair_b_01, loc=<3.9571 0 0.0227>, rot=<0 0.044 0 -0.999>}
 dyn_hair_b_03_01 {parent=dyn_hair_b_02, loc=<5.3501 0 0>, rot=<0 0 0 1>}
 dyn_hair_b_03_03 {parent=dyn_hair_b_02, loc=<5.7418 1.8571 0.2329>, rot=<0.1541 0.1772 -0.1211 -0.9645>}
 dyn_hair_b_03_02 {parent=dyn_hair_b_02, loc=<5.7418 -1.8571 0.2329>, rot=<-0.1541 0.1772 0.1211 -0.9645>}
 dyn_mstr_side_r {parent=head, loc=<6.0899 1.95 6.8544>, rot=<0.7071 0.0029 -0.7071 0.0029>}
 dyn_side_r_01 {parent=dyn_mstr_side_r, loc=<0 0 0>, rot=<-0.0972 -0.7055 -0.3184 -0.6256>}
 dyn_side_r_02 {parent=dyn_side_r_01, loc=<-3.4133 0.0001 0>, rot=<0 0 0 -1>}
 dyn_mstr_hair_r {parent=head, loc=<1.4688 -3.9841 3.5306>, rot=<0.7071 0.0029 -0.7071 0.0029>}
 dyn_hair_r_01 {parent=dyn_mstr_hair_r, loc=<0 0 0>, rot=<-0.1773 -0.6517 0.2143 -0.7057>}
 dyn_hair_r_02 {parent=dyn_hair_r_01, loc=<-4.3999 0 -0.0001>, rot=<-0.001 0.0177 0.0058 -0.9998>}
 dyn_hair_r_03 {parent=dyn_hair_r_02, loc=<-4.3608 0 0>, rot=<0 0 0 1>}
 dyn_hair_r_02_01 {parent=dyn_hair_r_02, loc=<-2.2325 -1.9216 -0.027>, rot=<0.0454 0.0475 -0.2499 -0.966>}
 dyn_hair_r_01_01 {parent=dyn_hair_r_01, loc=<-3.3946 -3.2627 -0.954>, rot=<0.3144 0.0502 -0.3325 -0.8877>}
 hairBun_lt_MSTR {parent=head, loc=<12.9832 -2.0533 -6>, rot=<0.9486 -0.0341 -0.2892 -0.1237>}
 hairBun_lt_00 {parent=hairBun_lt_MSTR, loc=<0 0 0>, rot=<0 0 0 1>}
 hairSidePiece_rt_MSTR {parent=head, loc=<5.2501 4.5527 7.2567>, rot=<-0.0567 0.0907 0.9941 0.0169>}
 hairSidePiece_rt_00 {parent=hairSidePiece_rt_MSTR, loc=<0 0 0>, rot=<-0.0374 0 0 -0.9993>}
 hairSidePiece_rt_01 {parent=hairSidePiece_rt_00, loc=<4.2832 0 0>, rot=<0.0072 -0.0412 0.0987 -0.9942>}
 hairSidePiece_rt_02 {parent=hairSidePiece_rt_01, loc=<5.1232 0 0>, rot=<0.0374 -0.0395 0.2272 -0.9723>}
 hairSidePiece_lt_MSTR {parent=head, loc=<10.6936 5.347 -4.3802>, rot=<0.107 -0.9586 -0.1837 -0.1893>}
 hairSidePiece_lt_00 {parent=hairSidePiece_lt_MSTR, loc=<0 0 0>, rot=<0.0404 0 0 -0.9992>}
 hairSidePiece_lt_01 {parent=hairSidePiece_lt_00, loc=<6.4886 0 0>, rot=<-0.0432 -0.181 -0.1357 -0.9731>}
 hairSidePiece_lt_02 {parent=hairSidePiece_lt_01, loc=<4.1188 0 0>, rot=<0 0 0 1>}
 hairBun_rt_MSTR {parent=head, loc=<12.9832 -2.0533 6>, rot=<-0.0012 0.2912 -0.004 -0.9567>}
 hairBun_rt_00 {parent=hairBun_rt_MSTR, loc=<0 0 0>, rot=<0 0 0 1>}
 hairBang_MSTR {parent=head, loc=<8.4212 10.7349 0.3193>, rot=<0.0475 -0.9989 0 0>}
 hairBang_00 {parent=hairBang_MSTR, loc=<0 0 0>, rot=<0 0 0 1>}
 hairBack_rt_MSTR {parent=head, loc=<4.8143 -6.009 4.0638>, rot=<-0.0884 -0.4361 0.8946 0.0403>}
 hairBack_rt_00 {parent=hairBack_rt_MSTR, loc=<0 0 0>, rot=<0 0 0 1>}
 hairBack_rt_01 {parent=hairBack_rt_00, loc=<5.5821 0 0>, rot=<0 0.2345 0 -0.9721>}
 hairBack_rt_02 {parent=hairBack_rt_01, loc=<5.3074 1.6206 0>, rot=<0 0 0 1>}
 hairBack_rt_03 {parent=hairBack_rt_01, loc=<4.9287 -3.6597 -1.0513>, rot=<-0.1291 -0.0074 0.0568 -0.99>}
 hairBack_lt_MSTR {parent=head, loc=<4.8146 -6.009 -4.0638>, rot=<-0.0403 -0.8946 0.4361 0.0884>}
 hairBack_lt_00 {parent=hairBack_lt_MSTR, loc=<0 0 0>, rot=<0 0 0 1>}
 hairBack_lt_01 {parent=hairBack_lt_00, loc=<5.5824 0 0.0001>, rot=<0 0.2345 0 -0.9721>}
 hairBack_lt_02 {parent=hairBack_lt_01, loc=<5.3077 -1.6206 -0.0001>, rot=<0 0 0 1>}
 hairBack_lt_03 {parent=hairBack_lt_01, loc=<4.9289 3.6597 -1.0513>, rot=<0.1291 -0.0074 -0.0568 -0.99>}
 dyn_hair_r_main_MSTR {parent=head, loc=<10.8876 -6.2822 6.8275>, rot=<-0.2933 0.6507 0.2825 -0.6409>}
 dyn_hair_r_main_Look_at {parent=dyn_hair_r_main_MSTR, loc=<0 0 0>, rot=<-0.31 0.6059 -0.2621 -0.6842>}
 dyn_hair_r_main_01 {parent=dyn_hair_r_main_Look_at, loc=<0 0 0>, rot=<0.31 -0.6059 0.2621 -0.6842>}
 dyn_hair_r_main_02 {parent=dyn_hair_r_main_01, loc=<2.3383 0 0>, rot=<-0.3535 0.5038 -0.1773 -0.768>}
 dyn_hair_r_main_03 {parent=dyn_hair_r_main_02, loc=<8.6686 0 0>, rot=<-0.0254 0.1014 -0.1385 -0.9848>}
 dyn_hair_r_main_04 {parent=dyn_hair_r_main_03, loc=<11.8196 0 0>, rot=<-0.0009 0.0115 -0.0268 -0.9996>}
 dyn_hair_r_main_05 {parent=dyn_hair_r_main_04, loc=<11.3236 0 0>, rot=<-0.0005 0.0386 -0.0385 -0.9985>}
 dyn_hair_r_main_06 {parent=dyn_hair_r_main_05, loc=<8.375 0 0>, rot=<-0.0002 -0.0221 0.0138 -0.9997>}
 dyn_hair_r_a_01 {parent=dyn_hair_r_main_Look_at, loc=<-1.0651 -3.5458 -2.1317>, rot=<-0.2515 -0.0375 0.1382 -0.9572>}
 dyn_hair_r_a_02 {parent=dyn_hair_r_a_01, loc=<9.9656 0 0>, rot=<0.0869 0.0376 -0.1484 -0.9844>}
 dyn_hair_r_a_03 {parent=dyn_hair_r_a_02, loc=<12.0945 0 0>, rot=<-0.0077 0.0073 -0.0426 -0.999>}
 dyn_hair_r_a_04 {parent=dyn_hair_r_a_03, loc=<11.324 0 0>, rot=<-0.001 0.0044 -0.0164 -0.9999>}
 dyn_hair_r_a_05 {parent=dyn_hair_r_a_04, loc=<7.9003 0 0>, rot=<0.015 -0.0046 -0.0029 -0.9999>}
 dyn_hair_r_c_01 {parent=dyn_hair_r_main_Look_at, loc=<0.0239 -1.1565 -2.8535>, rot=<0.8568 -0.0435 -0.1225 0.499>}
 dyn_hair_r_c_02 {parent=dyn_hair_r_c_01, loc=<8.8913 0 0>, rot=<-0.102 -0.1558 -0.0274 -0.9821>}
 dyn_hair_r_c_03 {parent=dyn_hair_r_c_02, loc=<11.646 0 0>, rot=<0.0747 -0.0232 0.0016 -0.9969>}
 dyn_hair_r_c_04 {parent=dyn_hair_r_c_03, loc=<11.3231 0 0>, rot=<-0.0487 -0.0185 0.0091 -0.9986>}
 dyn_hair_r_c_05 {parent=dyn_hair_r_c_04, loc=<8.6458 0 0>, rot=<0.0334 0.0117 0.0475 -0.9982>}
 dyn_hair_r_b_01 {parent=dyn_hair_r_main_Look_at, loc=<-1.2392 -2.5202 -3.3905>, rot=<-0.4814 -0.0058 0.1212 -0.8681>}
 dyn_hair_r_b_02 {parent=dyn_hair_r_b_01, loc=<10.1253 0 0>, rot=<0.0533 -0.0414 -0.1252 -0.9898>}
 dyn_hair_r_b_03 {parent=dyn_hair_r_b_02, loc=<11.8964 0 0>, rot=<-0.0931 -0.0124 -0.0394 -0.9948>}
 dyn_hair_r_b_04 {parent=dyn_hair_r_b_03, loc=<11.3203 0 0>, rot=<0.0243 -0.0087 0.0008 -0.9997>}
 dyn_hair_r_b_05 {parent=dyn_hair_r_b_04, loc=<9.321 0 0>, rot=<-0.024 -0.0264 0.0099 -0.9993>}
 dyn_hair_r_f_01 {parent=dyn_hair_r_main_Look_at, loc=<0.0004 -3.3593 -0.6329>, rot=<0.5046 -0.1081 0.0928 -0.8515>}
 dyn_hair_r_f_02 {parent=dyn_hair_r_f_01, loc=<8.7168 0 0>, rot=<-0.1722 0.1439 -0.0261 -0.9741>}
 dyn_hair_r_f_03 {parent=dyn_hair_r_f_02, loc=<12.0785 0 0>, rot=<0.0322 0.0438 -0.0168 -0.9984>}
 dyn_hair_r_f_04 {parent=dyn_hair_r_f_03, loc=<11.3207 0 0>, rot=<-0.1231 0.0191 -0.0165 -0.9921>}
 dyn_hair_r_f_05 {parent=dyn_hair_r_f_04, loc=<8.3704 0 0>, rot=<0.0602 -0.0156 0.031 -0.9976>}
 dyn_hair_r_e_01 {parent=dyn_hair_r_main_Look_at, loc=<1.008 -1.779 -0.7462>, rot=<-0.466 0.0506 0.1217 -0.8749>}
 dyn_hair_r_e_02 {parent=dyn_hair_r_e_01, loc=<7.6791 0 0>, rot=<-0.1191 -0.0988 -0.0908 -0.9838>}
 dyn_hair_r_e_03 {parent=dyn_hair_r_e_02, loc=<11.7781 0 0>, rot=<-0.0444 -0.0314 -0.0322 -0.998>}
 dyn_hair_r_e_04 {parent=dyn_hair_r_e_03, loc=<11.3197 0 0>, rot=<-0.0537 -0.0099 -0.0306 -0.998>}
 dyn_hair_r_e_05 {parent=dyn_hair_r_e_04, loc=<7.7211 0 0>, rot=<-0.1343 -0.0089 0.0018 -0.9909>}
 dyn_hair_r_d_01 {parent=dyn_hair_r_main_Look_at, loc=<1.8218 -1.4854 -1.916>, rot=<-0.3817 -0.0085 0.0598 -0.9223>}
 dyn_hair_r_d_02 {parent=dyn_hair_r_d_01, loc=<6.7643 0 0>, rot=<0.0303 0.0086 -0.0793 -0.9964>}
 dyn_hair_r_d_03 {parent=dyn_hair_r_d_02, loc=<11.5638 0 0>, rot=<0.1375 -0.0096 -0.0347 -0.9899>}
 dyn_hair_r_d_04 {parent=dyn_hair_r_d_03, loc=<11.3203 0 0>, rot=<0.0504 0.0013 -0.0191 -0.9985>}
 dyn_hair_r_d_05 {parent=dyn_hair_r_d_04, loc=<8.1824 0 0>, rot=<-0.0094 -0.0213 -0.0192 -0.9995>}
 dyn_hair_l_main_MSTR {parent=head, loc=<10.8876 -6.2822 -6.8275>, rot=<0.2933 -0.6507 0.2825 -0.6409>}
 dyn_hair_l_main_Look_at {parent=dyn_hair_l_main_MSTR, loc=<0 0 0>, rot=<0.31 -0.6059 -0.2621 -0.6842>}
 dyn_hair_l_main_01 {parent=dyn_hair_l_main_Look_at, loc=<0 0 0>, rot=<-0.31 0.6059 0.2621 -0.6842>}
 dyn_hair_l_main_02 {parent=dyn_hair_l_main_01, loc=<2.3383 0 0>, rot=<0.3535 -0.5038 -0.1773 -0.768>}
 dyn_hair_l_main_03 {parent=dyn_hair_l_main_02, loc=<8.6686 0 0>, rot=<0.0254 -0.1014 -0.1385 -0.9848>}
 dyn_hair_l_main_04 {parent=dyn_hair_l_main_03, loc=<11.8196 0 0>, rot=<0.0009 -0.0115 -0.0268 -0.9996>}
 dyn_hair_l_main_05 {parent=dyn_hair_l_main_04, loc=<11.3236 0 0>, rot=<0.0005 -0.0386 -0.0385 -0.9985>}
 dyn_hair_l_main_06 {parent=dyn_hair_l_main_05, loc=<8.375 0 0>, rot=<0.0002 0.0221 0.0138 -0.9997>}
 dyn_hair_l_c_01 {parent=dyn_hair_l_main_Look_at, loc=<-0.3078 -1.1611 2.7361>, rot=<0.9177 -0.0551 0.1169 -0.3758>}
 dyn_hair_l_c_02 {parent=dyn_hair_l_c_01, loc=<9.2309 0 0>, rot=<-0.0597 0.1572 -0.0146 -0.9857>}
 dyn_hair_l_c_03 {parent=dyn_hair_l_c_02, loc=<11.646 0 0>, rot=<0.0143 0.0232 0.0025 -0.9996>}
 dyn_hair_l_c_04 {parent=dyn_hair_l_c_03, loc=<11.3231 0 0>, rot=<0.0125 0.0176 0.0108 -0.9997>}
 dyn_hair_l_c_05 {parent=dyn_hair_l_c_04, loc=<8.6458 0 0>, rot=<0.0875 0.0006 0.0222 -0.9959>}
 dyn_hair_l_b_01 {parent=dyn_hair_l_main_Look_at, loc=<-1.6515 -1.8041 2.5061>, rot=<0.4899 0.0232 0.1637 -0.8559>}
 dyn_hair_l_b_02 {parent=dyn_hair_l_b_01, loc=<10.8319 0 0>, rot=<0.0566 0.0654 -0.1642 -0.9826>}
 dyn_hair_l_b_03 {parent=dyn_hair_l_b_02, loc=<11.8964 0 0>, rot=<-0.0067 0.0181 -0.0371 -0.9991>}
 dyn_hair_l_b_04 {parent=dyn_hair_l_b_03, loc=<11.3203 0 0>, rot=<-0.0237 0.0087 0.0012 -0.9997>}
 dyn_hair_l_b_05 {parent=dyn_hair_l_b_04, loc=<9.321 0 0>, rot=<-0.0046 0.0165 0.0091 -0.9998>}
 dyn_hair_l_a_01 {parent=dyn_hair_l_main_Look_at, loc=<-1.2191 -2.4902 1.5516>, rot=<0.2597 0.0479 0.1894 -0.9457>}
 dyn_hair_l_a_02 {parent=dyn_hair_l_a_01, loc=<10.5135 0 0>, rot=<-0.0001 -0.0194 -0.2037 -0.9788>}
 dyn_hair_l_a_03 {parent=dyn_hair_l_a_02, loc=<12.0945 0 0>, rot=<-0.0527 -0.0015 -0.0432 -0.9977>}
 dyn_hair_l_a_04 {parent=dyn_hair_l_a_03, loc=<11.324 0 0>, rot=<-0.0634 -0.0042 -0.0164 -0.9978>}
 dyn_hair_l_a_05 {parent=dyn_hair_l_a_04, loc=<7.9003 0 0>, rot=<0.0188 0.0039 0.0067 -0.9998>}
 dyn_hair_l_d_01 {parent=dyn_hair_l_main_Look_at, loc=<1.8218 -1.4854 1.916>, rot=<0.3465 0.0107 0.0595 -0.9361>}
 dyn_hair_l_d_02 {parent=dyn_hair_l_d_01, loc=<6.7643 0 0>, rot=<-0.0275 -0.0144 -0.0785 -0.9964>}
 dyn_hair_l_d_03 {parent=dyn_hair_l_d_02, loc=<11.5638 0 0>, rot=<-0.1173 0.0078 -0.0351 -0.9924>}
 dyn_hair_l_d_04 {parent=dyn_hair_l_d_03, loc=<11.3203 0 0>, rot=<-0.0369 -0.0016 -0.019 -0.9991>}
 dyn_hair_l_d_05 {parent=dyn_hair_l_d_04, loc=<8.1824 0 0>, rot=<0.0036 0.0082 -0.0252 -0.9996>}
 dyn_hair_l_e_01 {parent=dyn_hair_l_main_Look_at, loc=<1.9248 -2.3883 0.8326>, rot=<0.5304 -0.0494 0.0931 -0.8412>}
 dyn_hair_l_e_02 {parent=dyn_hair_l_e_01, loc=<6.643 0 0>, rot=<0.0979 0.0864 -0.0646 -0.9893>}
 dyn_hair_l_e_03 {parent=dyn_hair_l_e_02, loc=<11.7781 0 0>, rot=<0.0521 0.0347 -0.0285 -0.9976>}
 dyn_hair_l_e_04 {parent=dyn_hair_l_e_03, loc=<11.3197 0 0>, rot=<0.0211 0.0125 -0.0297 -0.9993>}
 dyn_hair_l_e_05 {parent=dyn_hair_l_e_04, loc=<7.7211 0 0>, rot=<0.0825 0.0144 0.0033 -0.9965>}
 dyn_hair_l_f_01 {parent=dyn_hair_l_main_Look_at, loc=<-0.3722 -3.2144 0.668>, rot=<-0.4218 0.0962 0.1062 -0.8953>}
 dyn_hair_l_f_02 {parent=dyn_hair_l_f_01, loc=<9.1101 0 0>, rot=<0.0635 -0.1403 -0.0417 -0.9872>}
 dyn_hair_l_f_03 {parent=dyn_hair_l_f_02, loc=<12.0785 0 0>, rot=<0.0226 -0.0434 -0.0178 -0.9986>}
 dyn_hair_l_f_04 {parent=dyn_hair_l_f_03, loc=<11.3207 0 0>, rot=<0.0409 -0.0192 -0.0164 -0.9988>}
 dyn_hair_l_f_05 {parent=dyn_hair_l_f_04, loc=<8.3704 0 0>, rot=<-0.0092 0.0129 0.0103 -0.9998>}
 dyn_ear_l {parent=head, loc=<-0.3737 -0.0742 -6.882>, rot=<-0.0029 0.7071 0.7071 -0.0029>}
 dyn_ear_l_01 {parent=dyn_ear_l, loc=<2.5565 -0.0028 -0.0029>, rot=<0 0 0 1>}
 dyn_ear_l_02 {parent=dyn_ear_l_01, loc=<6.2854 0.0157 0>, rot=<0 0 0 1>}
 dyn_ear_r {parent=head, loc=<-0.3737 -0.0742 6.882>, rot=<0.0029 -0.7071 0.7071 -0.0029>}
 dyn_ear_r_01 {parent=dyn_ear_r, loc=<2.5565 -0.0028 0.0029>, rot=<0 0 0 1>}
 dyn_ear_r_02 {parent=dyn_ear_r_01, loc=<6.2854 0.0157 0>, rot=<0 0 0 1>}
 dyn_front_bang_MSTR {parent=head, loc=<8.6129 7.9971 0>, rot=<0 0 1 0>}
 dyn_front_bang_01 {parent=dyn_front_bang_MSTR, loc=<0 0 0>, rot=<0 0 0 1>}
 dyn_front_bang_02 {parent=dyn_front_bang_01, loc=<3.1755 -0.0213 0>, rot=<0 0 0 1>}
 dyn_bang_right_MSTR {parent=head, loc=<8.4619 4.6816 5.5544>, rot=<-0.2111 0.2111 -0.6749 -0.6749>}
 dyn_bang_right_01 {parent=dyn_bang_right_MSTR, loc=<0 0 0>, rot=<0 0 0 1>}
 dyn_bang_left_MSTR {parent=head, loc=<8.4184 5.2897 -4.5318>, rot=<0.3064 -0.3064 -0.6373 -0.6373>}
 dyn_bang_left_01 {parent=dyn_bang_left_MSTR, loc=<0 0 0>, rot=<0 0 0 1>}
 dyn_hair_right_MSTR {parent=head, loc=<4.6066 1.347 7.666>, rot=<-0.0108 0.0974 0.99 0.1013>}
 dyn_hair_right_01 {parent=dyn_hair_right_MSTR, loc=<0 0 0>, rot=<0 0 0 1>}
 dyn_hair_right_02 {parent=dyn_hair_right_01, loc=<3.9921 -0.0275 -0.5071>, rot=<-0.0013 -0.0886 0.0924 -0.9918>}
 dyn_hair_left_MSTR {parent=head, loc=<5.4655 1.675 -7.3793>, rot=<-0.1342 -0.987 -0.0829 0.0289>}
 dyn_hair_left_01 {parent=dyn_hair_left_MSTR, loc=<0 0 0>, rot=<0 0 0 1>}
 dyn_hair_left_02 {parent=dyn_hair_left_01, loc=<4.7102 0 0>, rot=<0 0 0 1>}
 dyn_top_hair_MSTR {parent=head, loc=<8.3188 0.3408 0>, rot=<0 0 0 1>}
 dyn_top_hair_01 {parent=dyn_top_hair_MSTR, loc=<3.2486 8.4129 -0.3791>, rot=<0 0 -0.6419 -0.7668>}
 dyn_top_hair_02 {parent=dyn_top_hair_MSTR, loc=<6.9329 3.5086 0.9461>, rot=<0.0444 -0.0428 -0.5558 -0.829>}
 dyn_top_hair_03 {parent=dyn_top_hair_MSTR, loc=<7.153 -2.2765 0.3462>, rot=<0.0482 0.0339 -0.1746 -0.9829>}
 dyn_top_hair_04 {parent=dyn_top_hair_MSTR, loc=<4.9274 -4.5167 -5.2537>, rot=<0.7339 -0.5494 -0.2503 -0.3112>}
 dyn_back_bottom_MSTR {parent=head, loc=<-3.5744 -5.6461 -0.4202>, rot=<0.0022 -0.0132 0.9863 -0.1647>}
 dyn_back_bottom_01 {parent=dyn_back_bottom_MSTR, loc=<0 0 0>, rot=<0 0 0 1>}
 dyn_back_hair_01_MSTR {parent=head, loc=<1.1097 -7.7684 0.4742>, rot=<0.1138 0.1984 0.7608 -0.6073>}
 dyn_back_hair_01_01 {parent=dyn_back_hair_01_MSTR, loc=<0 0 0>, rot=<0 0 0 1>}
 dyn_back_hair_02_MSTR {parent=head, loc=<-0.2846 -5.8777 3.6907>, rot=<0.1399 0.1757 0.7242 -0.652>}
 dyn_back_hair_02_01 {parent=dyn_back_hair_02_MSTR, loc=<0 0 0>, rot=<0 0 0 1>}
 dyn_right_hair_02_MSTR {parent=head, loc=<2.531 -4.9226 6.9831>, rot=<0.4542 0.4532 0.5812 -0.5006>}
 dyn_right_hair_02_01 {parent=dyn_right_hair_02_MSTR, loc=<0 0 0>, rot=<0 0 0 1>}
 dyn_right_hair_03_MSTR {parent=head, loc=<5.9843 -2.7437 7.6348>, rot=<0.4931 0.4865 0.5486 -0.4682>}
 dyn_right_hair_03_01 {parent=dyn_right_hair_03_MSTR, loc=<0 0 0>, rot=<0 0 0 1>}
 dyn_right_hair_04_MSTR {parent=head, loc=<4.8173 -8.0654 5.0288>, rot=<0.2344 0.3437 0.7643 -0.4927>}
 dyn_right_hair_04_01 {parent=dyn_right_hair_04_MSTR, loc=<0 0 0>, rot=<0 0 0 1>}
 dyn_left_left_01_MSTR {parent=head, loc=<2.7373 -7.8514 -4.0987>, rot=<-0.1128 -0.1282 0.7907 -0.5879>}
 dyn_left_hair_01 {parent=dyn_left_left_01_MSTR, loc=<0 0 0>, rot=<0 0 0 1>}
 dyn_left_hair_02_MSTR {parent=head, loc=<1.1546 -3.5485 -6.297>, rot=<-0.3858 -0.3968 0.7272 -0.4061>}
 dyn_left_hair_02_01 {parent=dyn_left_hair_02_MSTR, loc=<0 0 0>, rot=<0 0 0 1>}
 dyn_left_hair_03_MSTR {parent=head, loc=<6.6655 -2.4292 -8.5738>, rot=<-0.488 -0.3759 0.6263 -0.4777>}
 dyn_left_hair_03_01 {parent=dyn_left_hair_03_MSTR, loc=<0 0 0>, rot=<0 0 0 1>}
 dyn_hairFront_MSTR {parent=head, loc=<13.6747 7.3282 5.7657>, rot=<0.4146 0.0612 0.0014 -0.9079>}
 dyn_hairFront_00 {parent=dyn_hairFront_MSTR, loc=<0 0 0>, rot=<0 0 0 1>}
 dyn_hairFront_01_MSTR {parent=head, loc=<15.3659 8.4923 4.876>, rot=<0.3272 0.0413 0.0918 -0.9396>}
 dyn_hairFront_01 {parent=dyn_hairFront_01_MSTR, loc=<0 0 0>, rot=<0 0 0 1>}
 dyn_hairFront_02_MSTR {parent=head, loc=<18.4076 6.7759 2.3325>, rot=<0.5138 -0.3088 -0.044 -0.7992>}
 dyn_hairFront_02 {parent=dyn_hairFront_02_MSTR, loc=<0 0 0>, rot=<0 0 0 1>}
 dyn_hairFront_03_MSTR {parent=head, loc=<17.5793 4.9473 -2.4896>, rot=<0.7416 -0.5337 -0.0419 -0.4043>}
 dyn_hairFront_03 {parent=dyn_hairFront_03_MSTR, loc=<0 0 0>, rot=<0 0 0 1>}
 dyn_hairFront_03_Grape {parent=dyn_hairFront_03_MSTR, loc=<0 0 0>, rot=<0 0 0 1>}
 dyn_hairFront_04_MSTR {parent=head, loc=<14.6727 4.1519 -5.4776>, rot=<0.7794 -0.5183 -0.0821 -0.3423>}
 dyn_hairFront_04 {parent=dyn_hairFront_04_MSTR, loc=<0 0 0>, rot=<0 0 0 1>}
 dyn_hairFront_04_Grape {parent=dyn_hairFront_04_MSTR, loc=<0 0 0>, rot=<0 0 0 1>}
 dyn_hairMidFront_MSTR {parent=head, loc=<17.6318 2.4365 6.0245>, rot=<0.2175 0.1062 0.4849 -0.8404>}
 dyn_hairMidFront_00 {parent=dyn_hairMidFront_MSTR, loc=<0 0 0>, rot=<0 0 0 1>}
 dyn_hairMidFront_01_MSTR {parent=head, loc=<19.7263 1.114 0.2953>, rot=<-0.1867 -0.1834 0.3714 -0.8908>}
 dyn_hairMidFront_01 {parent=dyn_hairMidFront_01_MSTR, loc=<0 0 0>, rot=<0 0 0 1>}
 dyn_hairMidFront_02_MSTR {parent=head, loc=<18.5347 -0.2262 -1.9901>, rot=<0.0035 -0.5037 0.4093 -0.7607>}
 dyn_hairMidFront_02 {parent=dyn_hairMidFront_02_MSTR, loc=<0 0 0>, rot=<0 0 0 1>}
 dyn_hairMid_MSTR {parent=head, loc=<11.9024 -2.6703 6.3162>, rot=<0.1213 -0.1691 0.6178 -0.7583>}
 dyn_hairMid_00 {parent=dyn_hairMid_MSTR, loc=<0 0 0>, rot=<0 0 0 1>}
 dyn_hairMid_01_MSTR {parent=head, loc=<16.282 -4.3033 5.2302>, rot=<-0.2182 -0.2382 0.4143 -0.8509>}
 dyn_hairMid_01 {parent=dyn_hairMid_01_MSTR, loc=<0 0 0>, rot=<0 0 0 1>}
 dyn_hairMid_02_MSTR {parent=head, loc=<18.2081 -2.9373 3.6098>, rot=<-0.0977 -0.2269 0.2337 -0.9404>}
 dyn_hairMid_02 {parent=dyn_hairMid_02_MSTR, loc=<0 0 0>, rot=<0 0 0 1>}
 dyn_hairMid_03_MSTR {parent=head, loc=<17.8277 -4.4044 -0.0483>, rot=<0.0583 -0.5785 0.3936 -0.712>}
 dyn_hairMid_03 {parent=dyn_hairMid_03_MSTR, loc=<0 0 0>, rot=<0 0 0 1>}
 dyn_hairMid_04_MSTR {parent=head, loc=<16.4545 -3.2541 -3.0598>, rot=<-0.0693 -0.6308 0.4034 -0.6592>}
 dyn_hairMid_04 {parent=dyn_hairMid_04_MSTR, loc=<0 0 0>, rot=<0 0 0 1>}
 dyn_hairMid_05_MSTR {parent=head, loc=<15.6613 -2.5568 -5.31>, rot=<0.606 -0.7534 -0.0807 -0.242>}
 dyn_hairMid_05 {parent=dyn_hairMid_05_MSTR, loc=<0 0 0>, rot=<0 0 0 1>}
 dyn_hairMidBack_MSTR {parent=head, loc=<13.6035 -7.1663 1.0308>, rot=<-0.1699 -0.296 0.692 -0.6361>}
 dyn_hairMidBack_00 {parent=dyn_hairMidBack_MSTR, loc=<0 0 0>, rot=<0 0 0 1>}
 dyn_hairMidBack_01_MSTR {parent=head, loc=<13.7913 -6.0321 -1.9405>, rot=<-0.0077 -0.757 0.4257 -0.4956>}
 dyn_hairMidBack_01 {parent=dyn_hairMidBack_01_MSTR, loc=<0 0 0>, rot=<0 0 0 1>}
 dyn_hairMidBack_02_MSTR {parent=head, loc=<11.2457 -5.4453 -4.3286>, rot=<0.2146 -0.745 0.1347 -0.6171>}
 dyn_hairMidBack_02 {parent=dyn_hairMidBack_02_MSTR, loc=<0 0 0>, rot=<0 0 0 1>}
 dyn_hairMidBack_03_MSTR {parent=head, loc=<11.7684 -2.7957 -7.0632>, rot=<0.4656 -0.8554 -0.1143 -0.1962>}
 dyn_hairMidBack_03 {parent=dyn_hairMidBack_03_MSTR, loc=<0 0 0>, rot=<0 0 0 1>}
 dyn_hairBack_02_MSTR {parent=head, loc=<8.8854 -5.3204 -5.555>, rot=<0.2419 -0.9301 0.2432 -0.1314>}
 dyn_hairBack_02 {parent=dyn_hairBack_02_MSTR, loc=<0 0 0>, rot=<0 0 0 1>}
 dyn_hairBack_01_MSTR {parent=head, loc=<8.105 -5.771 -3.6886>, rot=<0.0704 -0.7448 0.4398 -0.4968>}
 dyn_hairBack_01 {parent=dyn_hairBack_01_MSTR, loc=<0 0 0>, rot=<0 0 0 1>}
 dyn_hairBack_00_MSTR {parent=head, loc=<8.44 -7.5219 -1.3198>, rot=<-0.1038 -0.8204 0.406 -0.3891>}
 dyn_hairBack_00 {parent=dyn_hairBack_00_MSTR, loc=<0 0 0>, rot=<0 0 0 1>}
 dyn_earringTop_lt_MSTR {parent=head, loc=<4.2288 -1.0249 -7.5061>, rot=<0.7255 -0.2702 -0.3566 -0.5229>}
 dyn_earringTop_lt_00 {parent=dyn_earringTop_lt_MSTR, loc=<0 0 0>, rot=<0.6842 0 0 -0.7293>}
 dyn_earringBottom_lt_MSTR {parent=head, loc=<2.8642 -1.1887 -7.7793>, rot=<0.6575 -0.6193 -0.3032 -0.3038>}
 dyn_earringBottom_lt_00 {parent=dyn_earringBottom_lt_MSTR, loc=<0 0 0>, rot=<0.6842 0 0 -0.7293>}
 dyn_earringTop_rt_MSTR {parent=head, loc=<4.254 -0.9704 7.4821>, rot=<0.4956 0.3851 0.2922 -0.7216>}
 dyn_earringTop_rt_00 {parent=dyn_earringTop_rt_MSTR, loc=<0 0 0>, rot=<0 0 0 1>}
 dyn_hair_top_MSTR {parent=head, loc=<16.2008 3.2791 -0.2372>, rot=<0.6983 0.0008 -0.0052 -0.7158>}
 dyn_hair_top_01 {parent=dyn_hair_top_MSTR, loc=<0 0 0>, rot=<0 0 0 1>}
 dyn_back_right_MSTR {parent=head, loc=<9.9384 -2.9629 -5.5281>, rot=<0.5947 -0.3825 0.5915 -0.3874>}
 dyn_back_right_01 {parent=dyn_back_right_MSTR, loc=<0 0 0>, rot=<0 0 0 1>}
 dyn_back_right_02 {parent=dyn_back_right_01, loc=<3.3096 0 0>, rot=<0 0 0 1>}
 dyn_right_bang_MSTR {parent=head, loc=<11.5057 9.8421 -2.3442>, rot=<0.1126 -0.8309 0.2556 -0.4812>}
 dyn_right_bang_01 {parent=dyn_right_bang_MSTR, loc=<0 0 0>, rot=<0 0 0 1>}
 dyn_right_bang_02 {parent=dyn_right_bang_01, loc=<2.2715 0.0035 -0.4049>, rot=<0 0 0 1>}
 dyn_front_bump_MSTR {parent=head, loc=<14.2388 10.2324 0.5426>, rot=<0.6983 0.0008 -0.0052 -0.7158>}
 dyn_front_bump_01 {parent=dyn_front_bump_MSTR, loc=<0 0 0>, rot=<0 0 0 1>}
 dyn_bang_front_MSTR {parent=head, loc=<12.3991 11.6078 1.12>, rot=<0.022 -0.8382 0.202 -0.5061>}
 dyn_bang_front_01 {parent=dyn_bang_front_MSTR, loc=<0 0 0>, rot=<0 0 0 1>}
 dyn_bang_front_02 {parent=dyn_bang_front_01, loc=<3.3152 0.0163 -0.4049>, rot=<0 0 0 1>}
 dyn_top_bang_MSTR {parent=head, loc=<13.6446 3.0901 4.3887>, rot=<0.6915 0.0347 -0.2105 -0.6902>}
 dyn_top_bang_01 {parent=dyn_top_bang_MSTR, loc=<0 0 0>, rot=<0 0 0 1>}
 dyn_top_bang_02 {parent=dyn_top_bang_01, loc=<2.4742 0 0>, rot=<0 0 0 1>}
 dyn_back_left_MSTR {parent=head, loc=<11.2959 -2.1763 2.8742>, rot=<0.6058 -0.3647 0.6028 -0.3697>}
 dyn_back_left_01 {parent=dyn_back_left_MSTR, loc=<0 0 0>, rot=<0 0 0 1>}
 dyn_back_left_02 {parent=dyn_back_left_01, loc=<3.3096 0 0>, rot=<0 0 0 1>}
 dyn_hair_A {parent=head, loc=<11.0458 -5.3976 -2.146>, rot=<0.1912 -0.8467 0.4262 -0.2549>}
 dyn_hair_B {parent=head, loc=<13.0759 -3.8299 -1.3312>, rot=<0.2667 -0.7849 0.3835 -0.4071>}
 dyn_hair_C {parent=head, loc=<14.0616 5.0155 -0.3534>, rot=<0.7971 0.304 0.4936 0.1691>}
 dyn_hair_D {parent=head, loc=<14.7138 2.5354 -0.5008>, rot=<0.7981 -0.1808 0.4892 -0.3017>}
 dyn_hair_E {parent=head, loc=<12.5086 0.6797 -4.1298>, rot=<0.0048 -0.7662 0.0704 -0.6387>}
 dyn_hair_F {parent=head, loc=<12.6601 4.8584 -3.3292>, rot=<-0.0361 -0.7302 -0.2367 -0.6399>}
 dyn_hair_G {parent=head, loc=<12.0351 7.8791 -1.9963>, rot=<-0.3773 -0.5626 -0.5248 -0.5155>}
 dyn_hair_H {parent=head, loc=<10.8881 0.85 -5.7787>, rot=<0.0593 -0.9249 0.0946 -0.3634>}
 dyn_hair_I {parent=head, loc=<10.8715 5.6497 -4.6968>, rot=<0.0409 -0.8956 -0.2359 -0.3748>}
 dyn_hair_J {parent=head, loc=<10.2812 9.4289 -2.4652>, rot=<-0.191 -0.6972 -0.6175 -0.3101>}
 dyn_hair_K {parent=head, loc=<13.0757 -3.8299 1.3312>, rot=<0.4071 -0.3835 0.7849 -0.2667>}
 dyn_hair_L {parent=head, loc=<12.5086 0.6797 4.1298>, rot=<0.6387 -0.0704 0.7662 -0.0048>}
 dyn_hair_M {parent=head, loc=<12.6605 4.8584 3.3292>, rot=<0.6399 0.2367 0.7302 0.0361>}
 dyn_hair_N {parent=head, loc=<12.0352 7.8791 1.9963>, rot=<0.5155 0.5248 0.5626 0.3773>}
 dyn_hair_O {parent=head, loc=<11.0461 -5.3976 2.146>, rot=<0.2549 -0.4262 0.8467 -0.1912>}
 dyn_hair_P {parent=head, loc=<10.8882 0.85 5.7787>, rot=<0.3634 -0.0946 0.9249 -0.0593>}
 dyn_hair_Q {parent=head, loc=<10.8714 5.6497 4.6968>, rot=<0.3748 0.2359 0.8956 -0.0409>}
 dyn_hair_R {parent=head, loc=<10.2814 9.4289 2.4652>, rot=<0.3101 0.6175 0.6972 0.191>}
 dyn_hair_S {parent=head, loc=<10.5942 10.4902 0.1885>, rot=<0.3536 0.5936 0.7091 0.1404>}
 dyn_hair_front_00_MSTR {parent=head, loc=<8.3959 10.3526 -0.2484>, rot=<-0.0033 -0.9988 -0.0481 -0.005>}
 dyn_hair_front_00 {parent=dyn_hair_front_00_MSTR, loc=<0 0 0>, rot=<0 0 0 1>}
 dyn_hair_front_01_MSTR {parent=head, loc=<9.9511 10.9118 -1.3219>, rot=<0.9246 0.337 0.0968 -0.1492>}
 dyn_hair_front_01 {parent=dyn_hair_front_01_MSTR, loc=<0 0 0>, rot=<0 0 0 1>}
 dyn_hair_side_l_00_MSTR {parent=head, loc=<8.1613 9.9276 -2.9841>, rot=<0.2397 -0.8976 -0.3636 -0.0679>}
 dyn_hair_side_l_00 {parent=dyn_hair_side_l_00_MSTR, loc=<0 0 0>, rot=<0 0 0 1>}
 dyn_hair_side_l_01_MSTR {parent=head, loc=<7.9322 8.5402 -5.0629>, rot=<0.1486 -0.9551 -0.2561 -0.0121>}
 dyn_hair_side_l_01 {parent=dyn_hair_side_l_01_MSTR, loc=<0 0 0>, rot=<0 0 0 1>}
 dyn_hair_side_l_02_MSTR {parent=head, loc=<7.8785 5.8119 -6.7326>, rot=<0.2848 -0.9362 -0.1882 -0.0837>}
 dyn_hair_side_l_02 {parent=dyn_hair_side_l_02_MSTR, loc=<0 0 0>, rot=<0 0 0 1>}
 dyn_hair_side_l_03_MSTR {parent=head, loc=<5.5272 0.9838 -7.6201>, rot=<0.4499 -0.7971 -0.2499 -0.3159>}
 dyn_hair_side_l_03 {parent=dyn_hair_side_l_03_MSTR, loc=<0 0 0>, rot=<0 0 0 1>}
 dyn_hair_top_cowlick_MSTR {parent=head, loc=<16.1547 6.9712 -1.2879>, rot=<0.6125 0.3534 0.6845 0.1772>}
 dyn_hair_top_cowlick {parent=dyn_hair_top_cowlick_MSTR, loc=<0 0 0>, rot=<0 0 0 1>}
 dyn_hair_cowlick_r_00_MSTR {parent=head, loc=<13.6238 4.5637 6.6824>, rot=<0.4688 0.0639 0.8809 0.0126>}
 dyn_hair_cowlick_r_00 {parent=dyn_hair_cowlick_r_00_MSTR, loc=<0 0 0>, rot=<0 0 0 1>}
 dyn_hair_cowlick_r_01_MSTR {parent=head, loc=<15.971 2.9506 4.2162>, rot=<0.576 -0.0904 0.7899 -0.1901>}
 dyn_hair_cowlick_r_01 {parent=dyn_hair_cowlick_r_01_MSTR, loc=<0 0 0>, rot=<0 0 0 1>}
 dyn_hair_cowlick_r_02_MSTR {parent=head, loc=<16.4764 1.6269 3.3131>, rot=<0.9395 0.0039 0.3427 -0.0014>}
 dyn_hair_cowlick_r_02 {parent=dyn_hair_cowlick_r_02_MSTR, loc=<0 0 0>, rot=<0 0 0 1>}
 dyn_hair_cowlick_r_03_MSTR {parent=head, loc=<17.4031 -0.6361 -0.0569>, rot=<0.9894 0.0041 0.1449 -0.0006>}
 dyn_hair_cowlick_r_03 {parent=dyn_hair_cowlick_r_03_MSTR, loc=<0 0 0>, rot=<0 0 0 1>}
 dyn_hair_cowlick_r_04_MSTR {parent=head, loc=<13.4601 -2.0992 7.0614>, rot=<0.4392 -0.1777 0.7679 -0.4312>}
 dyn_hair_cowlick_r_04 {parent=dyn_hair_cowlick_r_04_MSTR, loc=<0 0 0>, rot=<0 0 0 1>}
 dyn_hair_cowlick_l_00_MSTR {parent=head, loc=<10.5873 8.9338 -5.4002>, rot=<-0.0894 -0.7014 -0.0952 -0.7007>}
 dyn_hair_cowlick_l_00 {parent=dyn_hair_cowlick_l_00_MSTR, loc=<0 0 0>, rot=<0 0 0 1>}
 dyn_hair_cowlick_l_01_MSTR {parent=head, loc=<10.5059 2.7153 -8.0883>, rot=<0.164 -0.7284 0.5242 -0.4095>}
 dyn_hair_cowlick_l_01 {parent=dyn_hair_cowlick_l_01_MSTR, loc=<0 0 0>, rot=<0 0 0 1>}
 dyn_hair_cowlick_l_02_MSTR {parent=head, loc=<15.1993 -1.313 -6.011>, rot=<0.8876 0.0037 -0.4605 0.0019>}
 dyn_hair_cowlick_l_02 {parent=dyn_hair_cowlick_l_02_MSTR, loc=<0 0 0>, rot=<0 0 0 1>}
 dyn_hair_cowlick_l_03_MSTR {parent=head, loc=<10.9681 -4.0831 -7.0421>, rot=<0.0768 -0.8864 -0.3632 -0.2766>}
 dyn_hair_cowlick_l_03 {parent=dyn_hair_cowlick_l_03_MSTR, loc=<0 0 0>, rot=<0 0 0 1>}
 dyn_hair_back_l_00_MSTR {parent=head, loc=<6.0314 -5.1967 -5.7429>, rot=<-0.036 -0.965 -0.2583 -0.0272>}
 dyn_hair_back_l_00 {parent=dyn_hair_back_l_00_MSTR, loc=<0 0 0>, rot=<0 0 0 1>}
 dyn_hair_back_l_01_MSTR {parent=head, loc=<5.5562 -6.9472 -0.527>, rot=<0.164 -0.917 0.021 -0.3631>}
 dyn_hair_back_l_01 {parent=dyn_hair_back_l_01_MSTR, loc=<0 0 0>, rot=<0 0 0 1>}
 dyn_hair_back_l_02_MSTR {parent=head, loc=<9.0879 -7.9245 -2.2244>, rot=<0.3578 -0.8001 0.3314 -0.3492>}
 dyn_hair_back_l_02 {parent=dyn_hair_back_l_02_MSTR, loc=<0 0 0>, rot=<0 0 0 1>}
 dyn_hair_back_r_00_MSTR {parent=head, loc=<10.3223 -7.1338 2.9305>, rot=<0.1654 -0.8106 0.4377 0.3521>}
 dyn_hair_back_r_00 {parent=dyn_hair_back_r_00_MSTR, loc=<0 0 0>, rot=<0 0 0 1>}
 dyn_hair_back_r_01_MSTR {parent=head, loc=<6.9556 -4.2712 7.0495>, rot=<0.0272 -0.8861 0.4624 -0.014>}
 dyn_hair_back_r_01 {parent=dyn_hair_back_r_01_MSTR, loc=<0 0 0>, rot=<0 0 0 1>}
 dyn_hair_backTop_00_MSTR {parent=head, loc=<14.4908 -5.3286 -1.9188>, rot=<0.4508 -0.5336 0.3749 -0.6096>}
 dyn_hair_backTop_00 {parent=dyn_hair_backTop_00_MSTR, loc=<0 0 0>, rot=<0 0 0 1>}
 dyn_hair_side_l_05_MSTR {parent=head, loc=<4.7567 -1.8064 -6.8707>, rot=<0.0951 -0.8737 -0.4771 0.0005>}
 dyn_hair_side_l_05 {parent=dyn_hair_side_l_05_MSTR, loc=<0 0 0>, rot=<0 0 0 1>}
 dyn_hair_side_l_04_MSTR {parent=head, loc=<9.5838 -0.4119 -7.9595>, rot=<0.318 -0.8193 -0.4609 -0.123>}
 dyn_hair_side_l_04 {parent=dyn_hair_side_l_04_MSTR, loc=<0 0 0>, rot=<0 0 0 1>}
 dyn_hair_side_r_00_MSTR {parent=head, loc=<8.1259 9.9982 3.4156>, rot=<-0.3488 0.4692 0.8036 0.1109>}
 dyn_hair_side_r_00 {parent=dyn_hair_side_r_00_MSTR, loc=<0 0 0>, rot=<0 0 0 1>}
 dyn_hair_side_r_01_MSTR {parent=head, loc=<9.5178 10.4101 4.0372>, rot=<0.5445 0.1177 0.7859 0.2682>}
 dyn_hair_side_r_01 {parent=dyn_hair_side_r_01_MSTR, loc=<0 0 0>, rot=<0 0 0 1>}
 dyn_hair_side_r_02_MSTR {parent=head, loc=<7.9033 8.3814 4.7455>, rot=<-0.2538 0.2896 0.8999 0.2048>}
 dyn_hair_side_r_02 {parent=dyn_hair_side_r_02_MSTR, loc=<0 0 0>, rot=<0 0 0 1>}
 dyn_hair_side_r_03_MSTR {parent=head, loc=<6.5872 7.1982 5.9465>, rot=<-0.1307 0.2499 0.9591 -0.0246>}
 dyn_hair_side_r_03 {parent=dyn_hair_side_r_03_MSTR, loc=<0 0 0>, rot=<0 0 0 1>}
 dyn_hair_side_r_04_MSTR {parent=head, loc=<8.912 6.5729 7.8658>, rot=<-0.0011 0.2509 0.968 0.009>}
 dyn_hair_side_r_04 {parent=dyn_hair_side_r_04_MSTR, loc=<0 0 0>, rot=<0 0 0 1>}
 dyn_hair_side_r_05_MSTR {parent=head, loc=<6.1788 5.3711 7.2034>, rot=<0.2261 0.1893 0.9523 -0.0778>}
 dyn_hair_side_r_05 {parent=dyn_hair_side_r_05_MSTR, loc=<0 0 0>, rot=<0 0 0 1>}
 dyn_hair_side_r_06_MSTR {parent=head, loc=<9.2543 3.6356 8.4446>, rot=<0.1806 0.248 0.88 -0.3626>}
 dyn_hair_side_r_06 {parent=dyn_hair_side_r_06_MSTR, loc=<0 0 0>, rot=<0 0 0 1>}
 dyn_hair_side_r_07_MSTR {parent=head, loc=<6.1445 3.1932 7.4967>, rot=<0.0388 0.114 0.9734 -0.1951>}
 dyn_hair_side_r_07 {parent=dyn_hair_side_r_07_MSTR, loc=<0 0 0>, rot=<0 0 0 1>}
 dyn_hair_cowlick_r_05_MSTR {parent=head, loc=<11.7527 -4.2171 6.5871>, rot=<0.2317 -0.2238 0.8998 -0.2944>}
 dyn_hair_cowlick_r_05 {parent=dyn_hair_cowlick_r_05_MSTR, loc=<0 0 0>, rot=<0 0 0 1>}
 dyn_hair_backTop_01_MSTR {parent=head, loc=<15.2507 -4.0493 -3.4273>, rot=<0.1355 -0.5663 0.2371 -0.7776>}
 dyn_hair_backTop_01 {parent=dyn_hair_backTop_01_MSTR, loc=<0 0 0>, rot=<0 0 0 1>}
 dyn_hair_front_02_MSTR {parent=head, loc=<9.9511 10.9118 -1.3219>, rot=<0.9246 0.337 0.0968 -0.1492>}
 dyn_hair_front_02 {parent=dyn_hair_front_02_MSTR, loc=<0 0 0>, rot=<0 0 0 1>}
 dyn_hair_front_00_MSTR_Rasp {parent=head, loc=<8.3959 10.3526 -0.2484>, rot=<-0.0033 -0.9988 -0.0481 -0.005>}
 dyn_hair_front_00_Rasp {parent=dyn_hair_front_00_MSTR_Rasp, loc=<0 0 0>, rot=<0 0 0 1>}
 dyn_hair_front_01_Rasp {parent=dyn_hair_front_00_Rasp, loc=<2.0502 0 0>, rot=<0 0 0 1>}
 dyn_hair_front_02_MSTR_Rasp {parent=head, loc=<9.6735 10.9505 -1.2426>, rot=<0.9646 -0.0861 -0.1551 -0.1952>}
 dyn_hair_front_02_Rasp {parent=dyn_hair_front_02_MSTR_Rasp, loc=<0 0 0>, rot=<0 0 0 1>}
 dyn_hair_side_l_03_MSTR_Rasp {parent=head, loc=<6.1867 1.8574 -7.4662>, rot=<0.4347 -0.8646 -0.184 -0.1721>}
 dyn_hair_side_l_03_Rasp {parent=dyn_hair_side_l_03_MSTR_Rasp, loc=<0 0 0>, rot=<0 0 0 1>}
 dyn_hair_cowlick_r_00_MSTR_Rasp {parent=head, loc=<13.6238 4.5637 6.6824>, rot=<0.4689 0.1015 0.8774 -0.0075>}
 dyn_hair_cowlick_r_00_Rasp {parent=dyn_hair_cowlick_r_00_MSTR_Rasp, loc=<0 0 0>, rot=<0 0 0 1>}
 dyn_hair_back_l_00_MSTR_Rasp {parent=head, loc=<6.0165 -5.3216 -5.6119>, rot=<-0.0738 -0.9179 -0.3661 -0.1346>}
 dyn_hair_back_l_00_Rasp {parent=dyn_hair_back_l_00_MSTR_Rasp, loc=<0 0 0>, rot=<0 0 0 1>}
 dyn_hair_side_l_05_MSTR_Rasp {parent=head, loc=<4.7567 -1.8064 -6.8707>, rot=<0.0951 -0.8737 -0.4771 0.0005>}
 dyn_hair_side_l_05_Rasp {parent=dyn_hair_side_l_05_MSTR_Rasp, loc=<0 0 0>, rot=<0 0 0 1>}
 dyn_apl_hair_back_L_MSTR {parent=head, loc=<6.87 -3.0439 -5.3836>, rot=<0.0249 -0.9645 0.2625 0.0131>}
 dyn_apl_hair_back_L_01 {parent=dyn_apl_hair_back_L_MSTR, loc=<0 0 0>, rot=<0 0 0 1>}
 dyn_apl_hair_back_L_02 {parent=dyn_apl_hair_back_L_01, loc=<4.676 0 0>, rot=<-0.0293 -0.0464 -0.0052 -0.9985>}
 dyn_apl_hair_back_L_03 {parent=dyn_apl_hair_back_L_02, loc=<4.3579 0 0>, rot=<0 -0.056 0 -0.9984>}
 dyn_apl_hair_back_L_04 {parent=dyn_apl_hair_back_L_02, loc=<4.1064 1.8565 -0.5139>, rot=<0.322 0.0698 -0.1339 -0.9346>}
 dyn_apl_hair_back_L_05 {parent=dyn_apl_hair_back_L_02, loc=<0.5772 -2.6745 0.6315>, rot=<-0.4501 0.5595 0.2233 -0.6592>}
 dyn_apl_hair_back_R_MSTR {parent=head, loc=<6.0282 -2.8457 5.4305>, rot=<0.0054 -0.2702 0.9626 0.0196>}
 dyn_apl_hair_back_R_01 {parent=dyn_apl_hair_back_R_MSTR, loc=<0 0 0>, rot=<0 0 0 1>}
 dyn_apl_hair_back_R_02 {parent=dyn_apl_hair_back_R_01, loc=<4.676 0 0>, rot=<-0.0295 0.0096 -0.0035 -0.9995>}
 dyn_apl_hair_back_R_03 {parent=dyn_apl_hair_back_R_02, loc=<4.3579 0 0>, rot=<0 0 0 1>}
 dyn_apl_hair_back_R_04 {parent=dyn_apl_hair_back_R_02, loc=<5.1559 -2.2139 -1.6483>, rot=<-0.3283 0.1404 -0.0602 -0.9321>}
 dyn_apl_hair_Side_R_MSTR {parent=head, loc=<8.6547 1.8338 6.7672>, rot=<0.0997 0.1126 0.9885 0.0149>}
 dyn_apl_hair_Side_R_01 {parent=dyn_apl_hair_Side_R_MSTR, loc=<0 0 0>, rot=<0 0 0 1>}
 dyn_apl_hair_Side_R_02 {parent=dyn_apl_hair_Side_R_01, loc=<4.676 0 0>, rot=<0.0095 -0.1938 0.1036 -0.9755>}
 dyn_apl_hair_Side_R_03 {parent=dyn_apl_hair_Side_R_02, loc=<4.676 0 0>, rot=<0.0016 -0.0333 0.0583 -0.9977>}
 dyn_apl_hair_Side_R_04 {parent=dyn_apl_hair_Side_R_03, loc=<4.676 0 0>, rot=<0.0003 -0.0054 0.0503 -0.9987>}
 dyn_apl_hairBangs_R1_MSTR {parent=head, loc=<10.2534 6.7554 5.1691>, rot=<0.1379 0.1979 0.9705 -0.0023>}
 dyn_apl_hairBangs_R1_01 {parent=dyn_apl_hairBangs_R1_MSTR, loc=<0 0 0>, rot=<0 0 0 1>}
 dyn_apl_hairBangs_R1_02 {parent=dyn_apl_hairBangs_R1_01, loc=<4.676 0 0>, rot=<-0.0024 -0.194 0.0437 -0.98>}
 dyn_apl_hairBangs_R1_03 {parent=dyn_apl_hairBangs_R1_02, loc=<4.676 0 0>, rot=<-0.0004 -0.0333 -0.0028 -0.9994>}
 dyn_apl_hairBangs_R_MSTR {parent=head, loc=<12.0753 9.567 1.7516>, rot=<0.5479 0.2706 0.7846 -0.1052>}
 dyn_apl_hairBangs_R_01 {parent=dyn_apl_hairBangs_R_MSTR, loc=<0 0 0>, rot=<0 0 0 1>}
 dyn_apl_hairBangs_R_02 {parent=dyn_apl_hairBangs_R_01, loc=<3.3498 0 0>, rot=<0.0466 -0.2486 0.0681 -0.9651>}
 dyn_apl_hairBangs_R_03 {parent=dyn_apl_hairBangs_R_02, loc=<3.1559 0 0>, rot=<-0.0494 -0.0662 0.0747 -0.9938>}
 dyn_apl_hair_back_MSTR {parent=head, loc=<5.2258 -6.5715 0.6174>, rot=<0.0199 -0.7008 0.713 -0.0122>}
 dyn_apl_hair_back_01 {parent=dyn_apl_hair_back_MSTR, loc=<0 0 0>, rot=<0 0 0 1>}
 dyn_apl_hair_back_02 {parent=dyn_apl_hair_back_01, loc=<4.676 0 0>, rot=<-0.0295 0.0096 -0.0035 -0.9995>}
 dyn_apl_hair_back_03 {parent=dyn_apl_hair_back_02, loc=<4.385 2.145 -0.8592>, rot=<0.0937 0.1914 -0.0184 -0.9769>}
 dyn_apl_hair_back_04 {parent=dyn_apl_hair_back_02, loc=<4.371 -3.363 -2.016>, rot=<-0.2704 0.2373 0.1034 -0.9273>}
 dyn_apl_hairBangs_L_MSTR {parent=head, loc=<11.2089 5.4433 -5.3183>, rot=<-0.0398 -0.9877 -0.117 -0.0956>}
 dyn_apl_hairBangs_L_01 {parent=dyn_apl_hairBangs_L_MSTR, loc=<0 0 0>, rot=<0 0 0 1>}
 dyn_apl_hairBangs_L_02 {parent=dyn_apl_hairBangs_L_01, loc=<4.5705 0 0>, rot=<0 0 0 1>}
 dyn_mstr_bang {parent=head, loc=<11.2197 8.2979 0>, rot=<-0.0029 0.7071 0.7071 -0.0029>}
 dyn_01_bang {parent=dyn_mstr_bang, loc=<0 0 0>, rot=<0 0 0 1>}
 dyn_mstr_bang_L {parent=head, loc=<11.2317 6.8542 -3.3865>, rot=<-0.1466 -0.8875 -0.4001 -0.1752>}
 dyn_01_bang_L {parent=dyn_mstr_bang_L, loc=<0 0 0>, rot=<0 0 0 1>}
 dyn_mstr_bang_R {parent=head, loc=<11.2315 6.8542 3.3865>, rot=<0.1752 0.4001 0.8875 0.1466>}
 dyn_01_bang_R {parent=dyn_mstr_bang_R, loc=<0 0 0>, rot=<0 0 0 1>}
 clavicle_l {parent=spine_05, loc=<4.3575 -0.1466 -2.3223>, rot=<0.0465 -0.7289 -0.009 -0.6829>}
 upperarm_l {parent=clavicle_l, loc=<15.5111 0 0>, rot=<0.0226 -0.3234 -0.0088 -0.9459>}
 clavicle_r {parent=spine_05, loc=<4.3575 -0.1466 2.3223>, rot=<0.7289 0.0465 -0.6829 0.009>}
 upperarm_r {parent=clavicle_r, loc=<-15.5111 0 0>, rot=<0.0226 -0.3234 -0.0088 -0.9459>}
 dyn_hair_r_TGT {parent=pelvis, loc=<0.8236 -14.3959 10.3103>, rot=<-0.0502 -0.7053 0.0502 -0.7053>}
 dyn_hair_l_TGT {parent=pelvis, loc=<0.8235 -14.3959 -10.3108>, rot=<-0.0502 -0.7053 0.0502 -0.7053>}
]
