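# Physics asset for the Pencil Apple head's hair on the medium male player character, providing capsule bodies and joints for dynamic hair bones.
# Seven chains (three bangs, right side, three back) hang from small MSTR bodies free-jointed to the head, with per-chain swing and twist limits down each strand.
# Head, neck, upper spine, clavicle and right upper-arm capsules form the body volumes.
# In all three back chains, segment 04 joints to 02 instead of 03.
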PhysicsAsset /BRCosmetics/Characters/Player/Male/Medium/Heads/M_MED_Pencil_Apple_Head/Meshes/Hair/M_MED_Pencil_Apple_Head_Physics
bodies {
 clavicle_l [{shapeType=Capsule, length=7.25, radius=6.75, shapeName=clavicle_l_capsule, center=<9.1843 0 0>, rotation=<rot 90 0 0>}]
 clavicle_r [
  {shapeType=Capsule, length=4.1323, radius=8.8677, shapeName=clavicle_r_capsule, center=<-8.9505 0 0>, rotation=<rot 90 0 0>}
  {
   shapeType=Capsule
   length=2.5
   radius=4.25
   shapeName=clavicle_r_capsule
   center=<-17.1196 -1.1241 1.161>
   rotation=<rot 57.4893 68.7886 -23.3557>
  }
 ]
 dyn_apl_hairBangs_L_01 [{shapeType=Capsule, length=2, radius=2.5, shapeName=dyn_apl_hairBangs_L_01_capsule, center=<2 0 2>, rotation=<rot 90 0 0>}]
 dyn_apl_hairBangs_L_02 [{shapeType=Capsule, length=2, radius=2.5, shapeName=dyn_apl_hairBangs_L_02_capsule, center=<2 0 2>, rotation=<rot 90 0 0>}]
 dyn_apl_hairBangs_L_MSTR [{shapeType=Capsule, length=0.505, radius=0.505, shapeName=dyn_apl_hairBangs_L_MSTR_capsule}]
 dyn_apl_hairBangs_R1_01 [{shapeType=Capsule, length=2, radius=2.5, shapeName=dyn_apl_hairBangs_R1_01_capsule, center=<2 0 2>, rotation=<rot 90 0 0>}]
 dyn_apl_hairBangs_R1_02 [{shapeType=Capsule, length=2, radius=2.5, shapeName=dyn_apl_hairBangs_R1_02_capsule, center=<2 0 2>, rotation=<rot 90 0 0>}]
 dyn_apl_hairBangs_R1_03 [{shapeType=Capsule, length=2, radius=2.5, shapeName=dyn_apl_hairBangs_R1_03_capsule, center=<2 0 2>, rotation=<rot 90 0 0>}]
 dyn_apl_hairBangs_R1_MSTR [{shapeType=Capsule, length=0.505, radius=0.505, shapeName=dyn_apl_hairBangs_R1_MSTR_capsule}]
 dyn_apl_hairBangs_R_01 [{shapeType=Capsule, length=0.5, radius=5.5, shapeName=dyn_apl_hairBangs_R_01_capsule, center=<2 0 2>, rotation=<rot 90 0 0>}]
 dyn_apl_hairBangs_R_02 [{shapeType=Capsule, length=0.5, radius=5.5, shapeName=dyn_apl_hairBangs_R_02_capsule, center=<2 0 2>, rotation=<rot 90 0 0>}]
 dyn_apl_hairBangs_R_03 [{shapeType=Capsule, length=0.5, radius=5.5, shapeName=dyn_apl_hairBangs_R_03_capsule, center=<2 0 2>, rotation=<rot 90 0 0>}]
 dyn_apl_hairBangs_R_MSTR [{shapeType=Capsule, length=0.505, radius=0.505, shapeName=dyn_apl_hairBangs_R_MSTR_capsule}]
 dyn_apl_hair_Side_R_01 [{shapeType=Capsule, length=20.6368, radius=5, shapeName=dyn_apl_hair_Side_R_01_capsule, center=<2.969 0 5>, rotation=<rot 0 0 90>}]
 dyn_apl_hair_Side_R_02 [{shapeType=Capsule, length=20.6368, radius=5, shapeName=dyn_apl_hair_Side_R_02_capsule, center=<2.969 0 5>, rotation=<rot 0 0 90>}]
 dyn_apl_hair_Side_R_03 [{shapeType=Capsule, length=20.6368, radius=5, shapeName=dyn_apl_hair_Side_R_03_capsule, center=<2.969 0 5>, rotation=<rot 0 0 90>}]
 dyn_apl_hair_Side_R_04 [{shapeType=Capsule, length=20.6368, radius=5, shapeName=dyn_apl_hair_Side_R_04_capsule, center=<2.969 0 5>, rotation=<rot 0 0 90>}]
 dyn_apl_hair_Side_R_MSTR [{shapeType=Capsule, length=0.505, radius=0.505, shapeName=dyn_apl_hair_Side_R_MSTR_capsule}]
 dyn_apl_hair_back_01 [{shapeType=Capsule, length=20.6368, radius=5, shapeName=dyn_apl_hair_back_01_capsule, center=<2.969 0 5>, rotation=<rot 0 0 90>}]
 dyn_apl_hair_back_02 [{shapeType=Capsule, length=22.0475, radius=3.5893, shapeName=dyn_apl_hair_back_02_capsule, center=<2.969 0 5>, rotation=<rot 0 0 90>}]
 dyn_apl_hair_back_03 [{shapeType=Capsule, length=22.0475, radius=3.5893, shapeName=dyn_apl_hair_back_03_capsule, center=<2.969 0 5>, rotation=<rot 0 0 90>}]
 dyn_apl_hair_back_04 [{shapeType=Capsule, length=22.0475, radius=3.5893, shapeName=dyn_apl_hair_back_04_capsule, center=<2.969 0 5>, rotation=<rot 0 0 90>}]
 dyn_apl_hair_back_L_01 [{shapeType=Capsule, length=1.5, radius=3, shapeName=dyn_apl_hair_back_L_01_capsule, center=<2 0 2>, rotation=<rot 90 0 0>}]
 dyn_apl_hair_back_L_02 [{shapeType=Capsule, length=13.0696, radius=4.0395, shapeName=dyn_apl_hair_back_L_02_capsule, center=<2 0 2>, rotation=<rot 0 0 90>}]
 dyn_apl_hair_back_L_03 [{shapeType=Capsule, length=21.0332, radius=3.9164, shapeName=dyn_apl_hair_back_L_03_capsule, center=<2 0 2>, rotation=<rot 0 0 90>}]
 dyn_apl_hair_back_L_04 [{shapeType=Capsule, length=23.192, radius=4.3319, shapeName=dyn_apl_hair_back_L_04_capsule, center=<3.414 0 2>, rotation=<rot 0 -23.3346 90>}]
 dyn_apl_hair_back_L_MSTR [{shapeType=Capsule, length=0.505, radius=0.505, shapeName=dyn_apl_hair_back_L_MSTR_capsule}]
 dyn_apl_hair_back_MSTR [{shapeType=Capsule, length=0.505, radius=0.505, shapeName=dyn_apl_hair_back_MSTR_capsule}]
 dyn_apl_hair_back_R_01 [{shapeType=Capsule, length=18.8965, radius=6.7403, shapeName=dyn_apl_hair_back_R_01_capsule, center=<2.969 0 5>, rotation=<rot 0 0 90>}]
 dyn_apl_hair_back_R_02 [{shapeType=Capsule, length=20.3072, radius=5.3296, shapeName=dyn_apl_hair_back_R_02_capsule, center=<2.969 0 5>, rotation=<rot 0 0 90>}]
 dyn_apl_hair_back_R_03 [{shapeType=Capsule, length=20.3072, radius=5.3296, shapeName=dyn_apl_hair_back_R_03_capsule, center=<2.969 0 5>, rotation=<rot 0 0 90>}]
 dyn_apl_hair_back_R_04 [{shapeType=Capsule, length=20.3072, radius=5.3296, shapeName=dyn_apl_hair_back_R_04_capsule, center=<2.969 0 5>, rotation=<rot 0 0 90>}]
 dyn_apl_hair_back_R_MSTR [{shapeType=Capsule, length=0.505, radius=0.505, shapeName=dyn_apl_hair_back_R_MSTR_capsule}]
 head [{shapeType=Capsule, length=3.5, radius=6.5, center=<3.702 2.0563 0>, rotation=<rot 90 45 19.1211>}]
 neck_01 [{shapeType=Capsule, length=3.4569, radius=6.5431, shapeName=neck_01_capsule, center=<-4.2492 0.1312 0>}]
 neck_02 [{shapeType=Capsule, length=3.0624, radius=6.9376, shapeName=neck_02_capsule, center=<0 0.9223 0>, rotation=<rot 90 0 0>}]
 spine_05 [{shapeType=Capsule, length=13.8415, radius=11.1585, shapeName=spine_05_capsule, center=<0.4288 0.1799 0>}]
 upperarm_r [{shapeType=Capsule, length=21.6478, radius=8.3522, shapeName=upperarm_r_capsule, center=<-12.3374 0 2.1983>, rotation=<rot -90 0 0>}]
}
constraints [
 {
  bone1Name=head
  bone2Name=neck_02
  swing1LimitDegrees=45
  swing1Motion=Free
  swing2LimitDegrees=45
  swing2Motion=Free
  twistLimitDegrees=45
  twistMotion=Free
 }
 {
  bone1Name=neck_02
  bone2Name=neck_01
  swing1LimitDegrees=45
  swing1Motion=Free
  swing2LimitDegrees=45
  swing2Motion=Free
  twistLimitDegrees=45
  twistMotion=Free
 }
 {
  bone1Name=neck_01
  bone2Name=spine_05
  swing1LimitDegrees=45
  swing1Motion=Free
  swing2LimitDegrees=45
  swing2Motion=Free
  twistLimitDegrees=45
  twistMotion=Free
 }
 {
  bone1Name=dyn_apl_hair_back_R_MSTR
  bone2Name=head
  swing1LimitDegrees=45
  swing1Motion=Free
  swing2LimitDegrees=45
  swing2Motion=Free
  twistLimitDegrees=45
  twistMotion=Free
 }
 {
  bone1Name=dyn_apl_hairBangs_L_MSTR
  bone2Name=head
  swing1LimitDegrees=45
  swing1Motion=Free
  swing2LimitDegrees=45
  swing2Motion=Free
  twistLimitDegrees=45
  twistMotion=Free
 }
 {
  bone1Name=dyn_apl_hairBangs_R1_MSTR
  bone2Name=head
  swing1LimitDegrees=45
  swing1Motion=Free
  swing2LimitDegrees=45
  swing2Motion=Free
  twistLimitDegrees=45
  twistMotion=Free
 }
 {
  bone1Name=dyn_apl_hairBangs_R_MSTR
  bone2Name=head
  swing1LimitDegrees=45
  swing1Motion=Free
  swing2LimitDegrees=45
  swing2Motion=Free
  twistLimitDegrees=45
  twistMotion=Free
 }
 {
  bone1Name=dyn_apl_hair_Side_R_MSTR
  bone2Name=head
  swing1LimitDegrees=45
  swing1Motion=Free
  swing2LimitDegrees=45
  swing2Motion=Free
  twistLimitDegrees=45
  twistMotion=Free
 }
 {
  bone1Name=dyn_apl_hair_back_L_MSTR
  bone2Name=head
  swing1LimitDegrees=45
  swing1Motion=Free
  swing2LimitDegrees=45
  swing2Motion=Free
  twistLimitDegrees=45
  twistMotion=Free
 }
 {
  bone1Name=dyn_apl_hair_back_MSTR
  bone2Name=head
  swing1LimitDegrees=45
  swing1Motion=Free
  swing2LimitDegrees=45
  swing2Motion=Free
  twistLimitDegrees=45
  twistMotion=Free
 }
 {bone1Name=dyn_apl_hairBangs_L_01, bone2Name=dyn_apl_hairBangs_L_MSTR, swing1LimitDegrees=45, swing2LimitDegrees=10, twistLimitDegrees=5}
 {bone1Name=dyn_apl_hairBangs_L_02, bone2Name=dyn_apl_hairBangs_L_01, swing1LimitDegrees=45, swing2LimitDegrees=10, twistLimitDegrees=5}
 {bone1Name=dyn_apl_hairBangs_R1_01, bone2Name=dyn_apl_hairBangs_R1_MSTR, swing1LimitDegrees=30, swing2LimitDegrees=30, twistLimitDegrees=30}
 {bone1Name=dyn_apl_hairBangs_R1_02, bone2Name=dyn_apl_hairBangs_R1_01, swing1LimitDegrees=30, swing2LimitDegrees=30, twistLimitDegrees=30}
 {bone1Name=dyn_apl_hairBangs_R1_03, bone2Name=dyn_apl_hairBangs_R1_02, swing1LimitDegrees=30, swing2LimitDegrees=30, twistLimitDegrees=30}
 {bone1Name=dyn_apl_hairBangs_R_01, bone2Name=dyn_apl_hairBangs_R_MSTR, swing1LimitDegrees=45, swing2LimitDegrees=30, twistLimitDegrees=5}
 {bone1Name=dyn_apl_hairBangs_R_02, bone2Name=dyn_apl_hairBangs_R_01, swing1LimitDegrees=45, swing2LimitDegrees=30, twistLimitDegrees=5}
 {bone1Name=dyn_apl_hairBangs_R_03, bone2Name=dyn_apl_hairBangs_R_02, swing1LimitDegrees=45, swing2LimitDegrees=30, twistLimitDegrees=5}
 {bone1Name=dyn_apl_hair_Side_R_01, bone2Name=dyn_apl_hair_Side_R_MSTR, swing1LimitDegrees=45, swing2LimitDegrees=45, twistLimitDegrees=45}
 {bone1Name=dyn_apl_hair_Side_R_02, bone2Name=dyn_apl_hair_Side_R_01, swing1LimitDegrees=45, swing2LimitDegrees=45, twistLimitDegrees=45}
 {bone1Name=dyn_apl_hair_Side_R_03, bone2Name=dyn_apl_hair_Side_R_02, swing1LimitDegrees=45, swing2LimitDegrees=45, twistLimitDegrees=45}
 {bone1Name=dyn_apl_hair_Side_R_04, bone2Name=dyn_apl_hair_Side_R_03, swing1LimitDegrees=45, swing2LimitDegrees=45, twistLimitDegrees=45}
 {bone1Name=dyn_apl_hair_back_L_01, bone2Name=dyn_apl_hair_back_L_MSTR, swing1LimitDegrees=20, swing2LimitDegrees=45, twistLimitDegrees=5}
 {bone1Name=dyn_apl_hair_back_L_02, bone2Name=dyn_apl_hair_back_L_01, swing1LimitDegrees=20, swing2LimitDegrees=45, twistLimitDegrees=5}
 {bone1Name=dyn_apl_hair_back_L_03, bone2Name=dyn_apl_hair_back_L_02, swing1LimitDegrees=45, swing2LimitDegrees=45, twistLimitDegrees=20}
 {bone1Name=dyn_apl_hair_back_L_04, bone2Name=dyn_apl_hair_back_L_02, swing1LimitDegrees=45, swing2LimitDegrees=45, twistLimitDegrees=20}
 {bone1Name=dyn_apl_hair_back_01, bone2Name=dyn_apl_hair_back_MSTR, swing1LimitDegrees=45, swing2LimitDegrees=45, twistLimitDegrees=20}
 {bone1Name=dyn_apl_hair_back_02, bone2Name=dyn_apl_hair_back_01, swing1LimitDegrees=45, swing2LimitDegrees=45, twistLimitDegrees=20}
 {bone1Name=dyn_apl_hair_back_03, bone2Name=dyn_apl_hair_back_02, swing1LimitDegrees=45, swing2LimitDegrees=45, twistLimitDegrees=20}
 {bone1Name=dyn_apl_hair_back_04, bone2Name=dyn_apl_hair_back_02, swing1LimitDegrees=45, swing2LimitDegrees=45, twistLimitDegrees=20}
 {bone1Name=dyn_apl_hair_back_R_01, bone2Name=dyn_apl_hair_back_R_MSTR, swing1LimitDegrees=30, swing2LimitDegrees=30, twistLimitDegrees=5}
 {bone1Name=dyn_apl_hair_back_R_02, bone2Name=dyn_apl_hair_back_R_01, swing1LimitDegrees=30, swing2LimitDegrees=30, twistLimitDegrees=5}
 {bone1Name=dyn_apl_hair_back_R_03, bone2Name=dyn_apl_hair_back_R_02, swing1LimitDegrees=30, swing2LimitDegrees=30, twistLimitDegrees=5}
 {bone1Name=dyn_apl_hair_back_R_04, bone2Name=dyn_apl_hair_back_R_02, swing1LimitDegrees=30, swing2LimitDegrees=30, twistLimitDegrees=5}
 {
  bone1Name=clavicle_l
  bone2Name=spine_05
  swing1LimitDegrees=45
  swing1Motion=Free
  swing2LimitDegrees=45
  swing2Motion=Free
  twistLimitDegrees=45
  twistMotion=Free
 }
 {
  bone1Name=clavicle_r
  bone2Name=spine_05
  swing1LimitDegrees=45
  swing1Motion=Free
  swing2LimitDegrees=45
  swing2Motion=Free
  twistLimitDegrees=45
  twistMotion=Free
 }
 {
  bone1Name=upperarm_r
  bone2Name=clavicle_r
  swing1LimitDegrees=45
  swing1Motion=Free
  swing2LimitDegrees=45
  swing2Motion=Free
  twistLimitDegrees=45
  twistMotion=Free
 }
]
