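# Post-style Animation Blueprint that adds secondary motion to the hair of the M_MED_Pencil_Apple character head.
# The Jacket flag picks a jacket Control Rig, RigidBody and AnimDynamics path, corrected by bone constraints and arm/backward-movement offsets.
# Otherwise, skydiving and rigid-body flags choose among alternative back-hair simulations built on a cached pose.
# The blended result then runs through spring-based AnimDynamics for bangs and side hair, using spherical head limits.
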AnimBlueprint /BRCosmetics/Characters/Player/Male/Medium/Heads/M_MED_Pencil_Apple_Head/Meshes/Hair/M_MED_Pencil_Apple_Hair_AnimBP
target_skeleton=/BRCosmetics/Characters/Player/Common/Fortnite_Base_Head/Pencil/Fortnite_Base_Head_Pencil_Skeleton
graphs [
 AnimGraph {nodes [
   RigidBody!(ComponentPose, Alpha, OverrideWorldGravity:Vector=Get GravityOverride(), ExternalForce) -> (Pose) {
    bClampLinearTranslationLimitToRefPose=true
    bOverrideWorldGravity=true
    overridePhysicsAsset=&/M_MED_Pencil_Apple_Head_Physics
    simulationSpace=BaseBoneSpace
    alphaBoolBlend {blendOption=Linear}
    baseBoneRef {boneName=neck_01}
    componentAppliedLinearAccClamp=<800 800 5000>
    componentLinearAccScale=<0.5 0.5 0.2>
    overrideWorldGravity {z=-980}
    simSpaceSettings {
     maxAngularAcceleration=1000
     maxAngularVelocity=3
     maxLinearAcceleration=700
     maxLinearVelocity=500
     velocityScaleZ=0
     worldAlpha=0.2
     externalLinearDragV=<0.2 0.2>
    }
   }
   n1 = Control Rig(Source=Use cached pose 'InitialCache'()) {
    alphaBoolBlend {blendOption=Linear}
    controlRigAssetReference {blueprintRigClass=&/M_MED_Pencil_Apple_Head_CtrlRig_Hair.M_MED_Pencil_Apple_Head_CtrlRig_Hair_C}
   }
   n2 = Anim Dynamics - Bone: dyn_apl_hair_back_01!(ComponentPose=Local To Component(LocalPose=@n1), Alpha=1, GravityOverride:Vector="0,0,500", ExternalForce="0,0,0") -> (Pose) {
    angularDampingOverride=0.99
    angularSpringConstant=800
    bAngularSpring=true
    bGravityOverrideInSimSpace=true
    bOverrideAngularDamping=true
    bUseGravityOverride=true
    bUsePlanarLimit=false
    bUseSphericalLimits=true
    gravityScale=0.3
    simulationSpace=BoneRelative
    alphaBoolBlend {blendOption=Linear}
    boundBone {boneName=dyn_apl_hair_back_01}
    componentAppliedLinearAccClamp=<2500 2500 1500>
    componentLinearAccScale=<0.5 0.5 0.4>
    componentLinearVelScale=<0.55 0.55 0.45>
    gravityOverride {z=500}
    physicsBodyDefinitions [{
      sphereCollisionRadius=0
      boundBone {boneName=dyn_apl_hair_back_01}
      boxExtents=<20 30 30>
      constraintSetup {angularLimitsMax=<2 -3 20>, angularLimitsMin=<-2 -45 -10>, angularTarget=<0 0 0>}
      localJointOffset=<3 0 1>
     }]
    relativeSpaceBone {boneName=root}
    sphericalLimits [
     {limitRadius=9, drivingBone {boneName=neck_01}, sphereLocalOffset=<0 0 0>}
     {limitRadius=13, drivingBone {boneName=clavicle_r}, sphereLocalOffset=<0 -14 0>}
     {limitRadius=0, drivingBone {boneName=None}, sphereLocalOffset=<0 0 0>}
    ]
   }
   n3 = Anim Dynamics - Bone: dyn_apl_hair_back_02!(ComponentPose=@n2, Alpha=1, GravityOverride:Vector="0,0,500", ExternalForce="0,0,0") -> (Pose) {
    angularDampingOverride=0.99
    angularSpringConstant=800
    bAngularSpring=true
    bGravityOverrideInSimSpace=true
    bOverrideAngularDamping=true
    bUseGravityOverride=true
    bUsePlanarLimit=false
    bUseSphericalLimits=true
    gravityScale=0.3
    simulationSpace=BoneRelative
    alphaBoolBlend {blendOption=Linear}
    boundBone {boneName=dyn_apl_hair_back_02}
    componentAppliedLinearAccClamp=<2500 2500 1500>
    componentLinearAccScale=<0.5 0.5 0.4>
    componentLinearVelScale=<0.55 0.55 0.45>
    gravityOverride {z=500}
    physicsBodyDefinitions [{
      sphereCollisionRadius=0
      boundBone {boneName=dyn_apl_hair_back_02}
      boxExtents=<20 30 30>
      constraintSetup {angularLimitsMax=<2 -3 20>, angularLimitsMin=<-2 -45 -10>, angularTarget=<0 0 0>}
      localJointOffset=<3 0 1>
     }]
    relativeSpaceBone {boneName=root}
    sphericalLimits [
     {limitRadius=9, drivingBone {boneName=neck_01}, sphereLocalOffset=<0 0 0>}
     {limitRadius=13, drivingBone {boneName=clavicle_r}, sphereLocalOffset=<0 -14 0>}
     {limitRadius=0, drivingBone {boneName=None}, sphereLocalOffset=<0 0 0>}
    ]
   }
   n4 = Anim Dynamics - Bone: dyn_apl_hair_back_03!(ComponentPose=@n3, Alpha=1, GravityOverride:Vector="0,0,500", ExternalForce="0,0,0") -> (Pose) {
    angularDampingOverride=0.99
    angularSpringConstant=800
    bAngularSpring=true
    bGravityOverrideInSimSpace=true
    bOverrideAngularDamping=true
    bUseGravityOverride=true
    bUsePlanarLimit=false
    bUseSphericalLimits=true
    gravityScale=0.3
    simulationSpace=BoneRelative
    alphaBoolBlend {blendOption=Linear}
    boundBone {boneName=dyn_apl_hair_back_03}
    componentAppliedLinearAccClamp=<2500 2500 1500>
    componentLinearAccScale=<0.5 0.5 0.4>
    componentLinearVelScale=<0.55 0.55 0.45>
    gravityOverride {z=500}
    physicsBodyDefinitions [{
      sphereCollisionRadius=0
      boundBone {boneName=dyn_apl_hair_back_03}
      boxExtents=<20 30 30>
      constraintSetup {angularLimitsMax=<2 -3 20>, angularLimitsMin=<-2 -45 -10>, angularTarget=<0 0 0>}
      localJointOffset=<3 0 1>
     }]
    relativeSpaceBone {boneName=root}
    sphericalLimits [
     {limitRadius=9, drivingBone {boneName=neck_01}, sphereLocalOffset=<0 0 0>}
     {limitRadius=13, drivingBone {boneName=clavicle_r}, sphereLocalOffset=<-10 0 15>}
     {limitRadius=0, drivingBone {boneName=None}, sphereLocalOffset=<0 0 0>}
    ]
   }
   n5 = Anim Dynamics - Bone: dyn_apl_hair_back_04!(ComponentPose=@n4, Alpha=1, GravityOverride:Vector="0,0,500", ExternalForce="0,0,0") -> (Pose) {
    angularDampingOverride=0.99
    angularSpringConstant=800
    bAngularSpring=true
    bGravityOverrideInSimSpace=true
    bOverrideAngularDamping=true
    bUseGravityOverride=true
    bUsePlanarLimit=false
    bUseSphericalLimits=true
    gravityScale=0.3
    simulationSpace=BoneRelative
    alphaBoolBlend {blendOption=Linear}
    boundBone {boneName=dyn_apl_hair_back_04}
    componentAppliedLinearAccClamp=<2500 2500 1500>
    componentLinearAccScale=<0.5 0.5 0.4>
    componentLinearVelScale=<0.55 0.55 0.45>
    gravityOverride {z=500}
    physicsBodyDefinitions [{
      sphereCollisionRadius=0
      boundBone {boneName=dyn_apl_hair_back_04}
      boxExtents=<20 30 30>
      constraintSetup {angularLimitsMax=<2 -3 20>, angularLimitsMin=<-2 -45 -10>, angularTarget=<0 0 0>}
      localJointOffset=<3 0 1>
     }]
    relativeSpaceBone {boneName=root}
    sphericalLimits [
     {limitRadius=9, drivingBone {boneName=neck_01}, sphereLocalOffset=<0 0 0>}
     {limitRadius=13, drivingBone {boneName=clavicle_r}, sphereLocalOffset=<-10 0 15>}
     {limitRadius=0, drivingBone {boneName=None}, sphereLocalOffset=<0 0 0>}
    ]
   }
   n6 = Anim Dynamics - Bone: dyn_apl_hair_back_R_01!(ComponentPose=@n5, Alpha=1, GravityOverride:Vector="0,0,500", ExternalForce="0,0,0") -> (Pose) {
    angularDampingOverride=0.9999
    angularSpringConstant=600
    bAngularSpring=true
    bGravityOverrideInSimSpace=true
    bOverrideAngularDamping=true
    bUseGravityOverride=true
    bUsePlanarLimit=false
    bUseSphericalLimits=true
    gravityScale=0.3
    simulationSpace=BoneRelative
    alphaBoolBlend {blendOption=Linear}
    boundBone {boneName=dyn_apl_hair_back_R_01}
    componentAppliedLinearAccClamp=<2500 2500 1500>
    componentLinearAccScale=<0.5 0.5 0.4>
    componentLinearVelScale=<0.55 0.55 0.45>
    gravityOverride {z=500}
    physicsBodyDefinitions [{
      sphereCollisionRadius=0
      boundBone {boneName=dyn_apl_hair_back_R_01}
      boxExtents=<10 20 20>
      constraintSetup {angularLimitsMax=<2 -8 20>, angularLimitsMin=<-2 -45 -10>, angularTarget=<0 0 0>}
      localJointOffset=<3 0 1>
     }]
    relativeSpaceBone {boneName=root}
    sphericalLimits [
     {limitRadius=9, drivingBone {boneName=neck_01}, sphereLocalOffset=<0 0 0>}
     {limitRadius=25, drivingBone {boneName=clavicle_r}, sphereLocalOffset=<0 -12.5 10>}
     {limitRadius=0, drivingBone {boneName=None}, sphereLocalOffset=<0 0 0>}
    ]
   }
   n7 = Anim Dynamics - Bone: dyn_apl_hair_back_R_02!(ComponentPose=@n6, Alpha=1, GravityOverride:Vector="0,0,500", ExternalForce="0,0,0") -> (Pose) {
    angularDampingOverride=0.9999
    angularSpringConstant=600
    bAngularSpring=true
    bGravityOverrideInSimSpace=true
    bOverrideAngularDamping=true
    bUseGravityOverride=true
    bUsePlanarLimit=false
    bUseSphericalLimits=true
    gravityScale=0.3
    simulationSpace=BoneRelative
    alphaBoolBlend {blendOption=Linear}
    boundBone {boneName=dyn_apl_hair_back_R_02}
    componentAppliedLinearAccClamp=<2500 2500 1500>
    componentLinearAccScale=<0.5 0.5 0.4>
    componentLinearVelScale=<0.55 0.55 0.45>
    gravityOverride {z=500}
    physicsBodyDefinitions [{
      sphereCollisionRadius=0
      boundBone {boneName=dyn_apl_hair_back_R_02}
      boxExtents=<10 20 20>
      constraintSetup {angularLimitsMax=<2 -8 20>, angularLimitsMin=<-2 -45 -10>, angularTarget=<0 0 0>}
      localJointOffset=<3 0 1>
     }]
    relativeSpaceBone {boneName=root}
    sphericalLimits [
     {limitRadius=9, drivingBone {boneName=neck_01}, sphereLocalOffset=<0 0 0>}
     {limitRadius=25, drivingBone {boneName=clavicle_r}, sphereLocalOffset=<0 -12.5 10>}
     {limitRadius=0, drivingBone {boneName=None}, sphereLocalOffset=<0 0 0>}
    ]
   }
   n8 = Anim Dynamics - Bone: dyn_apl_hair_back_R_03!(ComponentPose=@n7, Alpha=1, GravityOverride:Vector="0,0,500", ExternalForce="0,0,0") -> (Pose) {
    angularDampingOverride=0.9999
    angularSpringConstant=600
    bAngularSpring=true
    bGravityOverrideInSimSpace=true
    bOverrideAngularDamping=true
    bUseGravityOverride=true
    bUsePlanarLimit=false
    bUseSphericalLimits=true
    gravityScale=0.3
    simulationSpace=BoneRelative
    alphaBoolBlend {blendOption=Linear}
    boundBone {boneName=dyn_apl_hair_back_R_03}
    componentAppliedLinearAccClamp=<2500 2500 1500>
    componentLinearAccScale=<0.5 0.5 0.4>
    componentLinearVelScale=<0.55 0.55 0.45>
    gravityOverride {z=500}
    physicsBodyDefinitions [{
      sphereCollisionRadius=0
      boundBone {boneName=dyn_apl_hair_back_R_03}
      boxExtents=<10 20 20>
      constraintSetup {angularLimitsMax=<2 -8 20>, angularLimitsMin=<-2 -45 -10>, angularTarget=<0 0 0>}
      localJointOffset=<3 0 1>
     }]
    relativeSpaceBone {boneName=root}
    sphericalLimits [
     {limitRadius=9, drivingBone {boneName=neck_01}, sphereLocalOffset=<0 0 0>}
     {limitRadius=22, drivingBone {boneName=clavicle_r}, sphereLocalOffset=<-10 0 15>}
     {limitRadius=0, drivingBone {boneName=None}, sphereLocalOffset=<0 0 0>}
    ]
   }
   n9 = Anim Dynamics - Bone: dyn_apl_hair_back_R_04!(ComponentPose=@n8, Alpha=1, GravityOverride:Vector="0,0,500", ExternalForce="0,0,0") -> (Pose) {
    angularDampingOverride=0.9999
    angularSpringConstant=600
    bAngularSpring=true
    bGravityOverrideInSimSpace=true
    bOverrideAngularDamping=true
    bUseGravityOverride=true
    bUsePlanarLimit=false
    bUseSphericalLimits=true
    gravityScale=0.3
    simulationSpace=BoneRelative
    alphaBoolBlend {blendOption=Linear}
    boundBone {boneName=dyn_apl_hair_back_R_04}
    componentAppliedLinearAccClamp=<2500 2500 1500>
    componentLinearAccScale=<0.5 0.5 0.4>
    componentLinearVelScale=<0.55 0.55 0.45>
    gravityOverride {z=500}
    physicsBodyDefinitions [{
      sphereCollisionRadius=0
      boundBone {boneName=dyn_apl_hair_back_R_04}
      boxExtents=<10 20 20>
      constraintSetup {angularLimitsMax=<2 -8 20>, angularLimitsMin=<-2 -45 -10>, angularTarget=<0 0 0>}
      localJointOffset=<3 0 1>
     }]
    relativeSpaceBone {boneName=root}
    sphericalLimits [
     {limitRadius=9, drivingBone {boneName=neck_01}, sphereLocalOffset=<0 0 0>}
     {limitRadius=22, drivingBone {boneName=clavicle_r}, sphereLocalOffset=<-10 0 12>}
     {limitRadius=0, drivingBone {boneName=None}, sphereLocalOffset=<0 0 0>}
    ]
   }
   n10 = Anim Dynamics - Bone: dyn_apl_hairBangs_R1_03(ComponentPose=Local To Component(LocalPose=Use cached pose 'InitialCache'()), Alpha=1, ExternalForce="0,0,0") {
    angularDampingOverride=0.9999
    angularSpringConstant=5000
    bAngularSpring=true
    bOverrideAngularDamping=true
    bUseSphericalLimits=true
    simulationSpace=RootRelative
    alphaBoolBlend {blendOption=Linear}
    boundBone {boneName=dyn_apl_hairBangs_R1_03}
    componentAppliedLinearAccClamp=<5000 5000 5000>
    physicsBodyDefinitions [{
      boundBone {boneName=dyn_apl_hairBangs_R1_03}
      boxExtents=<10 15 15>
      constraintSetup {angularLimitsMax=<5 15 15>, angularLimitsMin=<-5 -15 -15>, angularTarget=<0 0 0>}
      localJointOffset=<2 0 0>
     }]
    relativeSpaceBone {boneName=pelvis}
    sphericalLimits [{limitRadius=12, drivingBone {boneName=head}, sphereLocalOffset=<5 0 0>}]
   }
   n11 = Anim Dynamics - Bone: dyn_apl_hair_Side_R_04(ComponentPose=@n10, Alpha=1, ExternalForce="0,0,0") {
    angularDampingOverride=0.9999
    angularSpringConstant=5000
    bAngularSpring=true
    bOverrideAngularDamping=true
    bUseSphericalLimits=true
    simulationSpace=RootRelative
    alphaBoolBlend {blendOption=Linear}
    boundBone {boneName=dyn_apl_hair_Side_R_04}
    componentAppliedLinearAccClamp=<5000 5000 5000>
    physicsBodyDefinitions [{
      boundBone {boneName=dyn_apl_hair_Side_R_04}
      boxExtents=<10 15 15>
      constraintSetup {angularLimitsMax=<5 15 15>, angularLimitsMin=<-5 -15 -15>, angularTarget=<0 0 0>}
      localJointOffset=<2 0 0>
     }]
    relativeSpaceBone {boneName=pelvis}
    sphericalLimits [{limitRadius=12, drivingBone {boneName=head}, sphereLocalOffset=<5 0 0>}]
   }
   n12 = RigidBody(ComponentPose=Local To Component(LocalPose=Use cached pose 'InitialCache'()), OverrideWorldGravity=Get GravityOverride()) {
    bClampLinearTranslationLimitToRefPose=true
    bOverrideWorldGravity=true
    overridePhysicsAsset=&M_MED_Pencil_Apple_Head_Physics
    simulationSpace=BaseBoneSpace
    alphaBoolBlend {blendOption=Linear}
    baseBoneRef {boneName=neck_01}
    componentAppliedLinearAccClamp=<2000 2000 2000>
    componentLinearAccScale=<0.37 0.37 0.5>
    overrideWorldGravity {z=-980}
   }
   n13 = "Transform (Modify) Bone - Bone: dyn_apl_hair_back_MSTR"!(ComponentPose=@n12, bAlphaBoolEnabled:bool=True, Translation="-1,0,0", Rotation="50,0,0", Scale="1,1,1") -> (Pose) {
    alphaInputType=Bool
    rotationMode=BMM_Additive
    rotationSpace=BCS_BoneSpace
    scaleMode=BMM_Additive
    scaleSpace=BCS_BoneSpace
    translationMode=BMM_Additive
    translationSpace=BCS_BoneSpace
    alphaBoolBlend {blendOption=Linear}
    boneToModify {boneName=dyn_apl_hair_back_MSTR}
    rotation {pitch=50}
    translation {x=-1}
   }
   n14 = "Transform (Modify) Bone - Bone: dyn_apl_hair_back_R_MSTR"!(ComponentPose=@n13, bAlphaBoolEnabled:bool=True, Translation="0,0,0", Rotation="50,0,0", Scale="1,1,1") -> (Pose) {
    alphaInputType=Bool
    rotationMode=BMM_Additive
    rotationSpace=BCS_BoneSpace
    scaleMode=BMM_Additive
    scaleSpace=BCS_BoneSpace
    translationMode=BMM_Additive
    translationSpace=BCS_BoneSpace
    alphaBoolBlend {blendOption=Linear}
    boneToModify {boneName=dyn_apl_hair_back_R_MSTR}
    rotation {pitch=50}
   }
   n15 = "Transform (Modify) Bone - Bone: dyn_apl_hair_back_MSTR"(ComponentPose=@n14, bAlphaBoolEnabled=True, Translation="1,0,0", Rotation="-50,0,0", Scale="1,1,1") {
    alphaInputType=Bool
    rotationMode=BMM_Additive
    rotationSpace=BCS_BoneSpace
    scaleMode=BMM_Additive
    scaleSpace=BCS_BoneSpace
    translationMode=BMM_Additive
    translationSpace=BCS_BoneSpace
    alphaBoolBlend {blendOption=Linear}
    boneToModify {boneName=dyn_apl_hair_back_MSTR}
    rotation {pitch=-50}
    translation {x=1}
   }
   n16 = "Transform (Modify) Bone - Bone: dyn_apl_hair_back_R_MSTR"(ComponentPose=@n15, bAlphaBoolEnabled=True, Translation="0,0,0", Rotation="-50,0,0", Scale="1,1,1") {
    alphaInputType=Bool
    rotationMode=BMM_Additive
    rotationSpace=BCS_BoneSpace
    scaleMode=BMM_Additive
    scaleSpace=BCS_BoneSpace
    translationMode=BMM_Additive
    translationSpace=BCS_BoneSpace
    alphaBoolBlend {blendOption=Linear}
    boneToModify {boneName=dyn_apl_hair_back_R_MSTR}
    rotation {pitch=-50}
   }
   n17 = Anim Dynamics - Bone: dyn_apl_hair_back_03(ComponentPose=@n16, Alpha=1, GravityOverride="0,0,500", ExternalForce="0,0,0") {
    angularDampingOverride=0.99
    angularSpringConstant=800
    bAngularSpring=true
    bGravityOverrideInSimSpace=true
    bOverrideAngularDamping=true
    bUseGravityOverride=true
    bUsePlanarLimit=false
    bUseSphericalLimits=true
    gravityScale=0.3
    simulationSpace=BoneRelative
    alphaBoolBlend {blendOption=Linear}
    boundBone {boneName=dyn_apl_hair_back_03}
    componentAppliedLinearAccClamp=<2500 2500 1500>
    componentLinearAccScale=<0.5 0.5 0.4>
    componentLinearVelScale=<0.55 0.55 0.45>
    gravityOverride {z=500}
    physicsBodyDefinitions [{
      sphereCollisionRadius=0
      boundBone {boneName=dyn_apl_hair_back_03}
      boxExtents=<20 30 30>
      constraintSetup {angularLimitsMax=<2 -3 20>, angularLimitsMin=<-2 -45 -10>, angularTarget=<0 0 0>}
      localJointOffset=<3 0 1>
     }]
    relativeSpaceBone {boneName=root}
    sphericalLimits [
     {limitRadius=9, drivingBone {boneName=neck_01}, sphereLocalOffset=<0 0 0>}
     {limitRadius=24, drivingBone {boneName=clavicle_r}, sphereLocalOffset=<-10 0 15>}
     {limitRadius=0, drivingBone {boneName=None}, sphereLocalOffset=<0 0 0>}
    ]
   }
   n18 = Anim Dynamics - Bone: dyn_apl_hair_back_04(ComponentPose=@n17, Alpha=1, GravityOverride="0,0,500", ExternalForce="0,0,0") {
    angularDampingOverride=0.99
    angularSpringConstant=800
    bAngularSpring=true
    bGravityOverrideInSimSpace=true
    bOverrideAngularDamping=true
    bUseGravityOverride=true
    bUsePlanarLimit=false
    bUseSphericalLimits=true
    gravityScale=0.3
    simulationSpace=BoneRelative
    alphaBoolBlend {blendOption=Linear}
    boundBone {boneName=dyn_apl_hair_back_04}
    componentAppliedLinearAccClamp=<2500 2500 1500>
    componentLinearAccScale=<0.5 0.5 0.4>
    componentLinearVelScale=<0.55 0.55 0.45>
    gravityOverride {z=500}
    physicsBodyDefinitions [{
      sphereCollisionRadius=0
      boundBone {boneName=dyn_apl_hair_back_04}
      boxExtents=<20 30 30>
      constraintSetup {angularLimitsMax=<2 -3 20>, angularLimitsMin=<-2 -45 -10>, angularTarget=<0 0 0>}
      localJointOffset=<3 0 1>
     }]
    relativeSpaceBone {boneName=root}
    sphericalLimits [
     {limitRadius=9, drivingBone {boneName=neck_01}, sphereLocalOffset=<0 0 0>}
     {limitRadius=24, drivingBone {boneName=clavicle_r}, sphereLocalOffset=<-10 0 15>}
     {limitRadius=0, drivingBone {boneName=None}, sphereLocalOffset=<0 0 0>}
    ]
   }
   n19 = Anim Dynamics - Bone: dyn_apl_hair_back_R_03(ComponentPose=@n18, Alpha=1, GravityOverride="0,0,500", ExternalForce="0,0,0") {
    angularDampingOverride=0.9999
    angularSpringConstant=600
    bAngularSpring=true
    bGravityOverrideInSimSpace=true
    bOverrideAngularDamping=true
    bUseGravityOverride=true
    bUsePlanarLimit=false
    bUseSphericalLimits=true
    gravityScale=0.3
    simulationSpace=BoneRelative
    alphaBoolBlend {blendOption=Linear}
    boundBone {boneName=dyn_apl_hair_back_R_03}
    componentAppliedLinearAccClamp=<2500 2500 1500>
    componentLinearAccScale=<0.5 0.5 0.4>
    componentLinearVelScale=<0.55 0.55 0.45>
    gravityOverride {z=500}
    physicsBodyDefinitions [{
      sphereCollisionRadius=0
      boundBone {boneName=dyn_apl_hair_back_R_03}
      boxExtents=<10 20 20>
      constraintSetup {angularLimitsMax=<2 -8 20>, angularLimitsMin=<-2 -45 -10>, angularTarget=<0 0 0>}
      localJointOffset=<3 0 1>
     }]
    relativeSpaceBone {boneName=root}
    sphericalLimits [
     {limitRadius=9, drivingBone {boneName=neck_01}, sphereLocalOffset=<0 0 0>}
     {limitRadius=22, drivingBone {boneName=clavicle_r}, sphereLocalOffset=<-10 0 15>}
     {limitRadius=0, drivingBone {boneName=None}, sphereLocalOffset=<0 0 0>}
    ]
   }
   n20 = Anim Dynamics - Bone: dyn_apl_hair_back_R_04(ComponentPose=@n19, Alpha=1, GravityOverride="0,0,500", ExternalForce="0,0,0") {
    angularDampingOverride=0.9999
    angularSpringConstant=600
    bAngularSpring=true
    bGravityOverrideInSimSpace=true
    bOverrideAngularDamping=true
    bUseGravityOverride=true
    bUsePlanarLimit=false
    bUseSphericalLimits=true
    gravityScale=0.3
    simulationSpace=BoneRelative
    alphaBoolBlend {blendOption=Linear}
    boundBone {boneName=dyn_apl_hair_back_R_04}
    componentAppliedLinearAccClamp=<2500 2500 1500>
    componentLinearAccScale=<0.5 0.5 0.4>
    componentLinearVelScale=<0.55 0.55 0.45>
    gravityOverride {z=500}
    physicsBodyDefinitions [{
      sphereCollisionRadius=0
      boundBone {boneName=dyn_apl_hair_back_R_04}
      boxExtents=<10 20 20>
      constraintSetup {angularLimitsMax=<2 -8 20>, angularLimitsMin=<-2 -45 -10>, angularTarget=<0 0 0>}
      localJointOffset=<3 0 1>
     }]
    relativeSpaceBone {boneName=root}
    sphericalLimits [
     {limitRadius=9, drivingBone {boneName=neck_01}, sphereLocalOffset=<0 0 0>}
     {limitRadius=22, drivingBone {boneName=clavicle_r}, sphereLocalOffset=<-10 0 12>}
     {limitRadius=0, drivingBone {boneName=None}, sphereLocalOffset=<0 0 0>}
    ]
   }
   n21 = Blend Poses by bool(
    BlendPose_0=Component To Local(ComponentPose=@n20)
    BlendPose_1=Blend Poses by bool(BlendPose_0=Component To Local(ComponentPose=@n9), BlendPose_1=Component To Local(ComponentPose=@n11), bActiveValue=Get bRigidBodyEnabled()) {blendType=Linear}
    bActiveValue=Get bIsSkyDiving()
   ) {blendType=Linear}
   n22 = Control Rig(Source=Use cached pose 'InitialCache'()) {
    alphaBoolBlend {blendOption=Linear}
    controlRigAssetReference {blueprintRigClass=&/M_MED_Pencil_Apple_Head_CtrlRig_Jacket.M_MED_Pencil_Apple_Head_CtrlRig_Jacket_C}
   }
   n23 = RigidBody(ComponentPose=Local To Component(LocalPose=@n22), OverrideWorldGravity="0,0,-980") {
    bClampLinearTranslationLimitToRefPose=true
    bOverrideWorldGravity=true
    overridePhysicsAsset=&/M_MED_Pencil_Apple_Head_Physics_Jacket
    simulationSpace=BaseBoneSpace
    alphaBoolBlend {blendOption=Linear}
    baseBoneRef {boneName=neck_01}
    componentAppliedLinearAccClamp=<2000 2000 2000>
    componentLinearAccScale=<0.37 0.37 0.3>
    overrideWorldGravity {z=-980}
    simSpaceSettings {
     maxAngularAcceleration=1000
     maxAngularVelocity=3
     maxLinearAcceleration=700
     maxLinearVelocity=500
     velocityScaleZ=0
     worldAlpha=0.2
     externalLinearDragV=<0.2 0.2>
    }
   }
   n24 = Anim Dynamics - Bone: dyn_apl_hair_back_01(ComponentPose=@n23, Alpha=1, GravityOverride="0,0,500", ExternalForce="0,0,0") {
    angularDampingOverride=0.99
    angularSpringConstant=800
    bAngularSpring=true
    bGravityOverrideInSimSpace=true
    bOverrideAngularDamping=true
    bUseGravityOverride=true
    gravityScale=0.3
    simulationSpace=BoneRelative
    alphaBoolBlend {blendOption=Linear}
    boundBone {boneName=dyn_apl_hair_back_01}
    componentAppliedLinearAccClamp=<2500 2500 1500>
    componentLinearAccScale=<0.5 0.5 0.4>
    componentLinearVelScale=<0.55 0.55 0.45>
    gravityOverride {z=500}
    physicsBodyDefinitions [{
      sphereCollisionRadius=0
      boundBone {boneName=dyn_apl_hair_back_01}
      boxExtents=<60 60 60>
      constraintSetup {angularLimitsMax=<2 3.5 20>, angularLimitsMin=<-2 -3.5 -10>, angularTarget=<0 0 0>}
      localJointOffset=<-5 0.2 -0.1>
     }]
    relativeSpaceBone {boneName=root}
   }
   n25 = Anim Dynamics - Bone: dyn_apl_hair_back_02(ComponentPose=@n24, Alpha=1, GravityOverride="0,0,500", ExternalForce="0,0,0") {
    angularDampingOverride=0.99
    angularSpringConstant=800
    bAngularSpring=true
    bGravityOverrideInSimSpace=true
    bOverrideAngularDamping=true
    bUseGravityOverride=true
    gravityScale=0.3
    simulationSpace=BoneRelative
    alphaBoolBlend {blendOption=Linear}
    boundBone {boneName=dyn_apl_hair_back_02}
    componentAppliedLinearAccClamp=<2500 2500 1500>
    componentLinearAccScale=<0.5 0.5 0.4>
    componentLinearVelScale=<0.55 0.55 0.45>
    gravityOverride {z=500}
    physicsBodyDefinitions [{
      sphereCollisionRadius=0
      boundBone {boneName=dyn_apl_hair_back_02}
      boxExtents=<60 60 60>
      constraintSetup {angularLimitsMax=<2 3.5 20>, angularLimitsMin=<-2 -3.5 -10>, angularTarget=<0 0 0>}
      localJointOffset=<-5 0.2 -0.1>
     }]
    relativeSpaceBone {boneName=root}
   }
   n26 = Anim Dynamics - Bone: dyn_apl_hair_back_03(ComponentPose=@n25, Alpha=1, GravityOverride="0,0,500", ExternalForce="0,0,0") {
    angularDampingOverride=0.99
    angularSpringConstant=800
    bAngularSpring=true
    bGravityOverrideInSimSpace=true
    bOverrideAngularDamping=true
    bUseGravityOverride=true
    gravityScale=0.3
    simulationSpace=BoneRelative
    alphaBoolBlend {blendOption=Linear}
    boundBone {boneName=dyn_apl_hair_back_03}
    componentAppliedLinearAccClamp=<2500 2500 1500>
    componentLinearAccScale=<0.5 0.5 0.4>
    componentLinearVelScale=<0.55 0.55 0.45>
    gravityOverride {z=500}
    physicsBodyDefinitions [{
      sphereCollisionRadius=0
      boundBone {boneName=dyn_apl_hair_back_03}
      boxExtents=<60 60 60>
      constraintSetup {angularLimitsMax=<2 3.5 20>, angularLimitsMin=<-2 -3.5 -10>, angularTarget=<0 0 0>}
      localJointOffset=<-5 0.2 -0.1>
     }]
    relativeSpaceBone {boneName=root}
   }
   n27 = Anim Dynamics - Bone: dyn_apl_hair_back_04(ComponentPose=@n26, Alpha=1, GravityOverride="0,0,500", ExternalForce="0,0,0") {
    angularDampingOverride=0.99
    angularSpringConstant=800
    bAngularSpring=true
    bGravityOverrideInSimSpace=true
    bOverrideAngularDamping=true
    bUseGravityOverride=true
    gravityScale=0.3
    simulationSpace=BoneRelative
    alphaBoolBlend {blendOption=Linear}
    boundBone {boneName=dyn_apl_hair_back_04}
    componentAppliedLinearAccClamp=<2500 2500 1500>
    componentLinearAccScale=<0.5 0.5 0.4>
    componentLinearVelScale=<0.55 0.55 0.45>
    gravityOverride {z=500}
    physicsBodyDefinitions [{
      sphereCollisionRadius=0
      boundBone {boneName=dyn_apl_hair_back_04}
      boxExtents=<60 60 60>
      constraintSetup {angularLimitsMax=<2 3.5 20>, angularLimitsMin=<-2 -3.5 -10>, angularTarget=<0 0 0>}
      localJointOffset=<-5 0.2 -0.1>
     }]
    relativeSpaceBone {boneName=root}
   }
   n28 = Anim Dynamics - Bone: dyn_apl_hair_back_R_01(ComponentPose=@n27, Alpha=1, GravityOverride="0,0,500", ExternalForce="0,0,0") {
    angularDampingOverride=0.99
    angularSpringConstant=800
    bAngularSpring=true
    bGravityOverrideInSimSpace=true
    bOverrideAngularDamping=true
    bUseGravityOverride=true
    gravityScale=0.3
    simulationSpace=BoneRelative
    alphaBoolBlend {blendOption=Linear}
    boundBone {boneName=dyn_apl_hair_back_R_01}
    componentAppliedLinearAccClamp=<2500 2500 1500>
    componentLinearAccScale=<0.5 0.5 0.4>
    componentLinearVelScale=<0.55 0.55 0.45>
    gravityOverride {z=500}
    physicsBodyDefinitions [{
      sphereCollisionRadius=0
      boundBone {boneName=dyn_apl_hair_back_R_01}
      boxExtents=<60 60 60>
      constraintSetup {angularLimitsMax=<2 3.5 20>, angularLimitsMin=<-2 -3.5 -10>, angularTarget=<0 0 0>}
      localJointOffset=<-5 0.2 -0.1>
     }]
    relativeSpaceBone {boneName=root}
   }
   n29 = Anim Dynamics - Bone: dyn_apl_hair_back_R_02(ComponentPose=@n28, Alpha=1, GravityOverride="0,0,500", ExternalForce="0,0,0") {
    angularDampingOverride=0.99
    angularSpringConstant=800
    bAngularSpring=true
    bGravityOverrideInSimSpace=true
    bOverrideAngularDamping=true
    bUseGravityOverride=true
    gravityScale=0.3
    simulationSpace=BoneRelative
    alphaBoolBlend {blendOption=Linear}
    boundBone {boneName=dyn_apl_hair_back_R_02}
    componentAppliedLinearAccClamp=<2500 2500 1500>
    componentLinearAccScale=<0.5 0.5 0.4>
    componentLinearVelScale=<0.55 0.55 0.45>
    gravityOverride {z=500}
    physicsBodyDefinitions [{
      sphereCollisionRadius=0
      boundBone {boneName=dyn_apl_hair_back_R_02}
      boxExtents=<60 60 60>
      constraintSetup {angularLimitsMax=<2 3.5 20>, angularLimitsMin=<-2 -3.5 -10>, angularTarget=<0 0 0>}
      localJointOffset=<-5 0.2 -0.1>
     }]
    relativeSpaceBone {boneName=root}
   }
   n30 = Anim Dynamics - Bone: dyn_apl_hair_back_R_03(ComponentPose=@n29, Alpha=1, GravityOverride="0,0,500", ExternalForce="0,0,0") {
    angularDampingOverride=0.99
    angularSpringConstant=800
    bAngularSpring=true
    bGravityOverrideInSimSpace=true
    bOverrideAngularDamping=true
    bUseGravityOverride=true
    gravityScale=0.3
    simulationSpace=BoneRelative
    alphaBoolBlend {blendOption=Linear}
    boundBone {boneName=dyn_apl_hair_back_R_03}
    componentAppliedLinearAccClamp=<2500 2500 1500>
    componentLinearAccScale=<0.5 0.5 0.4>
    componentLinearVelScale=<0.55 0.55 0.45>
    gravityOverride {z=500}
    physicsBodyDefinitions [{
      sphereCollisionRadius=0
      boundBone {boneName=dyn_apl_hair_back_R_03}
      boxExtents=<60 60 60>
      constraintSetup {angularLimitsMax=<2 3.5 20>, angularLimitsMin=<-2 -3.5 -10>, angularTarget=<0 0 0>}
      localJointOffset=<-5 0.2 -0.1>
     }]
    relativeSpaceBone {boneName=root}
   }
   n31 = Anim Dynamics - Bone: dyn_apl_hair_back_R_04(ComponentPose=@n30, Alpha=1, GravityOverride="0,0,500", ExternalForce="0,0,0") {
    angularDampingOverride=0.99
    angularSpringConstant=800
    bAngularSpring=true
    bGravityOverrideInSimSpace=true
    bOverrideAngularDamping=true
    bUseGravityOverride=true
    gravityScale=0.3
    simulationSpace=BoneRelative
    alphaBoolBlend {blendOption=Linear}
    boundBone {boneName=dyn_apl_hair_back_R_04}
    componentAppliedLinearAccClamp=<2500 2500 1500>
    componentLinearAccScale=<0.5 0.5 0.4>
    componentLinearVelScale=<0.55 0.55 0.45>
    gravityOverride {z=500}
    physicsBodyDefinitions [{
      sphereCollisionRadius=0
      boundBone {boneName=dyn_apl_hair_back_R_04}
      boxExtents=<60 60 60>
      constraintSetup {angularLimitsMax=<2 3.5 20>, angularLimitsMin=<-2 -3.5 -10>, angularTarget=<0 0 0>}
      localJointOffset=<-5 0.2 -0.1>
     }]
    relativeSpaceBone {boneName=root}
   }
   n32 = "Constraint - dyn_apl_hair_back_R_04 (PPPPP)"!(
    ComponentPose=@n31
    Alpha=1
    ConstraintWeights_0=0.0321
    ConstraintWeights_1:float=0.0601
    ConstraintWeights_2:float=0.2047
    ConstraintWeights_3:float=0.3277
    ConstraintWeights_4:float=0.3754
   ) -> (Pose) {
    alphaBoolBlend {blendOption=Linear}
    boneToModify {boneName=dyn_apl_hair_back_R_04}
    constraintSetup [
     {targetBone {boneName=spine_05}}
     {targetBone {boneName=C_jaw}}
     {targetBone {boneName=neck_01}}
     {targetBone {boneName=head}}
     {targetBone {boneName=neck_02}}
    ]
    constraintWeights [1,1,1,1,1]
   }
   n33 = "Constraint - dyn_apl_hair_back_L_04 (PPPP)"!(
    ComponentPose=@n32
    Alpha=1
    ConstraintWeights_0=0.0422
    ConstraintWeights_1:float=0
    ConstraintWeights_2:float=0.4969
    ConstraintWeights_3:float=0.4609
   ) -> (Pose) {
    alphaBoolBlend {blendOption=Linear}
    boneToModify {boneName=dyn_apl_hair_back_L_04}
    constraintSetup [{targetBone {boneName=C_jaw}}, {targetBone {boneName=neck_01}}, {targetBone {boneName=head}}, {targetBone {boneName=neck_02}}]
    constraintWeights [1,1,1,1]
   }
   n34 = "Transform (Modify) Bone - Bone: dyn_apl_hair_back_02"(ComponentPose=@n33, Alpha=Get Arm R rot() -> (Arm R rot:double), Translation="0,0,0", Rotation="-30,0,0", Scale="1,1,1") {
    rotationMode=BMM_Additive
    rotationSpace=BCS_BoneSpace
    scaleMode=BMM_Additive
    scaleSpace=BCS_BoneSpace
    translationMode=BMM_Additive
    translationSpace=BCS_BoneSpace
    alphaBoolBlend {blendOption=Linear}
    boneToModify {boneName=dyn_apl_hair_back_02}
    rotation {pitch=-30}
   }
   n35 = "Transform (Modify) Bone - Bone: dyn_apl_hair_back_R_MSTR"!(ComponentPose=@n34, Alpha:float=Get Arm R rot(), Translation="0,0,1", Rotation="-10,0,0", Scale="1,1,1") -> (Pose) {
    rotationMode=BMM_Additive
    rotationSpace=BCS_BoneSpace
    scaleMode=BMM_Additive
    scaleSpace=BCS_BoneSpace
    translationMode=BMM_Additive
    translationSpace=BCS_BoneSpace
    alphaBoolBlend {blendOption=Linear}
    boneToModify {boneName=dyn_apl_hair_back_R_MSTR}
    rotation {pitch=-10}
    translation {z=1}
   }
   n36 = Get Arm R rot()
   n37 = "Transform (Modify) Bone - Bone: dyn_apl_hair_back_04"(ComponentPose=@n35, Alpha=@n36, Translation="0,0,0", Rotation="-30,0,0", Scale="1,1,1") {
    rotationMode=BMM_Additive
    rotationSpace=BCS_BoneSpace
    scaleMode=BMM_Additive
    scaleSpace=BCS_BoneSpace
    translationMode=BMM_Additive
    translationSpace=BCS_BoneSpace
    alphaBoolBlend {blendOption=Linear}
    boneToModify {boneName=dyn_apl_hair_back_04}
    rotation {pitch=-30}
   }
   n38 = "Transform (Modify) Bone - Bone: dyn_apl_hair_back_R_MSTR"(ComponentPose=@n37, Alpha=@n36, Translation="0.524,-0.022,-2.534", Rotation="19.879,4.371,3.278", Scale="1,1,1") {
    rotationMode=BMM_Additive
    rotationSpace=BCS_BoneSpace
    scaleMode=BMM_Additive
    scaleSpace=BCS_BoneSpace
    translationMode=BMM_Additive
    translationSpace=BCS_BoneSpace
    alphaBoolBlend {blendOption=Linear}
    boneToModify {boneName=dyn_apl_hair_back_R_MSTR}
    rotation=<rot 19.879 4.371 3.278>
    translation=<0.524 -0.022 -2.534>
   }
   n39 = Get BwdMovement() -> (BwdMovement:float)
   n40 = "Transform (Modify) Bone - Bone: dyn_apl_hair_back_R_02"(ComponentPose=@n38, Alpha=@n39, Translation="0,0,0", Rotation="-13.813,14.750,-2.635", Scale="1,1,1") {
    rotationMode=BMM_Additive
    rotationSpace=BCS_BoneSpace
    scaleMode=BMM_Additive
    scaleSpace=BCS_BoneSpace
    translationMode=BMM_Additive
    translationSpace=BCS_BoneSpace
    alphaBoolBlend {blendOption=Linear}
    alphaScaleBiasClamp {bInterpResult=true}
    boneToModify {boneName=dyn_apl_hair_back_R_02}
    rotation=<rot -13.813 14.75 -2.635>
   }
   n41 = "Transform (Modify) Bone - Bone: dyn_apl_hair_back_R_01"(ComponentPose=@n40, Alpha=@n39, Translation="0,0,0", Rotation="-14.958,0.033,0.005", Scale="1,1,1") {
    rotationMode=BMM_Additive
    rotationSpace=BCS_BoneSpace
    scaleMode=BMM_Additive
    scaleSpace=BCS_BoneSpace
    translationMode=BMM_Additive
    translationSpace=BCS_BoneSpace
    alphaBoolBlend {blendOption=Linear}
    alphaScaleBiasClamp {bInterpResult=true}
    boneToModify {boneName=dyn_apl_hair_back_R_01}
    rotation=<rot -14.958 0.033 0.005>
   }
   n42 = Component To Local(ComponentPose=@n41)
   n43 = Get Jacket() -> (Jacket:bool)
   n44 = Anim Dynamics - Bone: dyn_apl_hairBangs_R_01(
    ComponentPose=Local To Component(LocalPose=Blend Poses by bool(BlendPose_0=@n42, BlendPose_1=@n21, bActiveValue=@n43) {blendType=Linear})
    Alpha=1
    ExternalForce="0,0,0"
   ) {
    angularDampingOverride=0.9999
    angularSpringConstant=5000
    bAngularSpring=true
    bOverrideAngularDamping=true
    bUseSphericalLimits=true
    simulationSpace=RootRelative
    alphaBoolBlend {blendOption=Linear}
    boundBone {boneName=dyn_apl_hairBangs_R_01}
    componentAppliedLinearAccClamp=<5000 5000 5000>
    physicsBodyDefinitions [{
      boundBone {boneName=dyn_apl_hairBangs_R_01}
      boxExtents=<10 15 15>
      constraintSetup {angularLimitsMax=<5 15 15>, angularLimitsMin=<-5 -15 -15>, angularTarget=<0 0 0>}
      localJointOffset=<2 0 0>
     }]
    relativeSpaceBone {boneName=pelvis}
    sphericalLimits [{limitRadius=12, drivingBone {boneName=head}, sphereLocalOffset=<5 0 0>}]
   }
   n45 = Anim Dynamics - Bone: dyn_apl_hairBangs_R_02(ComponentPose=@n44, Alpha=1, ExternalForce="0,0,0") {
    angularDampingOverride=0.9999
    angularSpringConstant=5000
    bAngularSpring=true
    bOverrideAngularDamping=true
    bUseSphericalLimits=true
    simulationSpace=RootRelative
    alphaBoolBlend {blendOption=Linear}
    boundBone {boneName=dyn_apl_hairBangs_R_02}
    componentAppliedLinearAccClamp=<5000 5000 5000>
    physicsBodyDefinitions [{
      boundBone {boneName=dyn_apl_hairBangs_R_02}
      boxExtents=<10 15 15>
      constraintSetup {angularLimitsMax=<5 15 15>, angularLimitsMin=<-5 -15 -15>, angularTarget=<0 0 0>}
      localJointOffset=<2 0 0>
     }]
    relativeSpaceBone {boneName=pelvis}
    sphericalLimits [{limitRadius=12, drivingBone {boneName=head}, sphereLocalOffset=<5 0 0>}]
   }
   n46 = Anim Dynamics - Bone: dyn_apl_hairBangs_R_03(ComponentPose=@n45, Alpha=1, ExternalForce="0,0,0") {
    angularDampingOverride=0.9999
    angularSpringConstant=5000
    bAngularSpring=true
    bOverrideAngularDamping=true
    bUseSphericalLimits=true
    simulationSpace=RootRelative
    alphaBoolBlend {blendOption=Linear}
    boundBone {boneName=dyn_apl_hairBangs_R_03}
    componentAppliedLinearAccClamp=<5000 5000 5000>
    physicsBodyDefinitions [{
      boundBone {boneName=dyn_apl_hairBangs_R_03}
      boxExtents=<10 15 15>
      constraintSetup {angularLimitsMax=<5 15 15>, angularLimitsMin=<-5 -15 -15>, angularTarget=<0 0 0>}
      localJointOffset=<2 0 0>
     }]
    relativeSpaceBone {boneName=pelvis}
    sphericalLimits [{limitRadius=12, drivingBone {boneName=head}, sphereLocalOffset=<5 0 0>}]
   }
   n47 = Anim Dynamics - Bone: dyn_apl_hairBangs_L_01(ComponentPose=@n46, Alpha=0.5, ExternalForce="0,0,0") {
    angularDampingOverride=0.9999
    angularSpringConstant=500
    bAngularSpring=true
    bOverrideAngularDamping=true
    bUseSphericalLimits=true
    simulationSpace=RootRelative
    alphaBoolBlend {blendOption=Linear}
    boundBone {boneName=dyn_apl_hairBangs_L_01}
    componentAppliedLinearAccClamp=<5000 5000 5000>
    physicsBodyDefinitions [{
      boundBone {boneName=dyn_apl_hairBangs_L_01}
      boxExtents=<12 18 18>
      constraintSetup {angularLimitsMax=<20 20 20>, angularLimitsMin=<-20 -25 -25>, angularTarget=<0 0 0>}
      localJointOffset=<2 0 0>
     }]
    relativeSpaceBone {boneName=pelvis}
    sphericalLimits [{limitRadius=9.3, drivingBone {boneName=head}, sphereLocalOffset=<6 0 0>}]
   }
   n48 = Anim Dynamics - Bone: dyn_apl_hairBangs_L_02(ComponentPose=@n47, Alpha=0.5, ExternalForce="0,0,0") {
    angularDampingOverride=0.9999
    angularSpringConstant=500
    bAngularSpring=true
    bOverrideAngularDamping=true
    bUseSphericalLimits=true
    simulationSpace=RootRelative
    alphaBoolBlend {blendOption=Linear}
    boundBone {boneName=dyn_apl_hairBangs_L_02}
    componentAppliedLinearAccClamp=<5000 5000 5000>
    physicsBodyDefinitions [{
      boundBone {boneName=dyn_apl_hairBangs_L_02}
      boxExtents=<12 18 18>
      constraintSetup {angularLimitsMax=<20 20 20>, angularLimitsMin=<-20 -25 -25>, angularTarget=<0 0 0>}
      localJointOffset=<2 0 0>
     }]
    relativeSpaceBone {boneName=pelvis}
    sphericalLimits [{limitRadius=9.3, drivingBone {boneName=head}, sphereLocalOffset=<6 0 0>}]
   }
   n49 = Anim Dynamics - Bone: dyn_apl_hair_Side_R_01(ComponentPose=@n48, Alpha=1, ExternalForce="0,0,0") {
    angularDampingOverride=1
    angularSpringConstant=4000
    bAngularSpring=true
    bOverrideAngularDamping=true
    bUseSphericalLimits=true
    gravityScale=0.8
    simulationSpace=RootRelative
    alphaBoolBlend {blendOption=Linear}
    boundBone {boneName=dyn_apl_hair_Side_R_01}
    componentAppliedLinearAccClamp=<5000 5000 5000>
    physicsBodyDefinitions [{
      boundBone {boneName=dyn_apl_hair_Side_R_01}
      boxExtents=<10 15 15>
      constraintSetup {angularLimitsMax=<5 15 15>, angularLimitsMin=<-5 -15 -15>, angularTarget=<0 0 0>}
      localJointOffset=<2 0 0>
     }]
    relativeSpaceBone {boneName=pelvis}
    sphericalLimits [{limitRadius=7.5, drivingBone {boneName=head}, sphereLocalOffset=<5 0 0>}]
   }
   n50 = Anim Dynamics - Bone: dyn_apl_hair_Side_R_02(ComponentPose=@n49, Alpha=1, ExternalForce="0,0,0") {
    angularDampingOverride=1
    angularSpringConstant=4000
    bAngularSpring=true
    bOverrideAngularDamping=true
    bUseSphericalLimits=true
    gravityScale=0.8
    simulationSpace=RootRelative
    alphaBoolBlend {blendOption=Linear}
    boundBone {boneName=dyn_apl_hair_Side_R_02}
    componentAppliedLinearAccClamp=<5000 5000 5000>
    physicsBodyDefinitions [{
      boundBone {boneName=dyn_apl_hair_Side_R_02}
      boxExtents=<10 15 15>
      constraintSetup {angularLimitsMax=<5 15 15>, angularLimitsMin=<-5 -15 -15>, angularTarget=<0 0 0>}
      localJointOffset=<2 0 0>
     }]
    relativeSpaceBone {boneName=pelvis}
    sphericalLimits [{limitRadius=7.5, drivingBone {boneName=head}, sphereLocalOffset=<5 0 0>}]
   }
   n51 = Anim Dynamics - Bone: dyn_apl_hair_Side_R_03(ComponentPose=@n50, Alpha=1, ExternalForce="0,0,0") {
    angularDampingOverride=1
    angularSpringConstant=4000
    bAngularSpring=true
    bOverrideAngularDamping=true
    bUseSphericalLimits=true
    gravityScale=0.8
    simulationSpace=RootRelative
    alphaBoolBlend {blendOption=Linear}
    boundBone {boneName=dyn_apl_hair_Side_R_03}
    componentAppliedLinearAccClamp=<5000 5000 5000>
    physicsBodyDefinitions [{
      boundBone {boneName=dyn_apl_hair_Side_R_03}
      boxExtents=<10 15 15>
      constraintSetup {angularLimitsMax=<5 15 15>, angularLimitsMin=<-5 -15 -15>, angularTarget=<0 0 0>}
      localJointOffset=<2 0 0>
     }]
    relativeSpaceBone {boneName=pelvis}
    sphericalLimits [{limitRadius=7.5, drivingBone {boneName=head}, sphereLocalOffset=<5 0 0>}]
   }
   n52 = Anim Dynamics - Bone: dyn_apl_hair_Side_R_04(ComponentPose=@n51, Alpha=1, ExternalForce="0,0,0") {
    angularDampingOverride=1
    angularSpringConstant=4000
    bAngularSpring=true
    bOverrideAngularDamping=true
    bUseSphericalLimits=true
    gravityScale=0.8
    simulationSpace=RootRelative
    alphaBoolBlend {blendOption=Linear}
    boundBone {boneName=dyn_apl_hair_Side_R_04}
    componentAppliedLinearAccClamp=<5000 5000 5000>
    physicsBodyDefinitions [{
      boundBone {boneName=dyn_apl_hair_Side_R_04}
      boxExtents=<10 15 15>
      constraintSetup {angularLimitsMax=<5 15 15>, angularLimitsMin=<-5 -15 -15>, angularTarget=<0 0 0>}
      localJointOffset=<2 0 0>
     }]
    relativeSpaceBone {boneName=pelvis}
    sphericalLimits [{limitRadius=7.5, drivingBone {boneName=head}, sphereLocalOffset=<5 0 0>}]
   }
   n53 = Anim Dynamics - Bone: dyn_apl_hair_back_L_01(ComponentPose=@n52, Alpha=1, ExternalForce="0,0,0") {
    angularDampingOverride=1
    angularSpringConstant=2500
    bAngularSpring=true
    bOverrideAngularDamping=true
    bUseSphericalLimits=true
    simulationSpace=RootRelative
    alphaBoolBlend {blendOption=Linear}
    boundBone {boneName=dyn_apl_hair_back_L_01}
    componentAppliedLinearAccClamp=<5000 5000 5000>
    physicsBodyDefinitions [{
      boundBone {boneName=dyn_apl_hair_back_L_01}
      boxExtents=<10 15 15>
      constraintSetup {angularLimitsMax=<5 15 15>, angularLimitsMin=<-5 -15 -15>, angularTarget=<0 0 0>}
      localJointOffset=<2 0 0>
     }]
    relativeSpaceBone {boneName=pelvis}
    sphericalLimits [{limitRadius=7.5, drivingBone {boneName=head}, sphereLocalOffset=<5 0 0>}]
   }
   n54 = Anim Dynamics - Bone: dyn_apl_hair_back_L_02(ComponentPose=@n53, Alpha=1, ExternalForce="0,0,0") {
    angularDampingOverride=1
    angularSpringConstant=2500
    bAngularSpring=true
    bOverrideAngularDamping=true
    bUseSphericalLimits=true
    simulationSpace=RootRelative
    alphaBoolBlend {blendOption=Linear}
    boundBone {boneName=dyn_apl_hair_back_L_02}
    componentAppliedLinearAccClamp=<5000 5000 5000>
    physicsBodyDefinitions [{
      boundBone {boneName=dyn_apl_hair_back_L_02}
      boxExtents=<10 15 15>
      constraintSetup {angularLimitsMax=<5 15 15>, angularLimitsMin=<-5 -15 -15>, angularTarget=<0 0 0>}
      localJointOffset=<2 0 0>
     }]
    relativeSpaceBone {boneName=pelvis}
    sphericalLimits [{limitRadius=7.5, drivingBone {boneName=head}, sphereLocalOffset=<5 0 0>}]
   }
   n55 = Anim Dynamics - Bone: dyn_apl_hair_back_L_03(ComponentPose=@n54, Alpha=1, ExternalForce="0,0,0") {
    angularDampingOverride=1
    angularSpringConstant=2500
    bAngularSpring=true
    bOverrideAngularDamping=true
    bUseSphericalLimits=true
    simulationSpace=RootRelative
    alphaBoolBlend {blendOption=Linear}
    boundBone {boneName=dyn_apl_hair_back_L_03}
    componentAppliedLinearAccClamp=<5000 5000 5000>
    physicsBodyDefinitions [{
      boundBone {boneName=dyn_apl_hair_back_L_03}
      boxExtents=<10 15 15>
      constraintSetup {angularLimitsMax=<5 15 15>, angularLimitsMin=<-5 -15 -15>, angularTarget=<0 0 0>}
      localJointOffset=<2 0 0>
     }]
    relativeSpaceBone {boneName=pelvis}
    sphericalLimits [{limitRadius=7.5, drivingBone {boneName=head}, sphereLocalOffset=<5 0 0>}]
   }
   n56 = Anim Dynamics - Bone: dyn_apl_hair_back_L_04(ComponentPose=@n55, Alpha=1, ExternalForce="0,0,0") {
    angularDampingOverride=1
    angularSpringConstant=2500
    bAngularSpring=true
    bOverrideAngularDamping=true
    bUseSphericalLimits=true
    simulationSpace=RootRelative
    alphaBoolBlend {blendOption=Linear}
    boundBone {boneName=dyn_apl_hair_back_L_04}
    componentAppliedLinearAccClamp=<5000 5000 5000>
    physicsBodyDefinitions [{
      boundBone {boneName=dyn_apl_hair_back_L_04}
      boxExtents=<10 15 15>
      constraintSetup {angularLimitsMax=<5 15 15>, angularLimitsMin=<-5 -15 -15>, angularTarget=<0 0 0>}
      localJointOffset=<2 0 0>
     }]
    relativeSpaceBone {boneName=pelvis}
    sphericalLimits [{limitRadius=7.5, drivingBone {boneName=head}, sphereLocalOffset=<5 0 0>}]
   }
   n57 = Anim Dynamics - Bone: dyn_apl_hairBangs_R1_01(ComponentPose=@n56, Alpha=1, ExternalForce="0,0,0") {
    angularDampingOverride=1
    angularSpringConstant=4500
    bAngularSpring=true
    bOverrideAngularDamping=true
    bUseSphericalLimits=true
    simulationSpace=RootRelative
    alphaBoolBlend {blendOption=Linear}
    boundBone {boneName=dyn_apl_hairBangs_R1_01}
    componentAppliedLinearAccClamp=<5000 5000 5000>
    physicsBodyDefinitions [{
      boundBone {boneName=dyn_apl_hairBangs_R1_01}
      boxExtents=<10 15 15>
      constraintSetup {angularLimitsMax=<5 6 12>, angularLimitsMin=<-5 -15 -15>, angularTarget=<0 0 0>}
      localJointOffset=<2 0 0>
     }]
    relativeSpaceBone {boneName=pelvis}
    sphericalLimits [{limitRadius=7.5, drivingBone {boneName=head}, sphereLocalOffset=<5 0 0>}]
   }
   n58 = Anim Dynamics - Bone: dyn_apl_hairBangs_R1_02(ComponentPose=@n57, Alpha=1, ExternalForce="0,0,0") {
    angularDampingOverride=1
    angularSpringConstant=4500
    bAngularSpring=true
    bOverrideAngularDamping=true
    bUseSphericalLimits=true
    simulationSpace=RootRelative
    alphaBoolBlend {blendOption=Linear}
    boundBone {boneName=dyn_apl_hairBangs_R1_02}
    componentAppliedLinearAccClamp=<5000 5000 5000>
    physicsBodyDefinitions [{
      boundBone {boneName=dyn_apl_hairBangs_R1_02}
      boxExtents=<10 15 15>
      constraintSetup {angularLimitsMax=<5 0 12>, angularLimitsMin=<-5 -15 -15>, angularTarget=<0 0 0>}
      localJointOffset=<2 0 0>
     }]
    relativeSpaceBone {boneName=pelvis}
    sphericalLimits [{limitRadius=7.5, drivingBone {boneName=head}, sphereLocalOffset=<5 0 0>}]
   }
   n59 = Anim Dynamics - Bone: dyn_apl_hairBangs_R1_03(ComponentPose=@n58, Alpha=1, ExternalForce="0,0,0") {
    angularDampingOverride=1
    angularSpringConstant=4500
    bAngularSpring=true
    bOverrideAngularDamping=true
    bUseSphericalLimits=true
    simulationSpace=RootRelative
    alphaBoolBlend {blendOption=Linear}
    boundBone {boneName=dyn_apl_hairBangs_R1_03}
    componentAppliedLinearAccClamp=<5000 5000 5000>
    physicsBodyDefinitions [{
      boundBone {boneName=dyn_apl_hairBangs_R1_03}
      boxExtents=<10 15 15>
      constraintSetup {angularLimitsMax=<5 0 12>, angularLimitsMin=<-5 -15 -15>, angularTarget=<0 0 0>}
      localJointOffset=<2 0 0>
     }]
    relativeSpaceBone {boneName=pelvis}
    sphericalLimits [{limitRadius=7.5, drivingBone {boneName=head}, sphereLocalOffset=<5 0 0>}]
   }
   Output Pose(Result=Component To Local(ComponentPose=@n59)) {layerGroup=None}
   Save cached pose 'InitialCache'(Pose=Input Pose()) {cacheName=InitialCache}
   Get isPistol() -> (isPistol:double)
  ]}
 EventGraph {nodes [
   Event BlueprintUpdateAnimation()
   > n1 = Sequence!(execute) -> (then_0, then_1, then_2:Exec, then_3:Exec, then_4:Exec)
   Set Arm R rot(
    execute=@n1.then_1
    Arm R rot:double=GetAlphaFromReferenceTransformDelta(SkeletalMeshComponent=Get Skel_Mesh(), BoneName=upperarm_r, BaseName=spine_05, InputRange=(X=98,Y=120), TwistAxis="1,0,0")
   ) -> (Output_Get:double)
   Set Jacket(
    execute=@n1.then_2
    Jacket:bool=CharacterItemDefinitionHasAnyMetaTag(Tags=Make Array("[0]"=(TagName="Cosmetics.MetaTags.Pencil.Upper.Jacket")))
   ) -> (Output_Get:bool)
   n2 = Branch(execute=@n1.then_3, Condition=AND Boolean(A=Get bIsCrouching(), B=Get bIsAccelerating2D()))
   > Set Moving Crouch(Moving Crouch:bool=true) -> (Output_Get:bool)
   Set Moving Crouch(execute=@n2.else, Moving Crouch=false)
   n3 = PlayerMovement(execute=@n1.then_4, AnimInstance=Self-Reference(), XDotProductCoefficient=1, YDotProductCoefficient=1, ZDotProductCoefficient=1)
   n4 = MapRangeClamped(Value=@n3.YMovement, InRangeA=0, InRangeB=-1)
   n5 = float * float(A="Absolute (Float)"(A=@n3), B=0.5)
   > Set BwdMovement(BwdMovement:float=MapRangeClamped(Value=float + float(A=@n5, B=@n4) {is=title K2Node_CommutativeAssociativeBinaryOperator}, InRangeA=1, InRangeB=0, OutRangeA=0.4, OutRangeB=0)) -> (Output_Get:float)
   n6 = Gravity Override State Function DEPRECATED(
    execute=@n1.then_0
    CharacterPartAnimInstance=Self-Reference()
    GameTimeInSeconds=GetGameTimeInSeconds()
    GravityOverrideParams=Get Gravity Override Params()
   )
   > Set GravityOverride(GravityOverride=@n6)
   Event BlueprintInitializeAnimation()
   > Set Skel_Mesh(Skel_Mesh=GetOwnerSkeletalMeshComponent())
  ]}
]
variables [
 Arm R rot {type=number}
 BwdMovement {type=number}
 Gravity Override Params {
  type=GravityOverrideParamsStruct
  default {
   skydive_DivingDown {gravityOverride=<0 0 -800>}
   skydive_FloatingDown {gravityOverride=<0 0 -600>}
   skydive_GoingUp {gravityOverride=<0 0 -800>, windAmplitude=<0 0 0>, windFrequency=<0 0 0>}
   skydive_Parachuting {gravityOverride=<0 -800 -700>}
   tactical_Sprint {gravityOverride=<0 -450 -1000>}
  }
 }
 GravityOverride {type=Vector}
 isPistol {type=number}
 Jacket {type=boolean}
 Moving Crouch {type=boolean}
 NewVar {type=boolean}
 Skel_Mesh {type=SkeletalMeshComponent}
]
